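# Post-dynamics Control Rig for the female LexaBlitz body that corrects skirt and ribbon bones.
# A reusable Clamp Capsule function confines bones against capsules on thighs and rear skirt bones to limit clipping.
# Parent constraints blend the skirt chain base, while spherical pose readers on thighs and skirt drive corrective offsets.
# Cape constrains ribbons and back skirt to pelvis, Skydive enables rear thigh clamps, and FE_idle fades front clamps for a left-side offset.
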ControlRigBlueprint /BRCosmetics/Characters/Player/Female/Medium/Bodies/F_MED_LexaBlitz/Meshes/F_MED_LexaBlitz_CtrlRig_PostDyn
graphs [
 New Function_ContainedGraph {nodes [
   n1 = RigVMFunctionEntryNode()
   Multiply = &RigVMFunction_MathTransformMul(A=@n1.CapsuleOffset, B=&RigUnit_GetTransform(bInitial=False, Item=@n1.CapsuleBone).Transform) {template="Multiply::Execute(in A,in B,out Result)"}
   RigUnit_GetTransform = &RigUnit_GetTransform(bInitial=False, Item=@n1.MoveBone)
   ClampSpatially = /Script/RigVM.RigVMFunction_MathTransformClampSpatially(
    DebugColor=(R=1,G=0,B=0,A=1)
    DebugThickness=1
    Type=Capsule
    Axis=@n1.CapsuleAxis
    Maximum=@n1.CapsuleLength
    Minimum=@n1.CapsuleWidth
    bDrawDebug=@n1.DrawDegub
    Space=@Multiply.Result
    Value=@RigUnit_GetTransform.Transform
   ) {
    is=struct RigVMUnitNode
    template="ClampSpatially::Execute(in Value,in Axis,in Type,in Minimum,in Maximum,in Space,in bDrawDebug,in DebugColor,in DebugThickness,out Result)"
   }
   > RigUnit_SetTransform = &RigUnit_SetTransform(bInitial=False, Value=@ClampSpatially.Result, Weight=@n1.Alpha, Item=@n1.MoveBone)
   > RigVMFunctionReturnNode()
  ]}
 RigVMFunctionLibrary {nodes [Clamp Capsule = RigVMCollapseNode(
    Alpha=20
    CapsuleAxis=Z
    CapsuleBone=(Type=None,Name="None")
    CapsuleLength=20
    CapsuleOffset=(Rotation=(X=0,Y=0,Z=0,W=1),Translation=(X=0,Y=0,Z=0),Scale3D=(X=1,Y=1,Z=1))
    CapsuleWidth=20
    ExecuteContext=()
    MoveBone=(Type=None,Name="None")
   )]}
 RigVMFunction_Sequence_4_ContainedGraph {nodes [
   n1 = RigVMFunctionEntryNode()
   n2 = &RigVMFunction_Sequence(ExecuteContext=@n1.ExecuteContext)
   n3 = &RigVMFunction_Sequence(ExecuteContext=@n2.B)
   n4 = &RigVMFunction_Sequence(ExecuteContext=@n3.B)
   n5 = &RigVMFunction_Sequence(ExecuteContext=@n4.B)
   RigVMFunctionReturnNode(A=@n2.A, B=@n3.A, C=@n4.A, D=@n5.A, E=@n5.B)
  ]}
 RigVMFunction_Sequence_6_ContainedGraph {nodes [
   n1 = RigVMFunctionEntryNode()
   n2 = &RigVMFunction_Sequence(ExecuteContext=@n1.ExecuteContext)
   n3 = &RigVMFunction_Sequence(ExecuteContext=@n2.B)
   n4 = &RigVMFunction_Sequence(ExecuteContext=@n3.B)
   n5 = &RigVMFunction_Sequence(ExecuteContext=@n4.B)
   RigVMFunctionReturnNode(B=@n3.A, C=@n4.A, D=@n5.A, E=@n5.B, A=@n2.A)
  ]}
 RigVMFunction_Sequence_6_ContainedGraph {nodes [
   n1 = RigVMFunctionEntryNode()
   n2 = &RigVMFunction_Sequence(ExecuteContext=@n1.ExecuteContext)
   n3 = &RigVMFunction_Sequence(ExecuteContext=@n2.B)
   RigVMFunctionReturnNode(B=@n3.A, C=@n3.B, A=@n2.A)
  ]}
 RigVMFunction_Sequence_7_ContainedGraph {nodes [
   n1 = RigVMFunctionEntryNode()
   n2 = &RigVMFunction_Sequence(ExecuteContext=@n1.ExecuteContext)
   n3 = &RigVMFunction_Sequence(ExecuteContext=@n2.B)
   RigVMFunctionReturnNode(B=@n3.A, C=@n3.B, A=@n2.A)
  ]}
 RigVMModel {nodes [
   RigUnit_BeginExecution = &RigUnit_BeginExecution()
   RigVMFunction_Sequence_1 = RigVMAggregateNode(ExecuteContext=@RigUnit_BeginExecution.ExecutePin)
   n1 = &RigVMFunction_Sequence(ExecuteContext=@RigVMFunction_Sequence_1.B)
   n2 = &RigUnit_SphericalPoseReader(
    ActiveRegionScaleFactors=(PositiveWidth=1,NegativeWidth=1,PositiveHeight=1,NegativeHeight=1)
    ActiveRegionSize=0.1
    Debug=(bDrawDebug=false,bDraw2D=false,DebugScale=25,DebugSegments=20,DebugThickness=0.25)
    DriverAxis=(X=1,Y=0,Z=0)
    DriverItem=(Type=Bone,Name="thigh_r")
    FalloffRegionScaleFactors=(PositiveWidth=1,NegativeWidth=1,PositiveHeight=1,NegativeHeight=1)
    FalloffSize=0.45
    FlipHeightScaling=false
    FlipWidthScaling=true
    OptionalParentItem=(Type=Bone)
    RotationOffset=(X=-90,Y=0,Z=-36)
    ExecutePin=@n1.A
   )
   AlphaInterp = &RigVMFunction_AlphaInterp(InterpSpeedDecreasing=10, InterpSpeedIncreasing=10, Scale=1, bClampResult=False, Value=@n2.OutputParam)
   > RigVMVariableNode(Value=@AlphaInterp.Result) {variable=AlphaSPR_thigh_R}
   n3 = &RigUnit_SphericalPoseReader(
    ActiveRegionScaleFactors=(PositiveWidth=1,NegativeWidth=1,PositiveHeight=1,NegativeHeight=1)
    ActiveRegionSize=0.1
    Debug=(bDrawDebug=false,DebugScale=25,DebugSegments=20,DebugThickness=0.25)
    DriverAxis=(X=-1,Y=0,Z=0)
    DriverItem=(Type=Bone,Name="thigh_l")
    FalloffRegionScaleFactors=(PositiveWidth=1,NegativeWidth=1,PositiveHeight=1,NegativeHeight=1)
    FalloffSize=0.45
    OptionalParentItem=(Type=Bone)
    RotationOffset=(X=114,Y=16,Z=-39)
    ExecutePin=@n1.B
   )
   n4 = &RigVMFunction_AlphaInterp(InterpSpeedDecreasing=2, InterpSpeedIncreasing=2, Scale=1, bClampResult=False, Value=@n3.OutputParam)
   > RigVMVariableNode(Value=@n4.Result) {variable=AlphaSPR_thigh_L}
   RigVMFunction_Sequence = &RigVMFunction_Sequence(ExecuteContext=@RigVMFunction_Sequence_1.C)
   &RigUnit_ParentConstraint(
    AdvancedSettings=(InterpolationType=Average,RotationOrderForFilter=XZY)
    Child=(Type=Bone,Name="base_skirt_chain")
    Filter=(TranslationFilter=(bX=False,bY=False,bZ=False),RotationFilter=(bX=True,bY=True,bZ=True),ScaleFilter=(bX=True,bY=True,bZ=True))
    Parents=((Item=(Type=Bone,Name="dyn_skirt_fr_01_r"),Weight=0.5),(Item=(Type=Bone,Name="dyn_skirt_fr_01_l"),Weight=0.5))
    Weight=1
    ExecutePin=@RigVMFunction_Sequence.A
   )
   > &RigUnit_ParentConstraint(
    AdvancedSettings=(RotationOrderForFilter=XZY)
    Child=(Type=Bone,Name="base_skirt_chain")
    Parents="((Item=(Type=Bone,Name=\"dyn_skirt_fr_03_l\"),Weight=0.0131),(Item=(Type=Bone,Name=\"dyn_skirt_fr_side_01_l\"),Weight=0.0158),(Item=(Type=Bone,Name=\"dyn_skirt_fr_03_r\"),Weight=0.0277),(Item=(Type=Bone,Name=\"dyn_skirt_fr_side_01_r\"),Weight=0.1082),(Item=(Type=B ...(truncated)"
    Weight=0.5
   )
   > &RigUnit_ParentConstraint(
    AdvancedSettings=(RotationOrderForFilter=XZY)
    Child=(Type=Bone,Name="base_skirt_chain")
    Parents="((Item=(Type=Bone,Name=\"dyn_skirt_fr_03_r\"),Weight=0.013),(Item=(Type=Bone,Name=\"dyn_skirt_fr_side_01_r\"),Weight=0.0157),(Item=(Type=Bone,Name=\"dyn_skirt_fr_03_l\"),Weight=0.0277),(Item=(Type=Bone,Name=\"dyn_skirt_fr_side_01_l\"),Weight=0.1084),(Item=(Type=Bo ...(truncated)"
    Weight=0.5
   )
   > n5 = &RigUnit_SphericalPoseReader(
    ActiveRegionScaleFactors=(PositiveWidth=1,NegativeWidth=0.5,PositiveHeight=1,NegativeHeight=1)
    ActiveRegionSize=0.1
    Debug=(bDrawDebug=false,bDraw2D=false,DebugScale=25,DebugSegments=20,DebugThickness=0.25)
    DriverAxis=(X=1,Y=0,Z=0)
    DriverItem=(Type=Bone,Name="dyn_skirt_fr_01_r")
    FalloffRegionScaleFactors=(PositiveWidth=1,NegativeWidth=1,PositiveHeight=1,NegativeHeight=1)
    FalloffSize=0.3124
    FlipHeightScaling=false
    FlipWidthScaling=true
    OptionalParentItem=(Type=Bone,Name="pelvis")
    RotationOffset=(X=-91,Y=0,Z=105)
   )
   > &RigUnit_OffsetTransformForItem(
    Item=(Type=Bone,Name="base_skirt_chain")
    OffsetTransform=(Rotation=(X=0,Y=-0.0436,Z=0,W=0.999),Translation=(X=0,Y=0,Z=0),Scale3D=(X=1,Y=1,Z=1))
    bPropagateToChildren=True
    Weight=@n5.OutputParam
   )
   &RigUnit_ParentConstraint(
    AdvancedSettings=(InterpolationType=Average,RotationOrderForFilter=XZY)
    Child=(Type=Bone,Name="dyn_skirt_fr_01_r")
    Parents=((Item=(Type=Bone,Name="thigh_twist_01_r"),Weight=0.5),(Item=(Type=Bone,Name="thigh_twist_01_r"),Weight=0.5))
    Weight=0.5
    ExecutePin=@RigVMFunction_Sequence.B
   )
   RigVMFunction_ControlFlowBranch = &RigVMFunction_ControlFlowBranch(Condition=RigVMVariableNode() {variable=Cape}.Value, ExecuteContext=@RigVMFunction_Sequence_1.E)
   RigVMFunction_Sequence_5 = RigVMAggregateNode(ExecuteContext=@RigVMFunction_ControlFlowBranch.True)
   &RigUnit_ParentConstraint(
    AdvancedSettings=(InterpolationType=Average,RotationOrderForFilter=XZY)
    Child=(Type=Bone,Name="dyn_ribbon_01_l")
    Filter=(TranslationFilter=(bX=True,bY=True,bZ=True),RotationFilter=(bX=True,bY=False,bZ=False),ScaleFilter=(bX=True,bY=True,bZ=True))
    Parents=((Item=(Type=Bone,Name="pelvis"),Weight=1))
    Weight=1
    ExecutePin=@RigVMFunction_Sequence_5.A
   )
   > &RigUnit_ParentConstraint(
    AdvancedSettings=(InterpolationType=Average,RotationOrderForFilter=XZY)
    Child=(Type=Bone,Name="dyn_ribbon_02_l")
    Filter=(TranslationFilter=(bX=True,bY=True,bZ=True),RotationFilter=(bX=True,bY=False,bZ=False),ScaleFilter=(bX=True,bY=True,bZ=True))
    Parents=((Item=(Type=Bone,Name="pelvis"),Weight=1))
    Weight=1
   )
   > &RigUnit_ParentConstraint(
    AdvancedSettings=(InterpolationType=Average,RotationOrderForFilter=XZY)
    Child=(Type=Bone,Name="dyn_ribbon_03_l")
    Filter=(TranslationFilter=(bX=True,bY=True,bZ=True),RotationFilter=(bX=True,bY=False,bZ=False),ScaleFilter=(bX=True,bY=True,bZ=True))
    Parents=((Item=(Type=Bone,Name="pelvis"),Weight=1))
    Weight=1
   )
   > &RigUnit_ParentConstraint(
    AdvancedSettings=(InterpolationType=Average,RotationOrderForFilter=XZY)
    Child=(Type=Bone,Name="dyn_ribbon_03_l")
    Parents=((Item=(Type=Bone,Name="pelvis"),Weight=1))
    Weight=0.5
   )
   &RigUnit_ParentConstraint(
    AdvancedSettings=(InterpolationType=Average,RotationOrderForFilter=XZY)
    Child=(Type=Bone,Name="dyn_ribbon_01_r")
    Filter=(TranslationFilter=(bX=True,bY=True,bZ=True),RotationFilter=(bX=True,bY=False,bZ=False),ScaleFilter=(bX=True,bY=True,bZ=True))
    Parents=((Item=(Type=Bone,Name="pelvis"),Weight=1))
    Weight=1
    ExecutePin=@RigVMFunction_Sequence_5.B
   )
   > &RigUnit_ParentConstraint(
    AdvancedSettings=(InterpolationType=Average,RotationOrderForFilter=XZY)
    Child=(Type=Bone,Name="dyn_ribbon_02_r")
    Filter=(TranslationFilter=(bX=True,bY=True,bZ=True),RotationFilter=(bX=True,bY=False,bZ=False),ScaleFilter=(bX=True,bY=True,bZ=True))
    Parents=((Item=(Type=Bone,Name="pelvis"),Weight=1))
    Weight=1
   )
   > &RigUnit_ParentConstraint(
    AdvancedSettings=(InterpolationType=Average,RotationOrderForFilter=XZY)
    Child=(Type=Bone,Name="dyn_ribbon_03_r")
    Filter=(TranslationFilter=(bX=True,bY=True,bZ=True),RotationFilter=(bX=True,bY=False,bZ=False),ScaleFilter=(bX=True,bY=True,bZ=True))
    Parents=((Item=(Type=Bone,Name="pelvis"),Weight=1))
    Weight=1
   )
   > &RigUnit_ParentConstraint(
    AdvancedSettings=(InterpolationType=Average,RotationOrderForFilter=XZY)
    Child=(Type=Bone,Name="dyn_ribbon_03_r")
    Parents=((Item=(Type=Bone,Name="pelvis"),Weight=1))
    Weight=0.5
   )
   &RigUnit_ParentConstraint(
    AdvancedSettings=(InterpolationType=Average,RotationOrderForFilter=XZY)
    Child=(Type=Bone,Name="dyn_skirt_bk_02_c")
    Parents=((Item=(Type=Bone,Name="pelvis"),Weight=1))
    Weight=1
    ExecutePin=@RigVMFunction_Sequence_5.C
   )
   > &RigUnit_ParentConstraint(
    AdvancedSettings=(InterpolationType=Average,RotationOrderForFilter=XZY)
    Child=(Type=Bone,Name="dyn_skirt_bk_03_c")
    Parents=((Item=(Type=Bone,Name="pelvis"),Weight=1))
    Weight=1
   )
   &RigUnit_ParentConstraint(
    AdvancedSettings=(InterpolationType=Average,RotationOrderForFilter=XZY)
    Child=(Type=Bone,Name="dyn_skirt_bk_side_02_l")
    Parents=((Item=(Type=Bone,Name="pelvis"),Weight=1))
    Weight=1
    ExecutePin=@RigVMFunction_Sequence_5.D
   )
   > &RigUnit_ParentConstraint(
    AdvancedSettings=(InterpolationType=Average,RotationOrderForFilter=XZY)
    Child=(Type=Bone,Name="dyn_skirt_bk_side_02_r")
    Parents=((Item=(Type=Bone,Name="pelvis"),Weight=1))
    Weight=1
   )
   n6 = &RigVMFunction_Sequence(ExecuteContext=@RigVMFunction_Sequence_1.A)
   Clamp Capsule_3 = RigVMFunctionReferenceNode(
    Alpha=1
    CapsuleAxis=Z
    CapsuleBone=(Type=Bone,Name="dyn_skirt_bk_side_02_l")
    CapsuleLength=30
    CapsuleOffset=(Rotation=(X=0,Y=0,Z=0,W=1),Translation=(X=0,Y=5,Z=0),Scale3D=(X=1,Y=1,Z=1))
    CapsuleWidth=7
    DrawDegub=false
    MoveBone=(Type=Bone,Name="dyn_ribbon_02_l")
    ExecuteContext=@n6.A
   )
   > Clamp Capsule_4 = RigVMFunctionReferenceNode(
    Alpha=1
    CapsuleAxis=Z
    CapsuleBone=(Type=Bone,Name="dyn_skirt_bk_side_02_r")
    CapsuleLength=30
    CapsuleOffset=(Rotation=(X=0,Y=0,Z=0,W=1),Translation=(X=0,Y=5,Z=0),Scale3D=(X=1,Y=1,Z=1))
    CapsuleWidth=7
    DrawDegub=false
    MoveBone=(Type=Bone,Name="dyn_ribbon_02_r")
   )
   n7 = &RigVMFunction_MathBoolToFloat(Value=RigVMVariableNode() {variable=Skydive}.Value)
   Clamp Capsule_6 = RigVMFunctionReferenceNode(
    CapsuleAxis=X
    CapsuleBone=(Type=Bone,Name="thigh_l")
    CapsuleLength=30
    CapsuleOffset=(Rotation=(X=0,Y=0,Z=0,W=1),Translation=(X=-9,Y=-1,Z=0),Scale3D=(X=1,Y=1,Z=1))
    CapsuleWidth=15
    DrawDegub=false
    MoveBone=(Type=Bone,Name="dyn_skirt_bk_side_02_l")
    ExecuteContext=@n6.B
    Alpha=@n7.Result
   )
   > Clamp Capsule_7 = RigVMFunctionReferenceNode(
    CapsuleAxis=X
    CapsuleBone=(Type=Bone,Name="thigh_r")
    CapsuleLength=30
    CapsuleOffset=(Rotation=(X=0,Y=0,Z=0,W=1),Translation=(X=9,Y=-1,Z=0),Scale3D=(X=1,Y=1,Z=1))
    CapsuleWidth=18
    DrawDegub=false
    MoveBone=(Type=Bone,Name="dyn_skirt_bk_side_02_r")
    Alpha=@n7.Result
   )
   RigVMFunction_Sequence_3 = RigVMAggregateNode(ExecuteContext=@RigVMFunction_Sequence_1.D)
   n8 = &RigUnit_SphericalPoseReader(
    ActiveRegionScaleFactors=(PositiveWidth=3.5,NegativeWidth=0.5,PositiveHeight=3,NegativeHeight=3)
    ActiveRegionSize=0.1
    Debug=(bDrawDebug=False,bDraw2D=false,DebugScale=25,DebugSegments=20,DebugThickness=0.25)
    DriverAxis=(X=1,Y=0,Z=0)
    DriverItem=(Type=Bone,Name="thigh_r")
    FalloffRegionScaleFactors=(PositiveWidth=1,NegativeWidth=1,PositiveHeight=1,NegativeHeight=1)
    FalloffSize=0.45
    FlipHeightScaling=false
    FlipWidthScaling=true
    OptionalParentItem=(Type=Bone)
    RotationOffset=(X=-90,Y=0,Z=0)
    ExecutePin=@RigVMFunction_Sequence_3.A
   )
   > &RigUnit_OffsetTransformForItem(
    Item=(Type=Bone,Name="extra_skirt_fr_r")
    OffsetTransform=(Rotation=(X=0,Y=0,Z=0,W=1),Translation=(X=4.5927,Y=-6.4482,Z=0),Scale3D=(X=1,Y=1,Z=1))
    bPropagateToChildren=False
    Weight=@n8.OutputParam
   )
   n9 = &RigUnit_SphericalPoseReader(
    ActiveRegionScaleFactors=(PositiveWidth=3.5,NegativeWidth=0.5,PositiveHeight=3,NegativeHeight=3)
    ActiveRegionSize=0.1
    Debug=(bDrawDebug=true,DebugScale=25,DebugSegments=20,DebugThickness=0.25)
    DriverAxis=(X=-1,Y=0,Z=0)
    DriverItem=(Type=Bone,Name="thigh_l")
    FalloffRegionScaleFactors=(PositiveWidth=1,NegativeWidth=1,PositiveHeight=1,NegativeHeight=1)
    FalloffSize=0.45
    OptionalParentItem=(Type=Bone)
    RotationOffset=(X=90,Y=0,Z=0)
    ExecutePin=@RigVMFunction_Sequence_3.B
   )
   > &RigUnit_OffsetTransformForItem(
    Item=(Type=Bone,Name="extra_skirt_fr_l")
    OffsetTransform=(Rotation=(X=0,Y=0,Z=0,W=1),Translation=(X=-4.5927,Y=-7.5518,Z=-5),Scale3D=(X=1,Y=1,Z=1))
    bPropagateToChildren=False
    Weight=@n9.OutputParam
   )
   RigVMFunction_Sequence_6 = RigVMAggregateNode(ExecuteContext=@RigVMFunction_Sequence_3.C)
   n10 = RigVMVariableNode() {variable=FE_idle}
   RigVMFunction_MathBoolNot = &RigVMFunction_MathBoolNot(Value=@n10.Value)
   RigVMFunction_MathBoolToFloat = &RigVMFunction_MathBoolToFloat(Value=@RigVMFunction_MathBoolNot.Result)
   n11 = &RigVMFunction_AlphaInterp(InterpSpeedDecreasing=10, InterpSpeedIncreasing=10, Scale=1, bClampResult=False, Value=@RigVMFunction_MathBoolToFloat.Result)
   Clamp Capsule_8 = RigVMFunctionReferenceNode(
    CapsuleAxis=X
    CapsuleBone=(Type=Bone,Name="thigh_l")
    CapsuleLength=60
    CapsuleOffset=(Rotation=(X=0,Y=0,Z=0,W=1),Translation=(X=-15,Y=0,Z=0),Scale3D=(X=1,Y=1,Z=1))
    CapsuleWidth=20
    DrawDegub=false
    MoveBone=(Type=Bone,Name="extra_skirt_fr_l")
    ExecuteContext=@RigVMFunction_Sequence_6.A
    Alpha=@n11.Result
   )
   Clamp Capsule_2 = RigVMFunctionReferenceNode(
    CapsuleAxis=X
    CapsuleBone=(Type=Bone,Name="thigh_r")
    CapsuleLength=60
    CapsuleOffset=(Rotation=(X=0,Y=0,Z=0,W=1),Translation=(X=15,Y=0,Z=0),Scale3D=(X=1,Y=1,Z=1))
    CapsuleWidth=11
    DrawDegub=True
    MoveBone=(Type=Bone,Name="dyn_skirt_fr_01_r")
    ExecuteContext=@RigVMFunction_Sequence_6.B
    Alpha=@n11.Result
   )
   > Clamp Capsule_1 = RigVMFunctionReferenceNode(
    CapsuleAxis=X
    CapsuleBone=(Type=Bone,Name="thigh_r")
    CapsuleLength=60
    CapsuleOffset=(Rotation=(X=0,Y=0,Z=0,W=1),Translation=(X=15,Y=0,Z=0),Scale3D=(X=1,Y=1,Z=1))
    CapsuleWidth=20
    DrawDegub=True
    MoveBone=(Type=Bone,Name="extra_skirt_fr_r")
    Alpha=@n11.Result
   )
   > Clamp Capsule_5 = RigVMFunctionReferenceNode(
    CapsuleAxis=X
    CapsuleBone=(Type=Bone,Name="thigh_r")
    CapsuleLength=60
    CapsuleOffset=(Rotation=(X=0,Y=0,Z=0,W=1),Translation=(X=15,Y=0,Z=0),Scale3D=(X=1,Y=1,Z=1))
    CapsuleWidth=7
    DrawDegub=True
    MoveBone=(Type=Bone,Name="dyn_skirt_fr_side_base_r")
    Alpha=@n11.Result
   )
   n12 = &RigVMFunction_AlphaInterp(
    InterpSpeedDecreasing=10
    InterpSpeedIncreasing=10
    Scale=1
    bClampResult=False
    Value=&RigVMFunction_MathBoolToFloat(Value=@n10.Value).Result
   )
   &RigUnit_OffsetTransformForItem(
    Item=(Type=Bone,Name="dyn_skirt_fr_side_02_l")
    OffsetTransform=(Rotation=(X=0.1795,Y=0.1707,Z=-0.0316,W=0.9683),Translation=(X=0,Y=0,Z=0),Scale3D=(X=1,Y=1,Z=1))
    bPropagateToChildren=False
    Weight=@n12.Result
    ExecutePin=@RigVMFunction_Sequence_6.C
   )
   &RigVMFunction_AlphaInterp(InterpSpeedDecreasing=10, InterpSpeedIncreasing=10, Scale=1, bClampResult=False, Value=@n8.OutputParam)
   &RigVMFunction_AlphaInterp(InterpSpeedDecreasing=2, InterpSpeedIncreasing=2, Scale=1, bClampResult=False, Value=@n9.OutputParam)
  ]}
]
hierarchy [
 root {loc=<0 0 0>, rot=<0 0 0 1>}
 pelvis {parent=root, loc=<0 -0.5301 98.0571>, rot=<0.0127 0.707 -0.0127 -0.707>}
 spine_01 {parent=pelvis, loc=<6.1208 -0.9328 0>, rot=<0 0 0.034 -0.9994>}
 spine_02 {parent=spine_01, loc=<6.5997 -0.2095 0>, rot=<0 0 0.0238 -0.9997>}
 spine_03 {parent=spine_02, loc=<6.4914 -0.1237 0>, rot=<0 0 0.0149 -0.9999>}
 spine_04 {parent=spine_03, loc=<6.9598 0.0794 0>, rot=<0 0 -0.0132 -0.9999>}
 spine_05 {parent=spine_04, loc=<6.9987 0 0>, rot=<0 0 -0.0015 -1>}
 neck_01 {parent=spine_05, loc=<11.3138 0.1529 0>, rot=<0 0 -0.2003 -0.9797>}
 neck_02 {parent=neck_01, loc=<4.5627 0 0>, rot=<0 0 0.0504 -0.9987>}
 head {parent=neck_02, loc=<5.3914 0 0>, rot=<0 0 0.115 -0.9934>}
 FACIAL_C_FacialRoot {parent=head, loc=<1.006 0.627 0>, rot=<0.5021 -0.4979 -0.5021 -0.4979>}
 FACIAL_C_UnderChin {parent=FACIAL_C_FacialRoot, loc=<0 7.7941 2.7356>, rot=<0.5 0 0 -0.866>}
 FACIAL_L_UnderChin {parent=FACIAL_C_FacialRoot, loc=<2.2845 6.4992 1.7895>, rot=<0.4025 -0.355 0.0823 -0.8398>}
 FACIAL_R_UnderChin {parent=FACIAL_C_FacialRoot, loc=<-2.2845 6.4992 1.7895>, rot=<0.4025 0.355 -0.0823 -0.8398>}
 FACIAL_C_Jaw {parent=FACIAL_C_FacialRoot, loc=<0 1.8864 1.0077>, rot=<0.342 0 0 -0.9397>}
 FACIAL_C_Jawline {parent=FACIAL_C_Jaw, loc=<-0.0002 2.3161 7.0033>, rot=<0.2588 0 0 -0.9659>}
 FACIAL_L_Jawline {parent=FACIAL_C_FacialRoot, loc=<4.4657 5.2391 3.79>, rot=<0.2092 0.1578 -0.0547 -0.9635>}
 FACIAL_R_Jawline {parent=FACIAL_C_FacialRoot, loc=<-4.4657 5.2391 3.79>, rot=<0.2092 -0.1578 0.0547 -0.9635>}
 clavicle_l {parent=spine_05, loc=<5.4831 -0.1053 -3.2946>, rot=<0.0465 -0.7289 -0.009 -0.6829>}
 upperarm_l {parent=clavicle_l, loc=<9.8658 -1.2136 -0.4105>, rot=<0.0164 -0.3245 -0.0269 -0.9454>}
 lowerarm_l {parent=upperarm_l, loc=<26.465 -0.8345 0.259>, rot=<0.0036 0.0161 -0.2298 -0.9731>}
 hand_l {parent=lowerarm_l, loc=<22.7355 -0.3203 -0.6105>, rot=<-0.6784 -0.0223 0.0735 -0.7307>}
 thumb_01_l {parent=hand_l, loc=<1.3118 -0.3191 -2.459>, rot=<0.6546 -0.4042 0.0664 -0.6354>}
 thumb_02_l {parent=thumb_01_l, loc=<5.9336 -0.0448 0.0401>, rot=<-0.0037 0.0213 0.298 -0.9543>}
 thumb_03_l {parent=thumb_02_l, loc=<2.2223 0.3682 -0.0907>, rot=<0.0118 -0.013 0.0066 -0.9998>}
 index_metacarpal_l {parent=hand_l, loc=<1.6535 0.0183 -2.96>, rot=<0.0158 -0.0124 -0.0173 -0.9996>}
 index_01_l {parent=index_metacarpal_l, loc=<8.3008 0.0357 -0.0524>, rot=<-0.0044 0.0485 0.0924 -0.9945>}
 index_02_l {parent=index_01_l, loc=<3.3587 -0.3004 0.0452>, rot=<-0.002 0.0027 0.3225 -0.9466>}
 index_03_l {parent=index_02_l, loc=<2.5578 -0.123 -0.0252>, rot=<-0.002 0.0067 0.2558 -0.9667>}
 middle_metacarpal_l {parent=hand_l, loc=<2.1856 -0.1725 -0.5984>, rot=<0.006 0.0168 -0.0329 -0.9993>}
 middle_01_l {parent=middle_metacarpal_l, loc=<7.3275 -0.2447 0.0129>, rot=<-0.0208 0.0042 0.0858 -0.9961>}
 middle_02_l {parent=middle_01_l, loc=<4.0123 0.1401 0.1175>, rot=<-0.0218 0.0258 0.3512 -0.9357>}
 middle_03_l {parent=middle_02_l, loc=<2.6077 0.0534 0.1373>, rot=<0.0082 0.0188 0.1708 -0.9851>}
 ring_metacarpal_l {parent=hand_l, loc=<2.2074 -0.3625 1.3638>, rot=<-0.0284 0.0722 -0.0458 -0.9959>}
 ring_01_l {parent=ring_metacarpal_l, loc=<7.0295 -0.6197 -0.1446>, rot=<-0.0489 -0.0252 0.1892 -0.9804>}
 ring_02_l {parent=ring_01_l, loc=<3.6174 0.4647 -0.1027>, rot=<0 0 0.2806 -0.9598>}
 ring_03_l {parent=ring_02_l, loc=<2.4152 0.114 0.0199>, rot=<0 0 0.1189 -0.9929>}
 pinky_metacarpal_l {parent=hand_l, loc=<1.7616 -0.9753 3.3814>, rot=<-0.0397 0.0814 -0.069 -0.9935>}
 pinky_01_l {parent=pinky_metacarpal_l, loc=<7.1709 -0.4831 -0.1094>, rot=<-0.0654 -0.0035 0.2242 -0.9723>}
 pinky_02_l {parent=pinky_01_l, loc=<2.8274 0.0893 -0.1043>, rot=<0.0037 0.0011 0.2854 -0.9584>}
 pinky_03_l {parent=pinky_02_l, loc=<2.2046 0.0307 -0.0078>, rot=<-0.0002 0.0098 0.1445 -0.9895>}
 deform_wrist_dn_base_l {parent=hand_l, loc=<0 -1.5326 0>, rot=<0 0 0 1>}
 deform_wrist_dn_l {parent=deform_wrist_dn_base_l, loc=<0 0 0>, rot=<0 0 0 1>}
 deform_wrist_up_base_l {parent=hand_l, loc=<0 2.3009 0>, rot=<0 0 0 1>}
 deform_wrist_up_l {parent=deform_wrist_up_base_l, loc=<0 0 0>, rot=<0 0 0 1>}
 lowerarm_twist_02_l {parent=lowerarm_l, loc=<10.6707 0 0>, rot=<0 0 0 1>}
 lowerarm_twist_01_l {parent=lowerarm_l, loc=<18.316 0 0>, rot=<0 0 0 1>}
 deform_elbow_in_base_l {parent=lowerarm_l, loc=<0.1834 3.6062 -0.2572>, rot=<0 0 0 1>}
 deform_elbow_in_l {parent=deform_elbow_in_base_l, loc=<0 0 0>, rot=<0 0 0 -1>}
 deform_elbow_out_base_l {parent=lowerarm_l, loc=<0 -4 0>, rot=<0 0 0 1>}
 deform_elbow_out_l {parent=deform_elbow_out_base_l, loc=<0 0 0>, rot=<0 0 0 1>}
 upperarm_twist_01_l {parent=upperarm_l, loc=<7.2135 -0.1134 0>, rot=<0 0 0 1>}
 deform_upperarm_bk_base_l {parent=upperarm_twist_01_l, loc=<-6.4053 -4.3746 0.314>, rot=<0 0 0 1>}
 deform_upperarm_bk_l {parent=deform_upperarm_bk_base_l, loc=<0 0 0>, rot=<0 0 0 1>}
 deform_upperarm_fr_base_l {parent=upperarm_twist_01_l, loc=<-5.6312 5.092 0.3354>, rot=<0 0 0 1>}
 deform_upperarm_fr_l {parent=deform_upperarm_fr_base_l, loc=<0 0 0>, rot=<0 0 0 1>}
 deform_upperarm_up_base_l {parent=upperarm_twist_01_l, loc=<-8.891 0.4825 3.9517>, rot=<0 0 0 1>}
 deform_upperarm_up_l {parent=deform_upperarm_up_base_l, loc=<0 0 0>, rot=<0 0 0 1>}
 upperarm_twist_02_l {parent=upperarm_l, loc=<19.7998 -0.1134 0>, rot=<0 0 0 1>}
 clavicle_r {parent=spine_05, loc=<5.4831 -0.1053 3.2946>, rot=<0.7289 0.0465 -0.6829 0.009>}
 upperarm_r {parent=clavicle_r, loc=<-9.8658 1.2136 0.4105>, rot=<0.0164 -0.3245 -0.0269 -0.9454>}
 lowerarm_r {parent=upperarm_r, loc=<-26.465 0.8345 -0.259>, rot=<0.0036 0.0161 -0.2298 -0.9731>}
 hand_r {parent=lowerarm_r, loc=<-22.7355 0.3203 0.6106>, rot=<-0.6784 -0.0223 0.0735 -0.7307>}
 thumb_01_r {parent=hand_r, loc=<-1.3118 0.3191 2.459>, rot=<0.6546 -0.4042 0.0664 -0.6354>}
 thumb_02_r {parent=thumb_01_r, loc=<-5.9336 0.0448 -0.0401>, rot=<-0.0037 0.0213 0.298 -0.9543>}
 thumb_03_r {parent=thumb_02_r, loc=<-2.2223 -0.3682 0.0907>, rot=<0.0118 -0.013 0.0066 -0.9998>}
 index_metacarpal_r {parent=hand_r, loc=<-1.6535 -0.0183 2.96>, rot=<0.0158 -0.0124 -0.0173 -0.9996>}
 index_01_r {parent=index_metacarpal_r, loc=<-8.3008 -0.0357 0.0524>, rot=<-0.0044 0.0485 0.0924 -0.9945>}
 index_02_r {parent=index_01_r, loc=<-3.3587 0.3004 -0.0452>, rot=<-0.002 0.0027 0.3225 -0.9466>}
 index_03_r {parent=index_02_r, loc=<-2.5578 0.123 0.0252>, rot=<-0.002 0.0067 0.2558 -0.9667>}
 middle_metacarpal_r {parent=hand_r, loc=<-2.1856 0.1725 0.5984>, rot=<0.006 0.0168 -0.0329 -0.9993>}
 middle_01_r {parent=middle_metacarpal_r, loc=<-7.3275 0.2447 -0.0129>, rot=<-0.0208 0.0042 0.0858 -0.9961>}
 middle_02_r {parent=middle_01_r, loc=<-4.0123 -0.1401 -0.1175>, rot=<-0.0218 0.0258 0.3512 -0.9357>}
 middle_03_r {parent=middle_02_r, loc=<-2.6077 -0.0534 -0.1373>, rot=<0.0082 0.0188 0.1708 -0.9851>}
 ring_metacarpal_r {parent=hand_r, loc=<-2.2074 0.3625 -1.3638>, rot=<-0.0284 0.0722 -0.0458 -0.9959>}
 ring_01_r {parent=ring_metacarpal_r, loc=<-7.0295 0.6197 0.1446>, rot=<-0.0489 -0.0252 0.1892 -0.9804>}
 ring_02_r {parent=ring_01_r, loc=<-3.6174 -0.4647 0.1027>, rot=<0 0 0.2806 -0.9598>}
 ring_03_r {parent=ring_02_r, loc=<-2.4152 -0.114 -0.0199>, rot=<0 0 0.1189 -0.9929>}
 pinky_metacarpal_r {parent=hand_r, loc=<-1.7616 0.9753 -3.3814>, rot=<-0.0397 0.0814 -0.069 -0.9935>}
 pinky_01_r {parent=pinky_metacarpal_r, loc=<-7.1709 0.4831 0.1094>, rot=<-0.0654 -0.0035 0.2242 -0.9723>}
 pinky_02_r {parent=pinky_01_r, loc=<-2.8274 -0.0893 0.1043>, rot=<0.0037 0.0011 0.2854 -0.9584>}
 pinky_03_r {parent=pinky_02_r, loc=<-2.2046 -0.0307 0.0078>, rot=<-0.0002 0.0098 0.1445 -0.9895>}
 deform_wrist_dn_base_r {parent=hand_r, loc=<-0.0001 1.5328 0>, rot=<0 0 0 1>}
 deform_wrist_dn_r {parent=deform_wrist_dn_base_r, loc=<0 0 0>, rot=<0 0 0 1>}
 deform_wrist_up_base_r {parent=hand_r, loc=<-0.0002 -2.3007 -0.0001>, rot=<0 0 0 1>}
 deform_wrist_up_r {parent=deform_wrist_up_base_r, loc=<0 0 0>, rot=<0 0 0 1>}
 lowerarm_twist_02_r {parent=lowerarm_r, loc=<-10.6707 0 0>, rot=<0 0 0 -1>}
 lowerarm_twist_01_r {parent=lowerarm_r, loc=<-18.316 0 0>, rot=<0 0 0 -1>}
 deform_elbow_in_base_r {parent=lowerarm_r, loc=<-0.1834 -3.6062 0.2572>, rot=<0 0 0 -1>}
 deform_elbow_in_r {parent=deform_elbow_in_base_r, loc=<0 0 0>, rot=<0 0 0 -1>}
 deform_elbow_out_base_r {parent=lowerarm_r, loc=<0 4 -0.0001>, rot=<0 0 0 -1>}
 deform_elbow_out_r {parent=deform_elbow_out_base_r, loc=<0 0 0>, rot=<0 0 0 1>}
 upperarm_twist_01_r {parent=upperarm_r, loc=<-7.2135 0.1134 0>, rot=<0 0 0 -1>}
 deform_upperarm_bk_base_r {parent=upperarm_twist_01_r, loc=<6.4053 4.3746 -0.314>, rot=<0 0 0 1>}
 deform_upperarm_bk_r {parent=deform_upperarm_bk_base_r, loc=<0 0 0>, rot=<0 0 0 1>}
 deform_upperarm_fr_base_r {parent=upperarm_twist_01_r, loc=<5.6312 -5.092 -0.3354>, rot=<0 0 0 1>}
 deform_upperarm_fr_r {parent=deform_upperarm_fr_base_r, loc=<0 0 0>, rot=<0 0 0 1>}
 deform_upperarm_up_base_r {parent=upperarm_twist_01_r, loc=<8.891 -0.4825 -3.9517>, rot=<0 0 0 1>}
 deform_upperarm_up_r {parent=deform_upperarm_up_base_r, loc=<0 0 0>, rot=<0 0 0 1>}
 upperarm_twist_02_r {parent=upperarm_r, loc=<-19.7998 0.1134 0>, rot=<0 0 0 -1>}
 attach_backpack {parent=spine_05, loc=<-0.124 -13.4692 0.0003>, rot=<-0.0092 -0.707 0.0092 -0.707>, scale=<1.0001 0.9999 1>}
 Backpack_BR {parent=attach_backpack, type=Null, loc=<0 0 0>, rot=<0 0 0 1>}
 attach_cape {parent=spine_05, loc=<9.0154 -11.7111 0.0011>, rot=<0.0617 -0.7044 -0.0617 -0.7044>, scale=<1.0001 0.9999 1>}
 cape {parent=attach_cape, type=Null, loc=<0 0 0>, rot=<0 0 0 1>}
 extra_neck {parent=spine_05, loc=<13.6069 1.1312 0>, rot=<0 0 -0.2003 -0.9797>}
 deform_pec_aim_base_l {parent=spine_04, loc=<3.2456 9.3309 -1.2237>, rot=<0.0603 -0.4033 0.3318 -0.8507>}
 deform_pec_aim_l {parent=deform_pec_aim_base_l, loc=<0 0 0>, rot=<0 0 0 1>}
 deform_pec_base_l {parent=deform_pec_aim_l, loc=<9.3665 0 0>, rot=<0 0 0 1>}
 deform_pec_l {parent=deform_pec_base_l, loc=<0 0 0>, rot=<0 0 0 1>}
 deform_pec_aim_base_r {parent=spine_04, loc=<3.2452 9.3309 1.2237>, rot=<-0.4033 -0.0603 0.8507 0.3318>}
 deform_pec_aim_r {parent=deform_pec_aim_base_r, loc=<0 0 0>, rot=<0 0 0 1>}
 deform_pec_base_r {parent=deform_pec_aim_r, loc=<-9.3667 -0.0003 -0.0004>, rot=<0 0 0 1>}
 deform_pec_r {parent=deform_pec_base_r, loc=<0 0 0>, rot=<0 0 0 1>}
 thigh_l {parent=pelvis, loc=<-1.0855 -0.2109 -8.7053>, rot=<0.0009 0.0355 0.0155 -0.9993>}
 calf_l {parent=thigh_l, loc=<-42.5284 -0.0019 -0.7297>, rot=<0.0006 -0.0161 -0.0391 -0.9991>}
 foot_l {parent=calf_l, loc=<-46.7401 -2.5014 -0.592>, rot=<0.0427 -0.0238 0.0353 -0.9982>}
 ball_l {parent=foot_l, loc=<-3.5658 10.8066 -0.1544>, rot=<0.0001 -0.0001 -0.7063 -0.7079>}
 extra_crystal_A_l {parent=foot_l, loc=<2.8946 -0.5403 -6.5354>, rot=<0.0093 -0.9719 -0.0284 0.2337>}
 extra_crystal_B_l {parent=foot_l, loc=<-2.2531 -0.0112 -7.3235>, rot=<-0.0221 0.8453 0.0376 -0.5324>}
 calf_twist_02_l {parent=calf_l, loc=<-19.1546 -1.7943 -0.2638>, rot=<0.0237 0.0095 -0.0035 -0.9997>}
 calf_twist_01_l {parent=calf_l, loc=<-35.8483 -1.9197 -0.5794>, rot=<0.0237 0.0095 -0.0035 -0.9997>}
 deform_knee_in_base_l {parent=calf_l, loc=<0.8031 -5.0231 0.2456>, rot=<0 0 0 -1>}
 deform_knee_in_l {parent=deform_knee_in_base_l, loc=<0 0 0>, rot=<0 0 0 1>}
 deform_knee_out_base_l {parent=calf_l, loc=<0 4 0>, rot=<0 0 0 -1>}
 deform_knee_out_l {parent=deform_knee_out_base_l, loc=<0 0 0>, rot=<0 0 0 1>}
 thigh_twist_01_l {parent=thigh_l, loc=<-27.952 0 0>, rot=<0.0013 0 0 -1>}
 deform_groin_base_l {parent=thigh_l, loc=<0.4801 5.6776 -0.2486>, rot=<0 0 0 -1>}
 deform_groin_l {parent=deform_groin_base_l, loc=<0 0 0>, rot=<0 0 0 1>}
 thigh_r {parent=pelvis, loc=<-1.0855 -0.2109 8.7053>, rot=<0.0355 -0.0009 0.9993 0.0155>}
 calf_r {parent=thigh_r, loc=<42.5284 0.0019 0.7297>, rot=<0.0006 -0.0161 -0.0391 -0.9991>}
 foot_r {parent=calf_r, loc=<46.7401 2.5014 0.592>, rot=<0.0427 -0.0238 0.0353 -0.9982>}
 ball_r {parent=foot_r, loc=<3.5658 -10.8066 0.1544>, rot=<0.0001 -0.0001 -0.7063 -0.7079>}
 extra_crystal_A_r {parent=foot_r, loc=<-3.1647 0.5705 6.0071>, rot=<0.9719 0.0093 -0.2337 -0.0284>}
 extra_crystal_B_r {parent=foot_r, loc=<2.2086 0.0132 7.3023>, rot=<0.8453 0.0221 -0.5324 -0.0376>}
 calf_twist_02_r {parent=calf_r, loc=<19.1546 1.7943 0.2638>, rot=<0.0237 0.0095 -0.0035 -0.9997>}
 calf_twist_01_r {parent=calf_r, loc=<35.8483 1.9197 0.5794>, rot=<0.0237 0.0095 -0.0035 -0.9997>}
 deform_knee_in_base_r {parent=calf_r, loc=<-0.8031 5.0231 -0.2457>, rot=<0 0 0 -1>}
 deform_knee_in_r {parent=deform_knee_in_base_r, loc=<0 0 0>, rot=<0 0 0 1>}
 deform_knee_out_base_r {parent=calf_r, loc=<0 -4 0>, rot=<0 0 0 -1>}
 deform_knee_out_r {parent=deform_knee_out_base_r, loc=<0 0 0>, rot=<0 0 0 1>}
 thigh_twist_01_r {parent=thigh_r, loc=<27.952 0 0>, rot=<0.0013 0 0 -1>}
 deform_groin_base_r {parent=thigh_r, loc=<-0.4801 -5.6776 0.2486>, rot=<0 0 0 -1>}
 deform_groin_r {parent=deform_groin_base_r, loc=<0 0 0>, rot=<0 0 0 1>}
 deform_glute_base_l {parent=pelvis, loc=<-7.3466 -9.0213 -9.1245>, rot=<0.0172 0.031 -0.4477 -0.8935>}
 deform_glute_l {parent=deform_glute_base_l, loc=<0 0 0>, rot=<0 0 0 1>}
 deform_glute_base_r {parent=pelvis, loc=<-7.3464 -9.0213 9.1245>, rot=<0.031 -0.0172 0.8935 -0.4477>}
 deform_glute_r {parent=deform_glute_base_r, loc=<0 0 0>, rot=<0 0 0 1>}
 dyn_skirt_fr_base_l {parent=pelvis, loc=<5.5765 7.6415 -4.2756>, rot=<-0.0212 0.0805 0.9922 0.0932>}
 dyn_skirt_fr_01_l {parent=dyn_skirt_fr_base_l, loc=<0.3608 0 0>, rot=<0 0 0 1>}
 dyn_skirt_fr_02_l {parent=dyn_skirt_fr_01_l, loc=<15.0829 0 0>, rot=<0 0 -0.0443 -0.999>}
 dyn_skirt_fr_03_l {parent=dyn_skirt_fr_02_l, loc=<15.0829 0 0>, rot=<0 0 -0.0494 -0.9988>}
 extra_skirt_fr_l {parent=dyn_skirt_fr_01_l, loc=<6.1521 0.9986 -5.2523>, rot=<-0.0874 -0.1098 -0.0309 -0.9896>}
 dyn_skirt_fr_side_base_l {parent=pelvis, loc=<5.4211 3.678 -10.8797>, rot=<-0.1153 0.3966 0.9035 0.1144>}
 dyn_skirt_fr_side_01_l {parent=dyn_skirt_fr_side_base_l, loc=<0.3608 0 0>, rot=<0 0 0 1>}
 dyn_skirt_fr_side_02_l {parent=dyn_skirt_fr_side_01_l, loc=<14.0106 0 0>, rot=<0 0 -0.0375 -0.9993>}
 dyn_skirt_side_base_l {parent=pelvis, loc=<5.7114 -1.7443 -12.6135>, rot=<0.1607 -0.7555 -0.62 -0.1377>}
 dyn_skirt_side_01_l {parent=dyn_skirt_side_base_l, loc=<0.3608 0 0>, rot=<0 0 0 1>}
 dyn_skirt_side_02_l {parent=dyn_skirt_side_01_l, loc=<11.6515 0 0>, rot=<0 0 -0.0662 -0.9978>}
 dyn_skirt_bk_side_base_l {parent=pelvis, loc=<6.0237 -7.9799 -8.714>, rot=<0.1438 -0.965 -0.2042 -0.0799>}
 dyn_skirt_bk_side_01_l {parent=dyn_skirt_bk_side_base_l, loc=<0.3608 0 0>, rot=<0 0 0 1>}
 dyn_skirt_bk_side_02_l {parent=dyn_skirt_bk_side_01_l, loc=<12.2251 0 0>, rot=<0 0 -0.0251 -0.9997>}
 dyn_collision_ribbon_l {parent=dyn_skirt_bk_side_01_l, loc=<12.0025 4.4264 -2.6773>, rot=<0 0 -0.0251 -0.9997>}
 dyn_skirt_bk_base_c {parent=pelvis, loc=<5.4907 -10.0786 0>, rot=<0.1359 -0.9907 0 0>}
 dyn_skirt_bk_01_c {parent=dyn_skirt_bk_base_c, loc=<0.3608 0 0>, rot=<0 0 0 1>}
 dyn_skirt_bk_02_c {parent=dyn_skirt_bk_01_c, loc=<14.5893 0 0>, rot=<0 0 -0.0366 -0.9993>}
 dyn_skirt_bk_03_c {parent=dyn_skirt_bk_02_c, loc=<14.5893 0 0>, rot=<0 0 -0.0554 -0.9985>}
 dyn_skirt_bk_side_base_r {parent=pelvis, loc=<6.0239 -7.98 8.714>, rot=<0.1438 -0.965 0.2042 0.0799>}
 dyn_skirt_bk_side_01_r {parent=dyn_skirt_bk_side_base_r, loc=<0.3609 0 0>, rot=<0 0 0 1>}
 dyn_skirt_bk_side_02_r {parent=dyn_skirt_bk_side_01_r, loc=<12.2252 0 0>, rot=<0 0 -0.0251 -0.9997>}
 dyn_collision_ribbon_r {parent=dyn_skirt_bk_side_01_r, loc=<12.0026 4.4264 2.6773>, rot=<0 0 -0.0251 -0.9997>}
 dyn_skirt_side_base_r {parent=pelvis, loc=<5.711 -1.7443 12.6135>, rot=<0.1607 -0.7555 0.62 0.1377>}
 dyn_skirt_side_01_r {parent=dyn_skirt_side_base_r, loc=<0.3605 -0.0001 0>, rot=<0 0 0 1>}
 dyn_skirt_side_02_r {parent=dyn_skirt_side_01_r, loc=<11.6515 0 0>, rot=<0 0 -0.0662 -0.9978>}
 dyn_skirt_fr_side_base_r {parent=pelvis, loc=<5.4211 3.678 10.8797>, rot=<0.1153 -0.3966 0.9035 0.1144>}
 dyn_skirt_fr_side_01_r {parent=dyn_skirt_fr_side_base_r, loc=<0.3608 0 0>, rot=<0 0 0 1>}
 dyn_skirt_fr_side_02_r {parent=dyn_skirt_fr_side_01_r, loc=<14.0106 0 0>, rot=<0 0 -0.0375 -0.9993>}
 dyn_skirt_fr_base_r {parent=pelvis, loc=<5.5762 7.6415 4.2756>, rot=<0.0212 -0.0805 0.9922 0.0932>}
 dyn_skirt_fr_01_r {parent=dyn_skirt_fr_base_r, loc=<0.3605 -0.0001 0>, rot=<0 0 0 1>}
 dyn_skirt_fr_02_r {parent=dyn_skirt_fr_01_r, loc=<15.0828 0 0>, rot=<0 0 -0.0443 -0.999>}
 dyn_skirt_fr_03_r {parent=dyn_skirt_fr_02_r, loc=<15.0829 0 0>, rot=<0 0 -0.0494 -0.9988>}
 extra_skirt_fr_r {parent=dyn_skirt_fr_01_r, loc=<6.1521 0.9986 5.2523>, rot=<0.0874 0.1098 -0.0309 -0.9896>}
 base_skirt_chain {parent=pelvis, loc=<10.5843 7.185 0>, rot=<0.0962 0.7005 0.7005 0.0962>}
 dyn_skirt_chain {parent=base_skirt_chain, loc=<0.3433 0 0>, rot=<0 0 0 1>}
 dyn_ribbon_base_r {parent=pelvis, loc=<12.7637 -7.3783 4.9434>, rot=<0.1785 -0.9654 0.1676 0.0899>}
 dyn_ribbon_01_r {parent=dyn_ribbon_base_r, loc=<0.3611 0.0001 0>, rot=<0 0 0 1>}
 dyn_ribbon_02_r {parent=dyn_ribbon_01_r, loc=<17.6812 -0.0001 0>, rot=<0 0 -0.0608 -0.9982>}
 dyn_ribbon_03_r {parent=dyn_ribbon_02_r, loc=<17.6814 0 -0.0001>, rot=<0 0 -0.0473 -0.9989>}
 dyn_ribbon_base_l {parent=pelvis, loc=<12.7638 -7.3783 -4.9434>, rot=<0.1785 -0.9654 -0.1676 -0.0899>}
 dyn_ribbon_01_l {parent=dyn_ribbon_base_l, loc=<0.3608 0 0>, rot=<0 0 0 1>}
 dyn_ribbon_02_l {parent=dyn_ribbon_01_l, loc=<17.6814 0 0>, rot=<0 0 -0.0608 -0.9982>}
 dyn_ribbon_03_l {parent=dyn_ribbon_02_l, loc=<17.6814 0 0>, rot=<0 0 -0.0473 -0.9989>}
 extra_belt_l {parent=pelvis, loc=<11.1233 7.1268 -5.9492>, rot=<-0.3236 -0.5005 -0.3665 -0.7144>}
 extra_belt_r {parent=pelvis, loc=<11.1233 7.1268 5.9492>, rot=<0.7144 0.3665 0.5005 0.3236>}
 dyn_ribbon_collision {parent=pelvis, loc=<-8.033 0.8184 0>, rot=<0 0 0 1>}
]
variables [
 AlphaSPR_thigh_L {type=double, default=0}
 AlphaSPR_thigh_R {type=double, default=0}
 BoneSpaceOffset {type=FTransform, default=(Rotation=(X=0,Y=0,Z=0,W=1),Translation=(X=0,Y=0,Z=0),Scale3D=(X=1,Y=1,Z=1))}
 Cape {type=bool, default=False}
 FE_idle {type=bool, default=False}
 Skydive {type=bool, default=False}
 Velocity {type=FVector, default=(X=0,Y=0,Z=0)}
]
