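# Skydive physics asset for the female LexaBlitz cosmetic body, covering the dynamic skirt and ribbon bones.
# Capsules on pelvis, spine, legs and a three-capsule ribbon collision bone provide body volumes.
# Limited-swing skirt chains hang from the pelvis, while left and right ribbon chains attach to spine_02 through free-rotation joints.
# Free-rotation cross constraints tie the front panels left to right and the right front-side panel to the right side panel.
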PhysicsAsset /BRCosmetics/Characters/Player/Female/Medium/Bodies/F_MED_LexaBlitz/Meshes/F_MED_LexaBlitz_Physics_Skydive
bodies {
 calf_l [{shapeType=Capsule, length=30.4296, radius=6, center=<-19.3752 0.0001 0>, rotation=<rot -90 -0.0003 0>}]
 calf_r [{shapeType=Capsule, length=30.4296, radius=6, center=<19.3752 -0.0001 0>, rotation=<rot -90 179.9997 0>}]
 dyn_ribbon_01_l [{shapeType=Capsule, length=7, radius=5.5, shapeName=dyn_ribbon_01_l_capsule, center=<9 0 0>, rotation=<rot 90 0 0>}]
 dyn_ribbon_01_r [{shapeType=Capsule, length=7, radius=5.5, shapeName=dyn_ribbon_01_r_capsule, center=<9 0 0>, rotation=<rot 90 0 0>}]
 dyn_ribbon_02_l [{shapeType=Capsule, length=7, radius=7, shapeName=dyn_ribbon_02_l_capsule, center=<9 0 0>, rotation=<rot 90 0 0>}]
 dyn_ribbon_02_r [{shapeType=Capsule, length=7, radius=7, shapeName=dyn_ribbon_02_r_capsule, center=<9 0 0>, rotation=<rot 90 0 0>}]
 dyn_ribbon_03_l [{shapeType=Capsule, length=20, radius=2, shapeName=dyn_ribbon_03_l_capsule, center=<14 0 0>, rotation=<rot 90 0 0>}]
 dyn_ribbon_03_r [{shapeType=Capsule, length=20, radius=2, shapeName=dyn_ribbon_03_r_capsule, center=<14 0 0>, rotation=<rot 90 0 0>}]
 dyn_ribbon_base_l [{shapeType=Capsule, length=1, radius=1, shapeName=dyn_ribbon_base_l_capsule}]
 dyn_ribbon_base_r [{shapeType=Capsule, length=1, radius=1, shapeName=dyn_ribbon_base_r_capsule}]
 dyn_ribbon_collision [
  {shapeType=Capsule, length=19.5, radius=7, shapeName=dyn_ribbon_collision_capsule, center=<-0.4304 -1.3546 0>, rotation=<rot 0 -90 0>}
  {shapeType=Capsule, length=19.5, radius=12, shapeName=dyn_ribbon_collision_capsule, center=<-10.7502 -1.8447 0>, rotation=<rot 0 -90 0>}
  {shapeType=Capsule, length=25, radius=18, shapeName=dyn_ribbon_collision_capsule, center=<-23.8103 -4.6454 0>, rotation=<rot 0 -90 0>}
 ]
 dyn_skirt_bk_01_c [{shapeType=Capsule, length=6, radius=3, shapeName=dyn_skirt_bk_01_c_capsule, center=<7 0 0>}]
 dyn_skirt_bk_02_c [{shapeType=Capsule, length=10, radius=10, shapeName=dyn_skirt_bk_02_c_capsule, center=<7 0 0>}]
 dyn_skirt_bk_03_c [{shapeType=Capsule, length=30, radius=6, shapeName=dyn_skirt_bk_03_c_capsule, center=<7 5 0>}]
 dyn_skirt_bk_base_c [{shapeType=Capsule, length=4, radius=2, shapeName=dyn_skirt_bk_base_c_capsule}]
 dyn_skirt_bk_side_01_l [{shapeType=Capsule, length=10, radius=3, shapeName=dyn_skirt_bk_side_01_l_capsule, center=<7 0 0>}]
 dyn_skirt_bk_side_01_r [{shapeType=Capsule, length=10, radius=3, shapeName=dyn_skirt_bk_side_01_r_capsule, center=<7 0 0>}]
 dyn_skirt_bk_side_02_l [{shapeType=Capsule, length=10, radius=10, shapeName=dyn_skirt_bk_side_02_l_capsule, center=<7 0 0>}]
 dyn_skirt_bk_side_02_r [{shapeType=Capsule, length=10, radius=10, shapeName=dyn_skirt_bk_side_02_r_capsule, center=<7 0 0>}]
 dyn_skirt_bk_side_base_l [{shapeType=Capsule, length=4, radius=2, shapeName=dyn_skirt_bk_side_base_l_capsule}]
 dyn_skirt_bk_side_base_r [{shapeType=Capsule, length=4, radius=2, shapeName=dyn_skirt_bk_side_base_r_capsule}]
 dyn_skirt_fr_01_l [{shapeType=Capsule, length=5, radius=3, shapeName=dyn_skirt_fr_01_l_capsule, center=<7 0 0>}]
 dyn_skirt_fr_01_r [{shapeType=Capsule, length=5, radius=3, shapeName=dyn_skirt_fr_01_l_capsule, center=<7 0 0>}]
 dyn_skirt_fr_02_l [{shapeType=Capsule, length=5, radius=4, shapeName=dyn_skirt_fr_02_l_capsule, center=<7 0 0>}]
 dyn_skirt_fr_02_r [{shapeType=Capsule, length=5, radius=4, shapeName=dyn_skirt_fr_02_r_capsule, center=<7 0 0>}]
 dyn_skirt_fr_03_l [{shapeType=Capsule, length=10, radius=3, shapeName=dyn_skirt_fr_03_l_capsule, center=<7 0 0>}]
 dyn_skirt_fr_03_r [{shapeType=Capsule, length=10, radius=3, shapeName=dyn_skirt_fr_03_r_capsule, center=<7 0 0>}]
 dyn_skirt_fr_base_l [{shapeType=Capsule, length=4, radius=2, shapeName=dyn_skirt_fr_base_l_capsule}]
 dyn_skirt_fr_base_r [{shapeType=Capsule, length=4, radius=2, shapeName=dyn_skirt_fr_base_r_capsule}]
 dyn_skirt_fr_side_01_l [{shapeType=Capsule, length=10, radius=2, shapeName=dyn_skirt_fr_side_01_l_capsule, center=<7 0 0>}]
 dyn_skirt_fr_side_01_r [
  {shapeType=Capsule, length=10, radius=2, shapeName=dyn_skirt_fr_side_01_r_capsule, center=<7 0 0>}
  {
   shapeType=Capsule
   length=10
   radius=3
   shapeName=dyn_skirt_fr_side_01_r_capsule
   center=<11.7417 2.636 -7.8403>
   rotation=<rot -0.0001 -0.0001 -54.9999>
  }
 ]
 dyn_skirt_fr_side_02_l [{shapeType=Capsule, length=10, radius=2, shapeName=dyn_skirt_fr_side_02_l_capsule, center=<7 0 0>}]
 dyn_skirt_fr_side_02_r [{shapeType=Capsule, length=10, radius=2, shapeName=dyn_skirt_fr_side_02_r_capsule, center=<7 0 0>}]
 dyn_skirt_fr_side_base_l [{shapeType=Capsule, length=4, radius=2, shapeName=dyn_skirt_fr_side_base_l_capsule}]
 dyn_skirt_fr_side_base_r [{shapeType=Capsule, length=4, radius=2, shapeName=dyn_skirt_fr_side_base_r_capsule}]
 dyn_skirt_side_01_l [{shapeType=Capsule, length=10, radius=3, shapeName=dyn_skirt_side_01_l_capsule, center=<7 0 0>}]
 dyn_skirt_side_01_r [{shapeType=Capsule, length=10, radius=3, shapeName=dyn_skirt_side_01_r_capsule, center=<7 0 0>}]
 dyn_skirt_side_02_l [{shapeType=Capsule, length=10, radius=2, shapeName=dyn_skirt_side_02_l_capsule, center=<7 0 0>}]
 dyn_skirt_side_02_r [{shapeType=Capsule, length=10, radius=2, shapeName=dyn_skirt_side_02_r_capsule, center=<7 0 0>}]
 dyn_skirt_side_base_l [{shapeType=Capsule, length=4, radius=2, shapeName=dyn_skirt_side_base_l_capsule}]
 dyn_skirt_side_base_r [{shapeType=Capsule, length=4, radius=2, shapeName=dyn_skirt_side_base_r_capsule}]
 pelvis [{shapeType=Capsule, length=12, radius=6, center=<-1.5193 -1.3644 0.1863>}]
 spine_01 [{shapeType=Capsule, length=5, radius=5, shapeName=spine_01_capsule}]
 spine_02 [{shapeType=Capsule, length=5, radius=5, shapeName=spine_02_capsule}]
 thigh_l [{shapeType=Capsule, length=25.2408, radius=7, center=<-18.8128 0 0>, rotation=<rot 90 -0.0005 0>}]
 thigh_r [{shapeType=Capsule, length=25.2408, radius=7, center=<18.8128 0 0>, rotation=<rot 90 179.9995 0>}]
}
constraints [
 {bone1Name=thigh_l, bone2Name=pelvis, swing1LimitDegrees=45, swing2LimitDegrees=45, twistLimitDegrees=45}
 {bone1Name=calf_l, bone2Name=thigh_l, swing1LimitDegrees=45, swing2LimitDegrees=45, twistLimitDegrees=45}
 {bone1Name=thigh_r, bone2Name=pelvis, swing1LimitDegrees=45, swing2LimitDegrees=45, twistLimitDegrees=45}
 {bone1Name=calf_r, bone2Name=thigh_r, swing1LimitDegrees=45, swing2LimitDegrees=45, twistLimitDegrees=45}
 {bone1Name=dyn_skirt_bk_base_c, bone2Name=pelvis, swing1LimitDegrees=30, swing2LimitDegrees=15, twistLimitDegrees=5}
 {bone1Name=dyn_skirt_bk_01_c, bone2Name=dyn_skirt_bk_base_c, swing1LimitDegrees=30, swing2LimitDegrees=15, twistLimitDegrees=5}
 {bone1Name=dyn_skirt_bk_02_c, bone2Name=dyn_skirt_bk_01_c, swing1LimitDegrees=15, swing2LimitDegrees=15, twistLimitDegrees=5}
 {bone1Name=dyn_skirt_bk_03_c, bone2Name=dyn_skirt_bk_02_c, swing1LimitDegrees=15, swing2LimitDegrees=15, twistLimitDegrees=5}
 {bone1Name=dyn_skirt_bk_side_base_l, bone2Name=pelvis, swing1LimitDegrees=30, swing2LimitDegrees=15, twistLimitDegrees=5}
 {bone1Name=dyn_skirt_bk_side_01_l, bone2Name=dyn_skirt_bk_side_base_l, swing1LimitDegrees=30, swing2LimitDegrees=15, twistLimitDegrees=5}
 {bone1Name=dyn_skirt_bk_side_02_l, bone2Name=dyn_skirt_bk_side_01_l, swing1LimitDegrees=30, swing2LimitDegrees=15, twistLimitDegrees=5}
 {bone1Name=dyn_skirt_bk_side_base_r, bone2Name=pelvis, swing1LimitDegrees=30, swing2LimitDegrees=15, twistLimitDegrees=5}
 {bone1Name=dyn_skirt_bk_side_01_r, bone2Name=dyn_skirt_bk_side_base_r, swing1LimitDegrees=30, swing2LimitDegrees=15, twistLimitDegrees=5}
 {bone1Name=dyn_skirt_bk_side_02_r, bone2Name=dyn_skirt_bk_side_01_r, swing1LimitDegrees=30, swing2LimitDegrees=15, twistLimitDegrees=5}
 {bone1Name=dyn_skirt_fr_base_l, bone2Name=pelvis, swing1LimitDegrees=30, swing2LimitDegrees=15, twistLimitDegrees=5}
 {bone1Name=dyn_skirt_fr_01_l, bone2Name=dyn_skirt_fr_base_l, swing1LimitDegrees=60, swing2LimitDegrees=30, twistLimitDegrees=5}
 {bone1Name=dyn_skirt_fr_02_l, bone2Name=dyn_skirt_fr_01_l, swing1LimitDegrees=30, swing2LimitDegrees=15, twistLimitDegrees=5}
 {bone1Name=dyn_skirt_fr_03_l, bone2Name=dyn_skirt_fr_02_l, swing1LimitDegrees=30, swing2LimitDegrees=15, twistLimitDegrees=5}
 {bone1Name=dyn_skirt_fr_base_r, bone2Name=pelvis, swing1LimitDegrees=30, swing2LimitDegrees=15, twistLimitDegrees=5}
 {bone1Name=dyn_skirt_fr_01_r, bone2Name=dyn_skirt_fr_base_r, swing1LimitDegrees=60, swing2LimitDegrees=30, twistLimitDegrees=5}
 {bone1Name=dyn_skirt_fr_02_r, bone2Name=dyn_skirt_fr_01_r, swing1LimitDegrees=15, swing2LimitDegrees=10, twistLimitDegrees=5}
 {bone1Name=dyn_skirt_fr_03_r, bone2Name=dyn_skirt_fr_02_r, swing1LimitDegrees=15, swing2LimitDegrees=10, twistLimitDegrees=5}
 {bone1Name=dyn_skirt_fr_side_base_l, bone2Name=pelvis, swing1LimitDegrees=30, swing2LimitDegrees=15, twistLimitDegrees=5}
 {bone1Name=dyn_skirt_fr_side_01_l, bone2Name=dyn_skirt_fr_side_base_l, swing1LimitDegrees=30, swing2LimitDegrees=15, twistLimitDegrees=5}
 {bone1Name=dyn_skirt_fr_side_02_l, bone2Name=dyn_skirt_fr_side_01_l, swing1LimitDegrees=30, swing2LimitDegrees=15, twistLimitDegrees=5}
 {bone1Name=dyn_skirt_fr_side_base_r, bone2Name=pelvis, swing1LimitDegrees=30, swing2LimitDegrees=15, twistLimitDegrees=5}
 {bone1Name=dyn_skirt_fr_side_01_r, bone2Name=dyn_skirt_fr_side_base_r, swing1LimitDegrees=15, swing2LimitDegrees=10, twistLimitDegrees=5}
 {bone1Name=dyn_skirt_fr_side_02_r, bone2Name=dyn_skirt_fr_side_01_r, swing1LimitDegrees=30, swing2LimitDegrees=15, twistLimitDegrees=5}
 {bone1Name=dyn_skirt_side_base_l, bone2Name=pelvis, swing1LimitDegrees=30, swing2LimitDegrees=15, twistLimitDegrees=5}
 {bone1Name=dyn_skirt_side_01_l, bone2Name=dyn_skirt_side_base_l, swing1LimitDegrees=30, swing2LimitDegrees=15, twistLimitDegrees=5}
 {bone1Name=dyn_skirt_side_02_l, bone2Name=dyn_skirt_side_01_l, swing1LimitDegrees=30, swing2LimitDegrees=15, twistLimitDegrees=5}
 {bone1Name=dyn_skirt_side_base_r, bone2Name=pelvis, swing1LimitDegrees=30, swing2LimitDegrees=15, twistLimitDegrees=5}
 {bone1Name=dyn_skirt_side_01_r, bone2Name=dyn_skirt_side_base_r, swing1LimitDegrees=30, swing2LimitDegrees=15, twistLimitDegrees=5}
 {bone1Name=dyn_skirt_side_02_r, bone2Name=dyn_skirt_side_01_r, swing1LimitDegrees=30, swing2LimitDegrees=30, twistLimitDegrees=5}
 {
  bone1Name=dyn_skirt_fr_03_l
  bone2Name=dyn_skirt_fr_03_r
  swing1LimitDegrees=45
  swing1Motion=Free
  swing2LimitDegrees=45
  swing2Motion=Free
  twistLimitDegrees=45
  twistMotion=Free
 }
 {
  bone1Name=spine_01
  bone2Name=pelvis
  swing1LimitDegrees=45
  swing1Motion=Free
  swing2LimitDegrees=45
  swing2Motion=Free
  twistLimitDegrees=45
  twistMotion=Free
 }
 {
  bone1Name=spine_02
  bone2Name=spine_01
  swing1LimitDegrees=45
  swing1Motion=Free
  swing2LimitDegrees=45
  swing2Motion=Free
  twistLimitDegrees=45
  twistMotion=Free
 }
 {
  bone1Name=dyn_ribbon_base_l
  bone2Name=spine_02
  swing1LimitDegrees=45
  swing1Motion=Free
  swing2LimitDegrees=45
  swing2Motion=Free
  twistLimitDegrees=45
  twistMotion=Free
 }
 {bone1Name=dyn_ribbon_01_l, bone2Name=dyn_ribbon_base_l, swing1LimitDegrees=15, swing2LimitDegrees=5, twistLimitDegrees=5}
 {bone1Name=dyn_ribbon_02_l, bone2Name=dyn_ribbon_01_l, swing1LimitDegrees=30, swing2LimitDegrees=15, twistLimitDegrees=5}
 {bone1Name=dyn_ribbon_03_l, bone2Name=dyn_ribbon_02_l, swing1LimitDegrees=30, swing2LimitDegrees=15, twistLimitDegrees=5}
 {
  bone1Name=dyn_ribbon_base_r
  bone2Name=spine_02
  swing1LimitDegrees=45
  swing1Motion=Free
  swing2LimitDegrees=45
  swing2Motion=Free
  twistLimitDegrees=45
  twistMotion=Free
 }
 {bone1Name=dyn_ribbon_01_r, bone2Name=dyn_ribbon_base_r, swing1LimitDegrees=15, swing2LimitDegrees=5, twistLimitDegrees=5}
 {bone1Name=dyn_ribbon_02_r, bone2Name=dyn_ribbon_01_r, swing1LimitDegrees=30, swing2LimitDegrees=15, twistLimitDegrees=5}
 {bone1Name=dyn_ribbon_03_r, bone2Name=dyn_ribbon_02_r, swing1LimitDegrees=30, swing2LimitDegrees=15, twistLimitDegrees=5}
 {
  bone1Name=dyn_skirt_fr_02_l
  bone2Name=dyn_skirt_fr_02_r
  swing1LimitDegrees=45
  swing1Motion=Free
  swing2LimitDegrees=45
  swing2Motion=Free
  twistLimitDegrees=45
  twistMotion=Free
 }
 {
  bone1Name=dyn_skirt_fr_side_02_r
  bone2Name=dyn_skirt_side_02_r
  swing1LimitDegrees=45
  swing1Motion=Free
  swing2LimitDegrees=45
  swing2Motion=Free
  twistLimitDegrees=45
  twistMotion=Free
 }
]
