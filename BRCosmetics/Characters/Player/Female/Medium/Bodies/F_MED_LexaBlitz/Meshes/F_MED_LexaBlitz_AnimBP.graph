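# Character-part AnimBP for the F_MED_LexaBlitz body, adding secondary motion to the pose copied from the attached parent mesh.
# A pre-dynamics Control Rig and crystal bone transforms feed a cached pose into default, crouch and skydive RigidBody simulations.
# Their blend runs through a post-dynamics Control Rig, with a bypass blend and jawline offset before output.
# The EventGraph sets state flags, crystal wave vectors, parameter-driven gravity override and backbling checks.
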AnimBlueprint /BRCosmetics/Characters/Player/Female/Medium/Bodies/F_MED_LexaBlitz/Meshes/F_MED_LexaBlitz_AnimBP
target_skeleton=&/F_MED_LexaBlitz_Skeleton
graphs [
 AnimGraph {nodes [
   "Transform (Modify) Bone - Bone: extra_neck"!(ComponentPose, bAlphaBoolEnabled:bool=True, Translation="0,43.938,0", Rotation="0,0,0", Scale="1,1,1") -> (Pose) {
    alphaInputType=Bool
    rotationMode=BMM_Additive
    rotationSpace=BCS_BoneSpace
    scaleMode=BMM_Additive
    scaleSpace=BCS_BoneSpace
    translationMode=BMM_Additive
    translationSpace=BCS_BoneSpace
    boneToModify {boneName=extra_neck}
    translation {y=43.938}
   }
   RigidBody!(
    ComponentPose
    bAlphaBoolEnabled:bool
    OverrideWorldGravity:Vector="0,0,0"
    ComponentLinearAccScale:Vector="0,0,0"
    ComponentLinearVelScale:Vector="0,0,0"
    bTransferBoneVelocities:bool=False
   ) -> (Pose) {
    alphaInputType=Bool
    bClampLinearTranslationLimitToRefPose=true
    bFreezeIncomingPoseOnStart=true
    bOverrideWorldGravity=true
    bTransferBoneVelocities=true
    overridePhysicsAsset=&/F_MED_LexaBlitz_Physics
    simulationSpace=BaseBoneSpace
    alphaBoolBlend {blendOption=Linear}
    baseBoneRef {boneName=pelvis}
    componentAppliedLinearAccClamp=<100 100 100>
    componentLinearAccScale=<0.6 0.2 0.2>
    componentLinearVelScale=<0.6 0.2 0.2>
    simSpaceSettings {
     dampingAlpha=0
     maxAngularAcceleration=50
     maxAngularVelocity=8
     maxLinearAcceleration=2000
     maxLinearVelocity=500
     worldAlpha=1
     externalLinearDragV=<1 1 1>
    }
   }
   n1 = RigidBody(
    ComponentPose=Local To Component(LocalPose=Use cached pose 'PreDyn'())
    OverrideWorldGravity="0,0,0"
    ComponentLinearAccScale="0,0,0"
    ComponentLinearVelScale="0,0,0"
    bTransferBoneVelocities=False
   ) {
    alphaInputType=Bool
    bClampLinearTranslationLimitToRefPose=true
    bFreezeIncomingPoseOnStart=true
    bOverrideWorldGravity=true
    bTransferBoneVelocities=true
    overridePhysicsAsset=&/F_MED_LexaBlitz_Physics_Crouch
    simulationSpace=BaseBoneSpace
    alphaBoolBlend {blendOption=Linear}
    baseBoneRef {boneName=pelvis}
    componentAppliedLinearAccClamp=<100 100 100>
    componentLinearAccScale=<0.6 0.2 0.2>
    componentLinearVelScale=<0.6 0.2 0.2>
    simSpaceSettings {
     dampingAlpha=0
     maxAngularAcceleration=50
     maxAngularVelocity=8
     maxLinearAcceleration=2000
     maxLinearVelocity=500
     worldAlpha=1
     externalLinearDragV=<1 1 1>
    }
   }
   n2 = Component To Local(ComponentPose=@n1)
   n3 = RigidBody(
    ComponentPose=Local To Component(LocalPose=Use cached pose 'PreDyn'())
    OverrideWorldGravity="0,0,0"
    ComponentLinearAccScale="0,0,0"
    ComponentLinearVelScale="0,0,0"
    bTransferBoneVelocities=False
   ) {
    alphaInputType=Bool
    bClampLinearTranslationLimitToRefPose=true
    bFreezeIncomingPoseOnStart=true
    bOverrideWorldGravity=true
    bTransferBoneVelocities=true
    overridePhysicsAsset=&F_MED_LexaBlitz_Physics
    simulationSpace=BaseBoneSpace
    alphaBoolBlend {blendOption=Linear}
    baseBoneRef {boneName=pelvis}
    componentAppliedLinearAccClamp=<100 100 100>
    componentLinearAccScale=<0.6 0.2 0.2>
    componentLinearVelScale=<0.6 0.2 0.2>
    simSpaceSettings {
     dampingAlpha=0
     maxAngularAcceleration=5000
     maxAngularVelocity=8
     maxLinearAcceleration=1000
     maxLinearVelocity=500
     worldAlpha=1
     externalLinearDragV=<1 1 1>
    }
   }
   n4 = Component To Local(ComponentPose=@n3)
   n5 = RigidBody(
    ComponentPose=Local To Component(LocalPose=Use cached pose 'PreDyn'())
    OverrideWorldGravity="0,0,0"
    ComponentLinearAccScale="0,0,0"
    ComponentLinearVelScale="0,0,0"
    bTransferBoneVelocities=False
   ) {
    alphaInputType=Bool
    bClampLinearTranslationLimitToRefPose=true
    bFreezeIncomingPoseOnStart=true
    bOverrideWorldGravity=true
    bTransferBoneVelocities=true
    overridePhysicsAsset=&/F_MED_LexaBlitz_Physics_Skydive
    simulationSpace=BaseBoneSpace
    alphaBoolBlend {blendOption=Linear}
    baseBoneRef {boneName=pelvis}
    componentAppliedLinearAccClamp=<100 100 100>
    componentLinearAccScale=<0.6 0.2 0.2>
    componentLinearVelScale=<0.6 0.2 0.2>
    simSpaceSettings {
     dampingAlpha=0
     maxAngularAcceleration=5000
     maxAngularVelocity=8
     maxLinearAcceleration=1000
     maxLinearVelocity=500
     worldAlpha=1
     externalLinearDragV=<1 1 1>
    }
   }
   n6 = Control Rig!(
    Source=Blend Poses by bool(BlendPose_0=Component To Local(ComponentPose=@n5), BlendPose_1=Blend Poses by bool(BlendPose_0=@n2, BlendPose_1=@n4))
    Alpha
    Skydive:bool=false
    Velocity:Vector="0, 0, 0"
    Cape:bool=false
    FE_idle:bool=false
   ) -> (Pose) {controlRigAssetReference {blueprintRigClass=&/F_MED_LexaBlitz_CtrlRig_PostDyn.F_MED_LexaBlitz_CtrlRig_PostDyn_C}}
   n7 = Copy Pose From Mesh() {bUseAttachedParent=true}
   n8 = "Transform (Modify) Bone - Bone: FACIAL_C_Jawline"!(
    ComponentPose=Local To Component(LocalPose=Blend Poses by bool(BlendPose_0=@n7, BlendPose_1=@n6))
    bAlphaBoolEnabled:bool=True
    Translation="-0.5,-3,0"
    Rotation="0,0,0"
    Scale="1,1,1"
   ) -> (Pose) {
    alphaInputType=Bool
    rotationMode=BMM_Additive
    rotationSpace=BCS_BoneSpace
    scaleMode=BMM_Additive
    scaleSpace=BCS_BoneSpace
    translationMode=BMM_Additive
    translationSpace=BCS_BoneSpace
    boneToModify {boneName=FACIAL_C_Jawline}
    translation=<-0.5 -3>
   }
   Output Pose(Result=Component To Local(ComponentPose=@n8)) {layerGroup=None}
   n9 = Control Rig!(
    Source=Copy Pose From Mesh() {bUseAttachedParent=true}
    Alpha
    Move:bool=false
    PawnVelocity:Vector="0, 0, 0"
    Crouch:bool=false
    Sliding:bool=false
    Cape=false
   ) -> (Pose) {controlRigAssetReference {blueprintRigClass=&/F_MED_LexaBlitz_CtrlRig_PreDyn.F_MED_LexaBlitz_CtrlRig_PreDyn_C}}
   n10 = "Transform (Modify) Bone - Bone: extra_crystal_A_l"(ComponentPose=Local To Component(LocalPose=@n9), Alpha=1, Translation="0,0,0", Rotation="0,0,0", Scale="1,1,1") {
    rotationSpace=BCS_BoneSpace
    scaleSpace=BCS_BoneSpace
    translationMode=BMM_Additive
    translationSpace=BCS_BoneSpace
    boneToModify {boneName=extra_crystal_A_l}
   }
   n11 = "Transform (Modify) Bone - Bone: extra_crystal_B_l"(ComponentPose=@n10, Alpha=1, Translation="0,0,0", Rotation="0,0,0", Scale="1,1,1") {
    rotationSpace=BCS_BoneSpace
    scaleSpace=BCS_BoneSpace
    translationMode=BMM_Additive
    translationSpace=BCS_BoneSpace
    boneToModify {boneName=extra_crystal_B_l}
   }
   n12 = "Transform (Modify) Bone - Bone: extra_crystal_A_r"(ComponentPose=@n11, Alpha=1, Translation="0,0,0", Rotation="0,0,0", Scale="1,1,1") {
    rotationSpace=BCS_BoneSpace
    scaleSpace=BCS_BoneSpace
    translationMode=BMM_Additive
    translationSpace=BCS_BoneSpace
    boneToModify {boneName=extra_crystal_A_r}
   }
   n13 = "Transform (Modify) Bone - Bone: extra_crystal_B_r"(ComponentPose=@n12, Alpha=1, Translation="0,0,0", Rotation="0,0,0", Scale="1,1,1") {
    rotationSpace=BCS_BoneSpace
    scaleSpace=BCS_BoneSpace
    translationMode=BMM_Additive
    translationSpace=BCS_BoneSpace
    boneToModify {boneName=extra_crystal_B_r}
   }
   Save cached pose 'PreDyn'(Pose=Component To Local(ComponentPose=@n13)) {cacheName=PreDyn}
  ]}
 EventGraph {nodes [
   Event BlueprintUpdateAnimation()
   > n1 = Sequence!(execute) -> (then_0, then_1, then_2:Exec, then_3:Exec, then_4:Exec)
   n2 = Branch(execute=@n1.then_1, Condition=Get bIsCrouching())
   > Set DynCrouch(DynCrouch:bool=true) -> (Output_Get:bool)
   > Set DynSkydive(DynSkydive:bool=false) -> (Output_Get:bool)
   > Set Dyn(Dyn:bool=false) -> (Output_Get:bool)
   n3 = Branch(execute=@n2.else, Condition=Get bIsSkyDiving())
   > Set DynCrouch(DynCrouch=false)
   > Set DynSkydive(DynSkydive=true)
   > Set Dyn(Dyn=false)
   Set DynCrouch(execute=@n3.else, DynCrouch=false)
   > Set DynSkydive(DynSkydive=false)
   > Set Dyn(Dyn=true)
   Set CrystalAnim_A(
    execute=@n1.then_2
    CrystalAnim_A:Vector=MakeWaveVector!(Amplitudes:Rotator="1,0.5,1.75", Frequencies:Rotator="100,50,80", Offsets:Rotator="0,0,0", bUseNoise:bool=false, Min:Vector, Max:Vector) -> (ReturnValue:Vector)
   ) -> (Output_Get:Vector)
   > Set CrystalAnim_B(CrystalAnim_B:Vector=MakeWaveVector(Amplitudes="1,0.5,1.75", Frequencies="100,50,80", Offsets="-45,-45,90", bUseNoise=false)) -> (Output_Get:Vector)
   > Set Pawn Velocity Z(Pawn Velocity Z:float=Get PawnVelocityZ()) -> (Output_Get:float)
   n4 = Get!() -> (PawnVelocity_X:double, PawnVelocity_Y:double, PawnVelocity_Z:double)
   n5 = float + float!(A=@n4, B=@n4.PawnVelocity_Y, C:double=@n4.PawnVelocity_Z) -> (ReturnValue)
   > Set Move(Move:bool="float > float"(A="Absolute (Float)"(A=@n5))) -> (Output_Get:bool)
   > Set Pawn Velocity(Pawn Velocity:Vector=Get PawnVelocity()) -> (Output_Get:Vector)
   > Set LOD1-3(LOD1-3:bool="integer >= integer"(A=Get CurrentLOD(), B=1)) -> (Output_Get:bool)
   n6 = Get bIsPlayingEmote()
   Set FE_Idle(execute=@n1.then_3, FE_Idle:bool=AND Boolean(A=Get bIsInFrontEnd(), B=NOT Boolean(A=@n6))) -> (Output_Get:bool)
   n7 = "Gravity Override State Function (Parameter Asset) (Nativized)"(
    execute=@n1.then_0
    CharacterPartAnimInstance=Self-Reference()
    GameTimeInSeconds=GetGameTimeInSeconds()
    DebugJointName=spine_05
    Parameters=&/F_MED_LexaBlitz_Parameters
   )
   > Set GravityOverride(GravityOverride=@n7)
   Event OnCharacterCustomizationLoaded()
   > Set BindPose(BindPose:bool=false) -> (Output_Get:bool)
   n8 = CheckForBackbling()
   > Set isWearingPackBool(isWearingPackBool:bool=OR Boolean!(A=@n8, B=@n8.bIsBackblingACape, C:bool=@n8.bIsBackblingLong) -> (ReturnValue)) -> (Output_Get:bool)
   > Set is_WearingCape(is_WearingCape:bool=@n8.bIsBackblingLong) -> (Output_Get:bool)
   AND Boolean(A=@n8, B=NOT Boolean(A=@n8.bIsBackblingLong))
  ]}
]
variables [
 BindPose {type=boolean, default=true}
 CrystalAnim_A {type=Vector}
 CrystalAnim_B {type=Vector}
 Dyn {type=boolean}
 DynCrouch {type=boolean}
 DynSkydive {type=boolean}
 FE_Idle {type=boolean}
 GravityOverride {type=Vector}
 is_WearingCape {type=boolean}
 isWearingPackBool {type=boolean}
 LOD1-3 {type=boolean}
 Move {type=boolean}
 Pawn Velocity {type=Vector}
 Pawn Velocity Z {type=number}
]
