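# Control Rig for LexaBlitz's face accessory mesh that procedurally poses her long back hair against the body.
# Nulls blended between head, neck and spine accompany a spline IK on the center chain, which left and right chains copy via constraints.
# Spherical pose readers on upper arms and head weight corrective offsets, then capsule clamps around spine controls constrain lower bones.
# A Cape flag adds fixed and head-up offsets; separate side-chain IK setups are unwired.
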ControlRigBlueprint /BRCosmetics/Characters/Player/Female/Medium/Bodies/F_MED_LexaBlitz/Meshes/Parts/F_MED_LexaBlitz_FaceAcc_CtrlRig
graphs [
 ContainedGraph {nodes [
   n1 = RigVMFunctionEntryNode()
   SequenceAggregate = &RigVMFunction_Sequence(ExecuteContext=@n1.ExecuteContext)
   SequenceAggregate_1 = &RigVMFunction_Sequence(ExecuteContext=@SequenceAggregate.B)
   SequenceAggregate_2 = &RigVMFunction_Sequence(ExecuteContext=@SequenceAggregate_1.B)
   RigVMFunctionReturnNode(A=@SequenceAggregate.A, B=@SequenceAggregate_1.A, C=@SequenceAggregate_2.A, D=@SequenceAggregate_2.B)
  ]}
 ContainedGraph {nodes [
   n1 = RigVMFunctionEntryNode()
   SequenceAggregate = &RigVMFunction_Sequence(ExecuteContext=@n1.ExecuteContext)
   SequenceAggregate_1 = &RigVMFunction_Sequence(ExecuteContext=@SequenceAggregate.B)
   SequenceAggregate_2 = &RigVMFunction_Sequence(ExecuteContext=@SequenceAggregate_1.B)
   RigVMFunctionReturnNode(A=@SequenceAggregate.A, B=@SequenceAggregate_1.A, C=@SequenceAggregate_2.A, D=@SequenceAggregate_2.B)
  ]}
 New Function_ContainedGraph {nodes [
   n1 = RigVMFunctionEntryNode()
   RigUnit_GetTransform = &RigUnit_GetTransform(bInitial=False, Item=@n1.BoneMove)
   ClampSpatially = /Script/RigVM.RigVMFunction_MathTransformClampSpatially(
    DebugColor=(R=1,G=0,B=0,A=1)
    DebugThickness=1
    Maximum=100
    Type=Capsule
    bDrawDebug=true
    Axis=@n1.Axis
    Minimum=@n1.Minimum
    Value=@RigUnit_GetTransform.Transform
    Space=&RigUnit_GetTransform(bInitial=False, Item=@n1.BoneSpace).Transform
   ) {
    is=struct RigVMUnitNode
    template="ClampSpatially::Execute(in Value,in Axis,in Type,in Minimum,in Maximum,in Space,in bDrawDebug,in DebugColor,in DebugThickness,out Result)"
   }
   > RigUnit_SetTransform = &RigUnit_SetTransform(Weight=1, bInitial=False, Value=@ClampSpatially.Result, Item=@n1.BoneMove)
   > RigVMFunctionReturnNode()
  ]}
 RigVMFunctionLibrary {nodes [Clamp Capsule = RigVMCollapseNode(Axis=Z, BoneMove=(Type=None,Name="None"), BoneSpace=(Type=None,Name="None"), ExecuteContext=(), Minimum=20)]}
 RigVMFunction_Sequence_1_ContainedGraph {nodes [
   n1 = RigVMFunctionEntryNode()
   RigVMFunction_Sequence = &RigVMFunction_Sequence(ExecuteContext=@n1.ExecuteContext)
   n2 = &RigVMFunction_Sequence(ExecuteContext=@RigVMFunction_Sequence.B)
   RigVMFunctionReturnNode(A=@RigVMFunction_Sequence.A, B=@n2.A, C=@n2.B)
  ]}
 RigVMFunction_Sequence_1_ContainedGraph {nodes [
   n1 = RigVMFunctionEntryNode()
   RigVMFunction_Sequence = &RigVMFunction_Sequence(ExecuteContext=@n1.ExecuteContext)
   n2 = &RigVMFunction_Sequence(ExecuteContext=@RigVMFunction_Sequence.B)
   RigVMFunctionReturnNode(A=@RigVMFunction_Sequence.A, B=@n2.A, C=@n2.B)
  ]}
 RigVMFunction_Sequence_2_ContainedGraph {nodes [
   n1 = RigVMFunctionEntryNode()
   n2 = &RigVMFunction_Sequence(ExecuteContext=@n1.ExecuteContext)
   n3 = &RigVMFunction_Sequence(ExecuteContext=@n2.B)
   n4 = &RigVMFunction_Sequence(ExecuteContext=@n3.B)
   RigVMFunctionReturnNode(A=@n2.A, B=@n3.A, C=@n4.A, D=@n4.B)
  ]}
 RigVMFunction_Sequence_2_ContainedGraph {nodes [
   n1 = RigVMFunctionEntryNode()
   n2 = &RigVMFunction_Sequence(ExecuteContext=@n1.ExecuteContext)
   n3 = &RigVMFunction_Sequence(ExecuteContext=@n2.B)
   n4 = &RigVMFunction_Sequence(ExecuteContext=@n3.B)
   n5 = &RigVMFunction_Sequence(ExecuteContext=@n4.B)
   RigVMFunctionReturnNode(A=@n2.A, B=@n3.A, C=@n4.A, D=@n5.A, E=@n5.B)
  ]}
 RigVMFunction_Sequence_2_ContainedGraph {nodes [
   n1 = RigVMFunctionEntryNode()
   n2 = &RigVMFunction_Sequence(ExecuteContext=@n1.ExecuteContext)
   n3 = &RigVMFunction_Sequence(ExecuteContext=@n2.B)
   RigVMFunctionReturnNode(A=@n2.A, B=@n3.A, C=@n3.B)
  ]}
 RigVMModel {nodes [
   &RigUnit_ParentConstraint(
    AdvancedSettings=(RotationOrderForFilter=XZY)
    Child=(Type=Null,Name="dyn_Hair_BackL_01_Null")
    ExecutePin=()
    Filter=(TranslationFilter=(bX=true,bY=true,bZ=true),RotationFilter=(bX=false,bY=false,bZ=false),ScaleFilter=(bX=True,bY=True,bZ=True))
    Parents=((Item=(Type=Bone,Name="head"),Weight=0.9),(Item=(Type=Bone,Name="spine_05"),Weight=0.1))
    Weight=1
   )
   > &RigUnit_ParentConstraint(
    AdvancedSettings=(RotationOrderForFilter=XZY)
    Child=(Type=Null,Name="dyn_Hair_BackL_02_Null")
    Filter=(TranslationFilter=(bX=True,bY=True,bZ=True),RotationFilter=(bX=false,bY=false,bZ=false),ScaleFilter=(bX=True,bY=True,bZ=True))
    Parents=((Item=(Type=Bone,Name="head"),Weight=0.6),(Item=(Type=Bone,Name="spine_05"),Weight=0.4))
    Weight=1
   )
   > &RigUnit_ParentConstraint(
    AdvancedSettings=(RotationOrderForFilter=XZY)
    Child=(Type=Null,Name="dyn_Hair_BackL_03_Null")
    Filter=(TranslationFilter=(bX=True,bY=True,bZ=True),RotationFilter=(bX=false,bY=false,bZ=false),ScaleFilter=(bX=True,bY=True,bZ=True))
    Parents=((Item=(Type=Bone,Name="dyn_Hair_BackL_03"),Weight=0.5),(Item=(Type=Bone,Name="spine_05"),Weight=0.5))
    Weight=1
   )
   > &RigUnit_ParentConstraint(
    AdvancedSettings=(RotationOrderForFilter=XZY)
    Child=(Type=Null,Name="dyn_Hair_BackL_04_Null")
    Filter=(TranslationFilter=(bX=True,bY=True,bZ=True),RotationFilter=(bX=false,bY=false,bZ=false),ScaleFilter=(bX=True,bY=True,bZ=True))
    Parents=((Item=(Type=Bone,Name="clavicle_l"),Weight=0),(Item=(Type=Bone,Name="spine_05"),Weight=1))
    Weight=1
   )
   > &RigUnit_ParentConstraint(
    AdvancedSettings=(RotationOrderForFilter=XZY)
    Child=(Type=Null,Name="dyn_Hair_BackL_05_Null")
    Filter=(TranslationFilter=(bX=True,bY=True,bZ=True),RotationFilter=(bX=false,bY=false,bZ=false),ScaleFilter=(bX=True,bY=True,bZ=True))
    Parents=((Item=(Type=Bone,Name="clavicle_l"),Weight=0.15),(Item=(Type=Bone,Name="spine_04"),Weight=0.85))
    Weight=1
   )
   > &RigUnit_ParentConstraint(
    AdvancedSettings=(RotationOrderForFilter=XZY)
    Child=(Type=Null,Name="dyn_Hair_BackL_06_Null")
    Filter=(TranslationFilter=(bX=true,bY=true,bZ=true),RotationFilter=(bX=true,bY=true,bZ=true),ScaleFilter=(bX=True,bY=True,bZ=True))
    Parents=((Item=(Type=Bone,Name="dyn_Hair_BackL_06"),Weight=0.2),(Item=(Type=Bone,Name="spine_02"),Weight=0.8))
    Weight=1
   )
   RerouteNode = RigVMDispatchNode(Values=((Type=Bone,Name="dyn_Hair_BackL_01"),((),Type=Bone,Name="dyn_Hair_BackL_02"),((),Type=Bone,Name="dyn_Hair_BackL_03"),((),Type=Bone,Name="dyn_Hair_BackL_04"),((),Type=Bone,Name="dyn_Hair_BackL_05"),((),Type=Bone,Name="dyn_Hair_BackL_06"))) {template="DISPATCH_RigVMDispatch_ArrayMake(in Values,out Array)"}
   > SplineIK = RigVMFunctionReferenceNode(
    Primary Axis=(X=0,Y=-1,Z=0)
    Samples Per Segment=1
    Secondary Spline Direction=(X=1,Y=1,Z=1)
    Spline Mode=BSpline
    Stretch=false
    Up Axis=(X=0,Y=0,Z=0)
    debug=false
    Bones=@RerouteNode.Array
    Controls=RigVMVariableNode() {variable=dyn_Hair_BackL_ctrls}.Value
   )
   > TwistBones_1 = RigVMFunctionReferenceNode(
    Axis to Twist=(bX=false,bY=true,bZ=false)
    Bones to Twist=((Type=Bone,Name="dyn_Hair_BackL_01"),(Type=Bone,Name="dyn_Hair_BackL_05"))
    End=(Type=Bone,Name="head")
    Start=(Type=Bone,Name="spine_02")
   )
   &RigUnit_ParentConstraint(
    AdvancedSettings=(RotationOrderForFilter=XZY)
    Child=(Type=Null,Name="dyn_Hair_BackR_01_Null")
    ExecutePin=()
    Filter=(TranslationFilter=(bX=true,bY=true,bZ=true),RotationFilter=(bX=false,bY=false,bZ=false),ScaleFilter=(bX=True,bY=True,bZ=True))
    Parents=((Item=(Type=Bone,Name="head"),Weight=0.8),(Item=(Type=Bone,Name="spine_05"),Weight=0.2))
    Weight=1
   )
   > &RigUnit_ParentConstraint(
    AdvancedSettings=(RotationOrderForFilter=XZY)
    Child=(Type=Null,Name="dyn_Hair_BackR_02_Null")
    Filter=(TranslationFilter=(bX=True,bY=True,bZ=True),RotationFilter=(bX=false,bY=false,bZ=false),ScaleFilter=(bX=True,bY=True,bZ=True))
    Parents=((Item=(Type=Bone,Name="neck_01"),Weight=0.8),(Item=(Type=Bone,Name="spine_05"),Weight=0.2))
    Weight=1
   )
   > &RigUnit_ParentConstraint(
    AdvancedSettings=(RotationOrderForFilter=XZY)
    Child=(Type=Null,Name="dyn_Hair_BackR_03_Null")
    Filter=(TranslationFilter=(bX=True,bY=True,bZ=True),RotationFilter=(bX=false,bY=false,bZ=false),ScaleFilter=(bX=True,bY=True,bZ=True))
    Parents=((Item=(Type=Bone,Name="dyn_Hair_BackR_03"),Weight=0.4),(Item=(Type=Bone,Name="spine_05"),Weight=0.6))
    Weight=1
   )
   > &RigUnit_ParentConstraint(
    AdvancedSettings=(RotationOrderForFilter=XZY)
    Child=(Type=Null,Name="dyn_Hair_BackR_04_Null")
    Filter=(TranslationFilter=(bX=True,bY=True,bZ=True),RotationFilter=(bX=false,bY=false,bZ=false),ScaleFilter=(bX=True,bY=True,bZ=True))
    Parents=((Item=(Type=Bone,Name="upperarm_r"),Weight=0.5),(Item=(Type=Bone,Name="spine_05"),Weight=0.5))
    Weight=1
   )
   > &RigUnit_ParentConstraint(
    AdvancedSettings=(RotationOrderForFilter=XZY)
    Child=(Type=Null,Name="dyn_Hair_BackR_05_Null")
    Filter=(TranslationFilter=(bX=True,bY=True,bZ=True),RotationFilter=(bX=false,bY=false,bZ=false),ScaleFilter=(bX=True,bY=True,bZ=True))
    Parents=((Item=(Type=Bone,Name="clavicle_r"),Weight=0),(Item=(Type=Bone,Name="spine_04"),Weight=1))
    Weight=1
   )
   > &RigUnit_ParentConstraint(
    AdvancedSettings=(RotationOrderForFilter=XZY)
    Child=(Type=Null,Name="dyn_Hair_BackR_06_Null")
    Filter=(TranslationFilter=(bX=True,bY=True,bZ=True),RotationFilter=(bX=true,bY=true,bZ=true),ScaleFilter=(bX=True,bY=True,bZ=True))
    Parents=((Item=(Type=Bone,Name="dyn_Hair_BackR_06"),Weight=0.2),(Item=(Type=Bone,Name="spine_02"),Weight=0.8))
    Weight=1
   )
   RerouteNode_1 = RigVMDispatchNode(Values=((Type=Bone,Name="dyn_Hair_BackR_01"),((),Type=Bone,Name="dyn_Hair_BackR_02"),((),Type=Bone,Name="dyn_Hair_BackR_03"),((),Type=Bone,Name="dyn_Hair_BackR_04"),((),Type=Bone,Name="dyn_Hair_BackR_05"),((),Type=Bone,Name="dyn_Hair_BackR_06"))) {template="DISPATCH_RigVMDispatch_ArrayMake(in Values,out Array)"}
   > SplineIK_1 = RigVMFunctionReferenceNode(
    Primary Axis=(X=0,Y=-1,Z=0)
    Samples Per Segment=1
    Secondary Spline Direction=(X=1,Y=1,Z=1)
    Spline Mode=BSpline
    Stretch=false
    Up Axis=(X=0,Y=0,Z=0)
    debug=false
    Bones=@RerouteNode_1.Array
    Controls=RigVMVariableNode() {variable=dyn_Hair_BackR_ctrls}.Value
   )
   RigUnit_BeginExecution = &RigUnit_BeginExecution()
   SequenceAggregate = RigVMAggregateNode(ExecuteContext=@RigUnit_BeginExecution.ExecutePin)
   &RigUnit_ParentConstraint(
    AdvancedSettings=(RotationOrderForFilter=XZY)
    Child=(Type=Null,Name="dyn_Hair_Back_01_Null")
    Filter=(TranslationFilter=(bX=true,bY=true,bZ=true),RotationFilter=(bX=false,bY=false,bZ=false),ScaleFilter=(bX=True,bY=True,bZ=True))
    Parents=((Item=(Type=Bone,Name="head"),Weight=0.9),(Item=(Type=Bone,Name="spine_05"),Weight=0.1))
    Weight=1
    ExecutePin=@SequenceAggregate.A
   )
   > &RigUnit_ParentConstraint(
    AdvancedSettings=(RotationOrderForFilter=XZY)
    Child=(Type=Null,Name="dyn_Hair_Back_02_Null")
    Filter=(TranslationFilter=(bX=True,bY=True,bZ=True),RotationFilter=(bX=false,bY=false,bZ=false),ScaleFilter=(bX=True,bY=True,bZ=True))
    Parents=((Item=(Type=Bone,Name="neck_02"),Weight=0.6),(Item=(Type=Bone,Name="spine_05"),Weight=0.4))
    Weight=1
   )
   > &RigUnit_ParentConstraint(
    AdvancedSettings=(RotationOrderForFilter=XZY)
    Child=(Type=Null,Name="dyn_Hair_Back_03_Null")
    Filter=(TranslationFilter=(bX=True,bY=True,bZ=True),RotationFilter=(bX=false,bY=false,bZ=false),ScaleFilter=(bX=True,bY=True,bZ=True))
    Parents=((Item=(Type=Bone,Name="neck_02"),Weight=0.5),(Item=(Type=Bone,Name="spine_05"),Weight=0.5))
    Weight=1
   )
   > &RigUnit_ParentConstraint(
    AdvancedSettings=(RotationOrderForFilter=XZY)
    Child=(Type=Null,Name="dyn_Hair_Back_04_Null")
    Filter=(TranslationFilter=(bX=True,bY=True,bZ=True),RotationFilter=(bX=false,bY=false,bZ=false),ScaleFilter=(bX=True,bY=True,bZ=True))
    Parents=((Item=(Type=Bone,Name="clavicle_r"),Weight=0),(Item=(Type=Bone,Name="spine_04"),Weight=1),(Item=(Type=Bone,Name="clavicle_l"),Weight=0))
    Weight=1
   )
   > &RigUnit_ParentConstraint(
    AdvancedSettings=(RotationOrderForFilter=XZY)
    Child=(Type=Null,Name="dyn_Hair_Back_05_Null")
    Filter=(TranslationFilter=(bX=True,bY=True,bZ=True),RotationFilter=(bX=false,bY=false,bZ=false),ScaleFilter=(bX=True,bY=True,bZ=True))
    Parents=((Item=(Type=Bone,Name="clavicle_l"),Weight=0),(Item=(Type=Bone,Name="spine_03"),Weight=1),(Item=(Type=Bone,Name="clavicle_r"),Weight=0))
    Weight=1
   )
   > &RigUnit_ParentConstraint(
    AdvancedSettings=(RotationOrderForFilter=XZY)
    Child=(Type=Null,Name="dyn_Hair_Back_06_Null")
    Filter=(TranslationFilter=(bX=True,bY=True,bZ=True),RotationFilter=(bX=false,bY=false,bZ=false),ScaleFilter=(bX=True,bY=True,bZ=True))
    Parents=((Item=(Type=Bone,Name="dyn_Hair_Back_06"),Weight=0),(Item=(Type=Bone,Name="spine_03"),Weight=1))
    Weight=1
   )
   RerouteNode_2 = RigVMDispatchNode(Values=((Type=Bone,Name="dyn_Hair_Back_01"),((),Type=Bone,Name="dyn_Hair_Back_02"),((),Type=Bone,Name="dyn_Hair_Back_03"),((),Type=Bone,Name="dyn_Hair_Back_04"),((),Type=Bone,Name="dyn_Hair_Back_05"),((),Type=Bone,Name="dyn_Hair_Back_06"))) {template="DISPATCH_RigVMDispatch_ArrayMake(in Values,out Array)"}
   > SplineIK_2 = RigVMFunctionReferenceNode(
    Primary Axis=(X=0.2,Y=-1,Z=0)
    Samples Per Segment=1
    Secondary Spline Direction=(X=-1,Y=-1,Z=-1)
    Spline Mode=BSpline
    Stretch=false
    Up Axis=(X=0,Y=0,Z=0)
    debug=false
    Bones=@RerouteNode_2.Array
    Controls=RigVMVariableNode() {variable=dyn_Hair_Back_ctrls}.Value
   )
   > TwistBones = RigVMFunctionReferenceNode(
    Axis to Twist=(bX=true,bY=true,bZ=true)
    Bones to Twist=((Type=Bone,Name="dyn_Hair_Back_03"),(Type=Bone,Name="dyn_Hair_Back_04"))
    End=(Type=Bone,Name="spine_05")
    Start=(Type=Bone,Name="spine_05")
   )
   &RigUnit_ParentConstraint(
    AdvancedSettings=(RotationOrderForFilter=XZY)
    Child=(Type=Bone,Name="dyn_Hair_BackL_01")
    Filter=(TranslationFilter=(bX=true,bY=true,bZ=true),RotationFilter=(bX=true,bY=true,bZ=true),ScaleFilter=(bX=True,bY=True,bZ=True))
    Parents=((Item=(Type=Bone,Name="dyn_Hair_Back_01"),Weight=1),(Item=(Type=Bone,Name="spine_05"),Weight=0))
    Weight=1
    ExecutePin=@SequenceAggregate.B
   )
   > &RigUnit_ParentConstraint(
    AdvancedSettings=(RotationOrderForFilter=XZY)
    Child=(Type=Bone,Name="dyn_Hair_BackL_02")
    Filter=(TranslationFilter=(bX=True,bY=True,bZ=True),RotationFilter=(bX=true,bY=true,bZ=true),ScaleFilter=(bX=True,bY=True,bZ=True))
    Parents=((Item=(Type=Bone,Name="dyn_Hair_Back_02"),Weight=1),(Item=(Type=Bone,Name="spine_05"),Weight=0))
    Weight=1
   )
   > &RigUnit_ParentConstraint(
    AdvancedSettings=(RotationOrderForFilter=XZY)
    Child=(Type=Bone,Name="dyn_Hair_BackL_03")
    Filter=(TranslationFilter=(bX=True,bY=True,bZ=True),RotationFilter=(bX=true,bY=true,bZ=true),ScaleFilter=(bX=True,bY=True,bZ=True))
    Parents=((Item=(Type=Bone,Name="dyn_Hair_Back_03"),Weight=1),(Item=(Type=Bone,Name="spine_05"),Weight=0))
    Weight=1
   )
   > &RigUnit_ParentConstraint(
    AdvancedSettings=(RotationOrderForFilter=XZY)
    Child=(Type=Bone,Name="dyn_Hair_BackL_04")
    Filter=(TranslationFilter=(bX=True,bY=True,bZ=True),RotationFilter=(bX=true,bY=true,bZ=true),ScaleFilter=(bX=True,bY=True,bZ=True))
    Parents=((Item=(Type=Bone,Name="dyn_Hair_Back_04"),Weight=1),(Item=(Type=Bone,Name="spine_05"),Weight=0))
    Weight=1
   )
   > &RigUnit_ParentConstraint(
    AdvancedSettings=(RotationOrderForFilter=XZY)
    Child=(Type=Bone,Name="dyn_Hair_BackL_05")
    Filter=(TranslationFilter=(bX=True,bY=True,bZ=True),RotationFilter=(bX=true,bY=true,bZ=true),ScaleFilter=(bX=True,bY=True,bZ=True))
    Parents=((Item=(Type=Bone,Name="dyn_Hair_Back_05"),Weight=1),(Item=(Type=Bone,Name="spine_04"),Weight=0))
    Weight=1
   )
   > &RigUnit_ParentConstraint(
    AdvancedSettings=(RotationOrderForFilter=XZY)
    Child=(Type=Bone,Name="dyn_Hair_BackL_06")
    Filter=(TranslationFilter=(bX=true,bY=true,bZ=true),RotationFilter=(bX=true,bY=true,bZ=true),ScaleFilter=(bX=True,bY=True,bZ=True))
    Parents=((Item=(Type=Bone,Name="dyn_Hair_Back_06"),Weight=1),(Item=(Type=Bone,Name="spine_02"),Weight=0))
    Weight=1
   )
   > n1 = &RigUnit_SphericalPoseReader(
    ActiveRegionScaleFactors=(PositiveWidth=0.6,NegativeWidth=0.6,PositiveHeight=0.6,NegativeHeight=0.6)
    ActiveRegionSize=0.1
    Debug=(bDrawDebug=false,DebugScale=25,DebugSegments=20,DebugThickness=0.25)
    DriverAxis=(X=1,Y=0,Z=0)
    DriverItem=(Type=Bone,Name="upperarm_l")
    FalloffRegionScaleFactors=(PositiveWidth=1,NegativeWidth=0.6,PositiveHeight=0.6,NegativeHeight=0.6)
    FalloffSize=0.55
    FlipHeightScaling=false
    FlipWidthScaling=false
    OptionalParentItem=(Type=Bone,Name="spine_05")
    RotationOffset=(X=-80,Y=-30,Z=-10)
   )
   > &RigUnit_ParentConstraint(
    AdvancedSettings=(RotationOrderForFilter=XZY)
    Child=(Type=Bone,Name="dyn_Hair_BackL_03")
    Filter=(TranslationFilter=(bX=false,bY=false,bZ=false),RotationFilter=(bX=true,bY=true,bZ=true),ScaleFilter=(bX=True,bY=True,bZ=True))
    Parents=((Item=(Type=Bone,Name="dyn_Hair_Back_03"),Weight=0.3),(Item=(Type=Bone,Name="upperarm_l"),Weight=0.7))
    Weight=@n1.OutputParam
   )
   n2 = &RigVMFunction_MathQuaternionFromEuler(Euler=(X=20,Y=-10,Z=0), RotationOrder=YXZ)
   > &RigUnit_ModifyTransforms(
    ItemToModify="((Item=(Type=Bone,Name=\"dyn_Hair_BackL_02\"),Transform=(Rotation=(X=0,Y=0,Z=0,W=1),Translation=(X=2,Y=-1,Z=1),Scale3D=(X=1,Y=1,Z=1))),(Item=(Type=Bone,Name=\"dyn_Hair_BackL_04\"),Transform=(Rotation=(X=0,Y=0,Z=0,W=1),Translation=(X=0,Y=0,Z=0),Scale3D=(X=1,Y=1 ...(truncated)"
    ItemToModify.0.Transform.Rotation=@n2.Result
    ItemToModify.1.Transform.Rotation=@n2.Result
    ItemToModify.2.Transform.Rotation=&RigVMFunction_MathQuaternionFromEuler(Euler=(X=0,Y=-20,Z=25)).Result
    Weight=@n1.OutputParam
   )
   n3 = &RigVMFunction_MathQuaternionFromEuler(Euler=(X=0,Y=0,Z=5))
   > &RigUnit_ModifyTransforms(
    ItemToModify="((Item=(Type=Bone,Name=\"dyn_Hair_Back_03\"),Transform=(Rotation=(X=0,Y=0,Z=0,W=1),Translation=(X=0,Y=0,Z=1),Scale3D=(X=1,Y=1,Z=1))),(Item=(Type=Bone,Name=\"dyn_Hair_Back_04\"),Transform=(Rotation=(X=0,Y=0,Z=0,W=1),Translation=(X=0,Y=0,Z=0),Scale3D=(X=1,Y=1,Z= ...(truncated)"
    ItemToModify.2.Transform.Rotation=&RigVMFunction_MathQuaternionFromEuler(Euler=(X=0,Y=0,Z=2)).Result
    ItemToModify.0.Transform.Rotation=@n3.Result
    ItemToModify.1.Transform.Rotation=@n3.Result
    Weight=@n1.OutputParam
   )
   > n4 = &RigUnit_SphericalPoseReader(
    ActiveRegionScaleFactors=(PositiveWidth=1,NegativeWidth=1,PositiveHeight=1,NegativeHeight=1)
    ActiveRegionSize=0.2
    Debug=(bDrawDebug=false,DebugScale=25,DebugSegments=20,DebugThickness=0.25)
    DriverAxis=(X=1,Y=0,Z=0)
    DriverItem=(Type=Bone,Name="head")
    FalloffRegionScaleFactors=(PositiveWidth=1,NegativeWidth=1,PositiveHeight=1,NegativeHeight=1)
    FalloffSize=0.7
    FlipHeightScaling=false
    FlipWidthScaling=true
    OptionalParentItem=(Type=Bone,Name="clavicle_l")
    RotationOffset=(X=-100,Y=0,Z=7)
   )
   > &RigUnit_ParentConstraint(
    AdvancedSettings=(RotationOrderForFilter=XZY)
    Child=(Type=Bone,Name="dyn_Hair_BackL_03a")
    Filter=(TranslationFilter=(bX=false,bY=false,bZ=false),RotationFilter=(bX=true,bY=true,bZ=true),ScaleFilter=(bX=True,bY=True,bZ=True))
    Parents=((Item=(Type=Bone,Name="dyn_Hair_BackL_03"),Weight=0.3),(Item=(Type=Bone,Name="clavicle_l"),Weight=0.7))
    Weight=@n4.OutputParam
   )
   > &RigUnit_ModifyTransforms(
    ItemToModify=((Item=(Type=Bone,Name="dyn_Hair_BackL_03a"),Transform=(Rotation=(X=0,Y=0,Z=0,W=1),Translation=(X=1,Y=-1,Z=2),Scale3D=(X=1,Y=1,Z=1))))
    ItemToModify.0.Transform.Rotation=&RigVMFunction_MathQuaternionFromEuler(Euler=(X=20,Y=-30,Z=20)).Result
    Weight=@n4.OutputParam
   )
   > n5 = &RigUnit_SphericalPoseReader(
    ActiveRegionScaleFactors=(PositiveWidth=0.6,NegativeWidth=0.6,PositiveHeight=0.6,NegativeHeight=0.6)
    ActiveRegionSize=0.1
    Debug=(bDrawDebug=true,DebugScale=25,DebugSegments=20,DebugThickness=0.25)
    DriverAxis=(X=1,Y=0,Z=0)
    DriverItem=(Type=Bone,Name="upperarm_l")
    FalloffRegionScaleFactors=(PositiveWidth=1,NegativeWidth=0.6,PositiveHeight=0.6,NegativeHeight=0.6)
    FalloffSize=0.55
    FlipHeightScaling=false
    FlipWidthScaling=false
    OptionalParentItem=(Type=Bone,Name="spine_05")
    RotationOffset=(X=-12,Y=-46,Z=-10)
   )
   > &RigUnit_OffsetTransformForItem(
    Item=(Type=Bone,Name="dyn_Hair_BackL_03")
    OffsetTransform=(Rotation=(X=0.2588,Y=0,Z=0,W=0.9659),Translation=(X=3,Y=0,Z=2),Scale3D=(X=1,Y=1,Z=1))
    bPropagateToChildren=False
    Weight=@n5.OutputParam
   )
   > &RigUnit_OffsetTransformForItem(
    Item=(Type=Bone,Name="dyn_Hair_BackL_03b")
    OffsetTransform=(Rotation=(X=0,Y=0,Z=0.2588,W=0.9659),Translation=(X=2,Y=0,Z=0),Scale3D=(X=1,Y=1,Z=1))
    bPropagateToChildren=False
    Weight=@n5.OutputParam
   )
   > &RigUnit_OffsetTransformForItem(
    Item=(Type=Bone,Name="dyn_Hair_BackL_04")
    OffsetTransform=(Rotation=(X=0.2588,Y=0,Z=0,W=0.9659),Translation=(X=3,Y=0,Z=2),Scale3D=(X=1,Y=1,Z=1))
    bPropagateToChildren=False
    Weight=@n5.OutputParam
   )
   &RigUnit_ParentConstraint(
    AdvancedSettings=(RotationOrderForFilter=XZY)
    Child=(Type=Bone,Name="dyn_Hair_BackR_01")
    Filter=(TranslationFilter=(bX=true,bY=true,bZ=true),RotationFilter=(bX=true,bY=true,bZ=true),ScaleFilter=(bX=True,bY=True,bZ=True))
    Parents=((Item=(Type=Bone,Name="dyn_Hair_Back_01"),Weight=1),(Item=(Type=Bone,Name="spine_05"),Weight=0))
    Weight=1
    ExecutePin=@SequenceAggregate.C
   )
   > &RigUnit_ParentConstraint(
    AdvancedSettings=(RotationOrderForFilter=XZY)
    Child=(Type=Bone,Name="dyn_Hair_BackR_02")
    Filter=(TranslationFilter=(bX=True,bY=True,bZ=True),RotationFilter=(bX=true,bY=true,bZ=true),ScaleFilter=(bX=True,bY=True,bZ=True))
    Parents=((Item=(Type=Bone,Name="dyn_Hair_Back_02"),Weight=1),(Item=(Type=Bone,Name="spine_05"),Weight=0))
    Weight=1
   )
   > &RigUnit_ParentConstraint(
    AdvancedSettings=(RotationOrderForFilter=XZY)
    Child=(Type=Bone,Name="dyn_Hair_BackR_03")
    Filter=(TranslationFilter=(bX=True,bY=True,bZ=True),RotationFilter=(bX=true,bY=true,bZ=true),ScaleFilter=(bX=True,bY=True,bZ=True))
    Parents=((Item=(Type=Bone,Name="dyn_Hair_Back_03"),Weight=1),(Item=(Type=Bone,Name="spine_05"),Weight=0))
    Weight=1
   )
   > &RigUnit_ParentConstraint(
    AdvancedSettings=(RotationOrderForFilter=XZY)
    Child=(Type=Bone,Name="dyn_Hair_BackR_04")
    Filter=(TranslationFilter=(bX=True,bY=True,bZ=True),RotationFilter=(bX=true,bY=true,bZ=true),ScaleFilter=(bX=True,bY=True,bZ=True))
    Parents=((Item=(Type=Bone,Name="dyn_Hair_Back_04"),Weight=1),(Item=(Type=Bone,Name="spine_05"),Weight=0))
    Weight=1
   )
   > &RigUnit_ParentConstraint(
    AdvancedSettings=(RotationOrderForFilter=XZY)
    Child=(Type=Bone,Name="dyn_Hair_BackR_05")
    Filter=(TranslationFilter=(bX=True,bY=True,bZ=True),RotationFilter=(bX=true,bY=true,bZ=true),ScaleFilter=(bX=True,bY=True,bZ=True))
    Parents=((Item=(Type=Bone,Name="dyn_Hair_Back_05"),Weight=1),(Item=(Type=Bone,Name="spine_04"),Weight=0))
    Weight=1
   )
   > &RigUnit_ParentConstraint(
    AdvancedSettings=(RotationOrderForFilter=XZY)
    Child=(Type=Bone,Name="dyn_Hair_BackR_06")
    Filter=(TranslationFilter=(bX=true,bY=true,bZ=true),RotationFilter=(bX=true,bY=true,bZ=true),ScaleFilter=(bX=True,bY=True,bZ=True))
    Parents=((Item=(Type=Bone,Name="dyn_Hair_Back_06"),Weight=1),(Item=(Type=Bone,Name="spine_02"),Weight=0))
    Weight=1
   )
   > n6 = &RigUnit_SphericalPoseReader(
    ActiveRegionScaleFactors=(PositiveWidth=0.6,NegativeWidth=0.6,PositiveHeight=0.6,NegativeHeight=0.6)
    ActiveRegionSize=0.1
    Debug=(bDrawDebug=false,DebugScale=25,DebugSegments=20,DebugThickness=0.25)
    DriverAxis=(X=-1,Y=0,Z=0)
    DriverItem=(Type=Bone,Name="upperarm_r")
    FalloffRegionScaleFactors=(PositiveWidth=1,NegativeWidth=0.6,PositiveHeight=0.6,NegativeHeight=0.6)
    FalloffSize=0.6
    FlipHeightScaling=false
    FlipWidthScaling=true
    OptionalParentItem=(Type=Bone,Name="spine_05")
    RotationOffset=(X=80,Y=-30,Z=-10)
   )
   > &RigUnit_ParentConstraint(
    AdvancedSettings=(RotationOrderForFilter=XZY)
    Child=(Type=Bone,Name="dyn_Hair_BackR_03")
    Filter=(TranslationFilter=(bX=false,bY=false,bZ=false),RotationFilter=(bX=true,bY=true,bZ=true),ScaleFilter=(bX=True,bY=True,bZ=True))
    Parents=((Item=(Type=Bone,Name="dyn_Hair_Back_03"),Weight=0.3),(Item=(Type=Bone,Name="upperarm_r"),Weight=0.7))
    Weight=@n6.OutputParam
   )
   n7 = &RigVMFunction_MathQuaternionFromEuler(Euler=(X=-20,Y=10,Z=0), RotationOrder=YXZ)
   > &RigUnit_ModifyTransforms(
    ItemToModify="((Item=(Type=Bone,Name=\"dyn_Hair_BackR_02\"),Transform=(Rotation=(X=0,Y=0,Z=0,W=1),Translation=(X=2,Y=-1,Z=1),Scale3D=(X=1,Y=1,Z=1))),(Item=(Type=Bone,Name=\"dyn_Hair_BackR_04\"),Transform=(Rotation=(X=0,Y=0,Z=0,W=1),Translation=(X=0,Y=0,Z=0),Scale3D=(X=1,Y=1 ...(truncated)"
    ItemToModify.0.Transform.Rotation=@n7.Result
    ItemToModify.1.Transform.Rotation=@n7.Result
    ItemToModify.2.Transform.Rotation=&RigVMFunction_MathQuaternionFromEuler(Euler=(X=0,Y=20,Z=25)).Result
    Weight=@n6.OutputParam
   )
   n8 = &RigVMFunction_MathQuaternionFromEuler(Euler=(X=0,Y=0,Z=5))
   > &RigUnit_ModifyTransforms(
    ItemToModify="((Item=(Type=Bone,Name=\"dyn_Hair_Back_03\"),Transform=(Rotation=(X=0,Y=0,Z=0,W=1),Translation=(X=0,Y=0,Z=1),Scale3D=(X=1,Y=1,Z=1))),(Item=(Type=Bone,Name=\"dyn_Hair_Back_04\"),Transform=(Rotation=(X=0,Y=0,Z=0,W=1),Translation=(X=0,Y=0,Z=0),Scale3D=(X=1,Y=1,Z= ...(truncated)"
    ItemToModify.2.Transform.Rotation=&RigVMFunction_MathQuaternionFromEuler(Euler=(X=0,Y=0,Z=2)).Result
    ItemToModify.0.Transform.Rotation=@n8.Result
    ItemToModify.1.Transform.Rotation=@n8.Result
    Weight=@n6.OutputParam
   )
   > n9 = &RigUnit_SphericalPoseReader(
    ActiveRegionScaleFactors=(PositiveWidth=1,NegativeWidth=1,PositiveHeight=1,NegativeHeight=1)
    ActiveRegionSize=0.2
    Debug=(bDrawDebug=false,DebugScale=25,DebugSegments=20,DebugThickness=0.25)
    DriverAxis=(X=1,Y=0,Z=0)
    DriverItem=(Type=Bone,Name="head")
    FalloffRegionScaleFactors=(PositiveWidth=1,NegativeWidth=1,PositiveHeight=1,NegativeHeight=1)
    FalloffSize=0.7
    FlipHeightScaling=false
    FlipWidthScaling=true
    OptionalParentItem=(Type=Bone,Name="clavicle_r")
    RotationOffset=(X=-80,Y=0,Z=7)
   )
   > &RigUnit_ParentConstraint(
    AdvancedSettings=(RotationOrderForFilter=XZY)
    Child=(Type=Bone,Name="dyn_Hair_BackR_03a")
    Filter=(TranslationFilter=(bX=false,bY=false,bZ=false),RotationFilter=(bX=true,bY=true,bZ=true),ScaleFilter=(bX=True,bY=True,bZ=True))
    Parents=((Item=(Type=Bone,Name="dyn_Hair_BackR_03"),Weight=0.3),(Item=(Type=Bone,Name="clavicle_r"),Weight=0.7))
    Weight=@n9.OutputParam
   )
   > &RigUnit_ModifyTransforms(
    ItemToModify="((Item=(Type=Bone,Name=\"dyn_Hair_BackR_03a\"),Transform=(Rotation=(X=0,Y=0,Z=0,W=1),Translation=(X=1,Y=1,Z=-2),Scale3D=(X=1,Y=1,Z=1))),(Item=(Type=Bone,Name=\"dyn_Hair_BackR_03\"),Transform=(Rotation=(X=0,Y=0,Z=0,W=1),Translation=(X=0,Y=0,Z=0),Scale3D=(X=1,Y= ...(truncated)"
    ItemToModify.0.Transform.Rotation=&RigVMFunction_MathQuaternionFromEuler(Euler=(X=-20,Y=10,Z=20)).Result
    Weight=@n9.OutputParam
   )
   > n10 = &RigUnit_SphericalPoseReader(
    ActiveRegionScaleFactors=(PositiveWidth=0.6,NegativeWidth=0.6,PositiveHeight=0.6,NegativeHeight=0.6)
    ActiveRegionSize=0.1
    Debug=(bDrawDebug=true,DebugScale=25,DebugSegments=20,DebugThickness=0.25)
    DriverAxis=(X=-1,Y=0,Z=0)
    DriverItem=(Type=Bone,Name="upperarm_r")
    FalloffRegionScaleFactors=(PositiveWidth=1,NegativeWidth=0.6,PositiveHeight=0.6,NegativeHeight=0.6)
    FalloffSize=0.55
    FlipHeightScaling=false
    FlipWidthScaling=false
    OptionalParentItem=(Type=Bone,Name="spine_05")
    RotationOffset=(X=115,Y=0,Z=69)
   )
   > &RigUnit_OffsetTransformForItem(
    Item=(Type=Bone,Name="dyn_Hair_BackR_03")
    OffsetTransform=(Rotation=(X=0,Y=0,Z=0,W=1),Translation=(X=2,Y=0,Z=-1),Scale3D=(X=1,Y=1,Z=1))
    bPropagateToChildren=True
    Weight=@n10.OutputParam
   )
   > &RigUnit_OffsetTransformForItem(
    Item=(Type=Bone,Name="dyn_Hair_BackR_04")
    OffsetTransform=(Rotation=(X=0,Y=0,Z=0,W=1),Translation=(X=2,Y=0,Z=-2),Scale3D=(X=1,Y=1,Z=1))
    bPropagateToChildren=True
    Weight=@n10.OutputParam
   )
   RigVMFunction_Sequence_3 = RigVMAggregateNode(ExecuteContext=@SequenceAggregate.D)
   RigVMFunction_Sequence = RigVMAggregateNode(ExecuteContext=@RigVMFunction_Sequence_3.A)
   Clamp Capsule = RigVMFunctionReferenceNode(
    Axis=Z
    BoneMove=(Type=Bone,Name="dyn_Hair_BackL_05")
    BoneSpace=(Type=Control,Name="spine_04_ctrl")
    Minimum=16
    ExecuteContext=@RigVMFunction_Sequence.A
   )
   > Clamp Capsule_1 = RigVMFunctionReferenceNode(Axis=Z, BoneMove=(Type=Bone,Name="dyn_Hair_BackL_06"), BoneSpace=(Type=Control,Name="spine_04_ctrl"), Minimum=16)
   Clamp Capsule_2 = RigVMFunctionReferenceNode(
    Axis=Z
    BoneMove=(Type=Bone,Name="dyn_Hair_Back_05")
    BoneSpace=(Type=Control,Name="spine_04_ctrl")
    Minimum=16
    ExecuteContext=@RigVMFunction_Sequence.B
   )
   > Clamp Capsule_3 = RigVMFunctionReferenceNode(Axis=Z, BoneMove=(Type=Bone,Name="dyn_Hair_Back_06"), BoneSpace=(Type=Control,Name="spine_03_ctrl"), Minimum=22)
   Clamp Capsule_4 = RigVMFunctionReferenceNode(
    Axis=Z
    BoneMove=(Type=Bone,Name="dyn_Hair_BackR_05")
    BoneSpace=(Type=Control,Name="spine_04_ctrl")
    Minimum=16
    ExecuteContext=@RigVMFunction_Sequence.C
   )
   > Clamp Capsule_5 = RigVMFunctionReferenceNode(Axis=Z, BoneMove=(Type=Bone,Name="dyn_Hair_BackR_06"), BoneSpace=(Type=Control,Name="spine_04_ctrl"), Minimum=16)
   RigVMFunction_ControlFlowBranch = &RigVMFunction_ControlFlowBranch(ExecuteContext=@RigVMFunction_Sequence_3.B, Condition=RigVMVariableNode() {variable=Cape}.Value)
   RigVMFunction_Sequence_4 = RigVMAggregateNode(ExecuteContext=@RigVMFunction_ControlFlowBranch.True)
   RigVMFunction_Sequence_2 = RigVMAggregateNode(ExecuteContext=@RigVMFunction_Sequence_4.A)
   n11 = &RigUnit_SphericalPoseReader(
    ActiveRegionScaleFactors=(PositiveWidth=0.5,NegativeWidth=0.1,PositiveHeight=1,NegativeHeight=1)
    ActiveRegionSize=0.2
    Debug=(bDrawDebug=True,bDraw2D=False,bDrawLocalAxes=False,DebugScale=25,DebugSegments=20,DebugThickness=0.25)
    DriverAxis=(X=0,Y=1,Z=0)
    DriverItem=(Type=Bone,Name="head")
    FalloffRegionScaleFactors=(PositiveWidth=1,NegativeWidth=0.2,PositiveHeight=1,NegativeHeight=1)
    FalloffSize=0.4
    FlipHeightScaling=False
    FlipWidthScaling=False
    OptionalParentItem=(Type=Bone,Name="None")
    RotationOffset=(X=90,Y=0,Z=-30)
    ExecutePin=@RigVMFunction_Sequence_2.A
   )
   > RigVMVariableNode(Value=@n11.OutputParam) {variable=SPR_HeadUp}
   &RigUnit_OffsetTransformForItem(
    Item=(Type=Bone,Name="dyn_Hair_BackL_04")
    OffsetTransform=(Rotation=(X=0,Y=0,Z=0,W=1),Translation=(X=-2,Y=0,Z=0),Scale3D=(X=0.7,Y=1,Z=1))
    bPropagateToChildren=True
    ExecutePin=@RigVMFunction_Sequence_2.C
    Weight=RigVMVariableNode() {variable=SPR_HeadUp}.Value
   )
   &RigUnit_OffsetTransformForItem(
    Item=(Type=Bone,Name="dyn_Hair_Back_04")
    OffsetTransform=(Rotation=(X=0,Y=0,Z=0,W=1),Translation=(X=-2,Y=0,Z=0),Scale3D=(X=0.7,Y=1,Z=1))
    bPropagateToChildren=True
    ExecutePin=@RigVMFunction_Sequence_2.D
    Weight=RigVMVariableNode() {variable=SPR_HeadUp}.Value
   )
   &RigUnit_OffsetTransformForItem(
    Item=(Type=Bone,Name="dyn_Hair_BackR_04")
    OffsetTransform=(Rotation=(X=0,Y=0,Z=0,W=1),Translation=(X=-2,Y=0,Z=0),Scale3D=(X=0.7,Y=1,Z=1))
    bPropagateToChildren=True
    ExecutePin=@RigVMFunction_Sequence_2.E
    Weight=RigVMVariableNode() {variable=SPR_HeadUp}.Value
   )
   RigVMFunction_Sequence_1 = RigVMAggregateNode(ExecuteContext=@RigVMFunction_Sequence_4.B)
   &RigUnit_OffsetTransformForItem(
    Item=(Type=Bone,Name="dyn_Hair_BackL_04")
    OffsetTransform=(Rotation=(X=0,Y=0,Z=0,W=1),Translation=(X=-0.5,Y=0,Z=0),Scale3D=(X=1,Y=1,Z=1))
    Weight=1
    bPropagateToChildren=false
    ExecutePin=@RigVMFunction_Sequence_1.B
   )
   > &RigUnit_OffsetTransformForItem(
    Item=(Type=Bone,Name="dyn_Hair_BackL_05")
    OffsetTransform=(Rotation=(X=0,Y=0,Z=0,W=1),Translation=(X=-0.5,Y=0,Z=0),Scale3D=(X=0.5,Y=1,Z=1))
    Weight=1
    bPropagateToChildren=True
   )
   > &RigUnit_OffsetTransformForItem(
    Item=(Type=Bone,Name="dyn_Hair_BackL_06")
    OffsetTransform=(Rotation=(X=0,Y=0,Z=0.1736,W=0.9848),Translation=(X=0,Y=0,Z=0),Scale3D=(X=1,Y=1,Z=1))
    Weight=1
    bPropagateToChildren=True
   )
   &RigUnit_OffsetTransformForItem(
    Item=(Type=Bone,Name="dyn_Hair_Back_04")
    OffsetTransform=(Rotation=(X=0,Y=0,Z=0,W=1),Translation=(X=-0.5,Y=0,Z=0),Scale3D=(X=1,Y=1,Z=1))
    Weight=1
    bPropagateToChildren=false
    ExecutePin=@RigVMFunction_Sequence_1.C
   )
   > &RigUnit_OffsetTransformForItem(
    Item=(Type=Bone,Name="dyn_Hair_Back_05")
    OffsetTransform=(Rotation=(X=0,Y=0,Z=0,W=1),Translation=(X=-0.5,Y=0,Z=0),Scale3D=(X=0.5,Y=1,Z=1))
    Weight=1
    bPropagateToChildren=True
   )
   > &RigUnit_OffsetTransformForItem(
    Item=(Type=Bone,Name="dyn_Hair_Back_06")
    OffsetTransform=(Rotation=(X=0,Y=0,Z=0.1736,W=0.9848),Translation=(X=0,Y=0,Z=0),Scale3D=(X=1,Y=1,Z=1))
    Weight=1
    bPropagateToChildren=True
   )
   &RigUnit_OffsetTransformForItem(
    Item=(Type=Bone,Name="dyn_Hair_BackR_04")
    OffsetTransform=(Rotation=(X=0,Y=0,Z=0,W=1),Translation=(X=-0.5,Y=0,Z=0),Scale3D=(X=1,Y=1,Z=1))
    Weight=1
    bPropagateToChildren=false
    ExecutePin=@RigVMFunction_Sequence_1.D
   )
   > &RigUnit_OffsetTransformForItem(
    Item=(Type=Bone,Name="dyn_Hair_BackR_05")
    OffsetTransform=(Rotation=(X=0,Y=0,Z=0,W=1),Translation=(X=-0.5,Y=0,Z=0),Scale3D=(X=0.5,Y=1,Z=1))
    Weight=1
    bPropagateToChildren=True
   )
   > &RigUnit_OffsetTransformForItem(
    Item=(Type=Bone,Name="dyn_Hair_BackR_06")
    OffsetTransform=(Rotation=(X=0,Y=0,Z=0.1736,W=0.9848),Translation=(X=0,Y=0,Z=0),Scale3D=(X=1,Y=1,Z=1))
    Weight=1
    bPropagateToChildren=True
   )
   &RigVMFunction_MathQuaternionFromEuler(Euler=(X=0,Y=20,Z=10))
   &RigUnit_ModifyTransforms(
    ExecutePin=()
    ItemToModify=((Item=(Type=Bone,Name="dyn_Hair_BackL_04"),Transform=(Rotation=(X=0,Y=0,Z=0,W=1),Translation=(X=0,Y=0,Z=2),Scale3D=(X=1,Y=1,Z=1))))
    Weight=1
   )
   RerouteNode_2_1 = RigVMDispatchNode(Values=((Type=Bone,Name="dyn_Hair_Back_01"),((),Type=Bone,Name="dyn_Hair_Back_02"),((),Type=Bone,Name="dyn_Hair_Back_03"),((),Type=Bone,Name="dyn_Hair_Back_04"),((),Type=Bone,Name="dyn_Hair_Back_05"),((),Type=Bone,Name="dyn_Hair_Back_06"))) {template="DISPATCH_RigVMDispatch_ArrayMake(in Values,out Array)"}
   TwistBones_1_1 = RigVMFunctionReferenceNode(
    Axis to Twist=(bX=true,bY=false,bZ=false)
    Bones to Twist=((Type=Bone,Name="dyn_Hair_BackR_01"),(Type=Bone,Name="dyn_Hair_BackR_06"))
    End=(Type=Bone,Name="spine_05")
    ExecuteContext=()
    Start=(Type=Bone,Name="spine_02")
   )
  ]}
 RigVMModel Construction Event Graph {nodes [
   n1 = &RigUnit_PrepareForExecution()
   SequenceAggregate = RigVMAggregateNode(ExecuteContext=@n1.ExecutePin)
   RerouteNode_6_1 = RigVMDispatchNode(Values="((Type=Control,Name=\"dyn_Hair_Back_01_ctrl\"),((),Type=Control,Name=\"dyn_Hair_Back_02_ctrl\"),((),Type=Control,Name=\"dyn_Hair_Back_03_ctrl\"),((),Type=Control,Name=\"dyn_Hair_Back_04_ctrl\"),((),Type=Control,Name=\"dyn_Hair_Back_05_ctrl\"),((),Type=Control,Name=\" ...(truncated)") {template="DISPATCH_RigVMDispatch_ArrayMake(in Values,out Array)"}
   RigVMVariableNode(Value=@RerouteNode_6_1.Array, ExecuteContext=@SequenceAggregate.A) {variable=dyn_Hair_Back_ctrls}
   > RigVMVariableNode(Value=/Script/ControlRig.RigUnit_HierarchyGetPoseItemArray(ElementType=All, Initial=false, ItemsToGet=@RerouteNode_6_1.Array) {is=struct RigVMUnitNode}.Pose) {variable=dyn_Hair_Back_pose}
  ]}
]
hierarchy [
 root {loc=<0 0 0>, rot=<0 0 0 1>}
 pelvis {parent=root, loc=<0 -0.5301 98.0571>, rot=<0.0127 0.707 -0.0127 -0.707>}
 spine_01 {parent=pelvis, loc=<6.1208 -0.9328 0>, rot=<0 0 0.034 -0.9994>}
 spine_02 {parent=spine_01, loc=<6.5997 -0.2095 0>, rot=<0 0 0.0238 -0.9997>}
 spine_03 {parent=spine_02, loc=<6.4914 -0.1237 0>, rot=<0 0 0.0149 -0.9999>}
 spine_04 {parent=spine_03, loc=<6.9598 0.0794 0>, rot=<0 0 -0.0132 -0.9999>}
 spine_05 {parent=spine_04, loc=<6.9987 0 0>, rot=<0 0 -0.0015 -1>}
 neck_01 {parent=spine_05, loc=<11.3138 0.1529 0>, rot=<0 0 -0.2003 -0.9797>}
 neck_02 {parent=neck_01, loc=<4.5627 0 0>, rot=<0 0 0.0504 -0.9987>}
 head {parent=neck_02, loc=<5.3914 0 0>, rot=<0 0 0.115 -0.9934>}
 dyn_Hair_a {parent=head, loc=<6.5163 5.4354 -7.1396>, rot=<0 0 0.7097 -0.7045>}
 dyn_Hair_b01 {parent=head, loc=<9.1844 10.2015 0.4664>, rot=<0 0 0.7097 -0.7045>}
 dyn_Hair_b02 {parent=dyn_Hair_b01, loc=<-0.5056 -3.1477 -0.2904>, rot=<0 0 0 1>}
 dyn_Hair_c {parent=head, loc=<8.3766 8.015 5.7049>, rot=<0 0 0.7097 -0.7045>}
 dyn_Hair_d01 {parent=head, loc=<6.1331 4.6678 7.4092>, rot=<0 0 0.7097 -0.7045>}
 dyn_Hair_d02 {parent=dyn_Hair_d01, loc=<2.0726 -1.1216 1.4383>, rot=<0 0 0 1>}
 dyn_Hair_d03 {parent=dyn_Hair_d01, loc=<1.1598 -3.2554 0.639>, rot=<0 0 0 1>}
 dyn_Hair_d04 {parent=dyn_Hair_d03, loc=<1.2444 -3.3796 0>, rot=<0 0 0 1>}
 dyn_Hair_BackL_01 {parent=head, loc=<7.9722 -1.0353 -5.5286>, rot=<0 0 0.0928 -0.9957>}
 dyn_Hair_BackL_02 {parent=dyn_Hair_BackL_01, loc=<-1.9753 -3.7984 0>, rot=<0 0 0.5593 -0.8289>}
 dyn_Hair_BackL_03 {parent=dyn_Hair_BackL_02, loc=<-0.2757 -6.9116 0>, rot=<0 0 -0.004 -1>}
 dyn_Hair_BackL_04 {parent=dyn_Hair_BackL_03, loc=<1.6889 -8.9654 -2.335>, rot=<0 0 0 1>}
 dyn_Hair_BackL_05 {parent=dyn_Hair_BackL_04, loc=<1.8626 -7.0493 0>, rot=<0 0 0 1>}
 dyn_Hair_BackL_06 {parent=dyn_Hair_BackL_05, loc=<0.0921 -7.9976 -0.3361>, rot=<0 0 0.0617 -0.9981>}
 dyn_Hair_BackL_03a {parent=dyn_Hair_BackL_03, loc=<-2.3657 -3.361 -1.8709>, rot=<0 0 0 1>}
 dyn_Hair_BackL_03b {parent=dyn_Hair_BackL_03, loc=<2.2224 -3.1242 -1.714>, rot=<-0.2538 0.0111 0.1736 -0.9515>}
 dyn_Hair_Back_01 {parent=head, loc=<8.5016 -1.2653 -0.3258>, rot=<0 0 0.0928 -0.9957>}
 dyn_Hair_Back_02 {parent=dyn_Hair_Back_01, loc=<-1.9753 -3.7984 0>, rot=<0 0 0.5593 -0.8289>}
 dyn_Hair_Back_03 {parent=dyn_Hair_Back_02, loc=<-0.2757 -6.9116 0>, rot=<0 0 -0.004 -1>}
 dyn_Hair_Back_04 {parent=dyn_Hair_Back_03, loc=<2.1735 -8.8431 -0.3554>, rot=<0 0 0 1>}
 dyn_Hair_Back_05 {parent=dyn_Hair_Back_04, loc=<2.6338 -9.663 0>, rot=<0 0 0 1>}
 dyn_Hair_Back_06 {parent=dyn_Hair_Back_05, loc=<-0.9292 -7.1818 0>, rot=<0 0 0.0617 -0.9981>}
 dyn_Hair_BackR_01 {parent=head, loc=<8.1854 -0.439 5.0959>, rot=<0 0 0.0928 -0.9957>}
 dyn_Hair_BackR_02 {parent=dyn_Hair_BackR_01, loc=<-1.9753 -3.7984 0>, rot=<0 0 0.5593 -0.8289>}
 dyn_Hair_BackR_03 {parent=dyn_Hair_BackR_02, loc=<-0.2757 -6.9116 0>, rot=<0 0 -0.004 -1>}
 dyn_Hair_BackR_04 {parent=dyn_Hair_BackR_03, loc=<2.5072 -8.9279 0.6166>, rot=<0 0 0 1>}
 dyn_Hair_BackR_05 {parent=dyn_Hair_BackR_04, loc=<1.8626 -7.0493 0>, rot=<0 0 0 1>}
 dyn_Hair_BackR_06 {parent=dyn_Hair_BackR_05, loc=<0.1207 -7.3917 0>, rot=<0 0 0.0617 -0.9981>}
 dyn_Hair_BackR_03a {parent=dyn_Hair_BackR_03, loc=<-1.8255 -3.6914 2.3036>, rot=<0 0 0 1>}
 extra_cristal_head_l {parent=head, loc=<16.5815 1.5538 -12.7606>, rot=<0.0039 -0.9341 0.0015 0.357>}
 extra_cristal_head_r {parent=head, loc=<16.5817 1.5538 12.7606>, rot=<0.0039 -0.9341 -0.0015 -0.357>}
 extra_crystal_head_l {parent=head, loc=<16.5815 1.5538 -12.7606>, rot=<0.0039 -0.9341 0.0015 0.357>}
 extra_crystal_head_r {parent=head, loc=<16.5817 1.5538 12.7606>, rot=<0.0039 -0.9341 -0.0015 -0.357>}
 dyn_Hair_SideL_base {parent=head, loc=<8.4813 5.5447 -6.6454>, rot=<0.0097 -0.9725 -0.2153 -0.0879>}
 dyn_Hair_SideL_01 {parent=dyn_Hair_SideL_base, loc=<0.3929 0 0>, rot=<0 0 0 1>}
 dyn_Hair_SideL_02 {parent=dyn_Hair_SideL_01, loc=<3.5144 0 0>, rot=<-0.0003 -0.0964 0.0286 -0.9949>}
 dyn_Hair_SideL_03 {parent=dyn_Hair_SideL_02, loc=<3.6014 0 0>, rot=<0 -0.047 0 -0.9989>}
 dyn_Hair_C_base {parent=head, loc=<11.0754 9.8186 0.8023>, rot=<-0.116 -0.6975 -0.6975 -0.116>}
 dyn_Hair_C_01 {parent=dyn_Hair_C_base, loc=<0.3929 0 0>, rot=<0 0 0 1>}
 dyn_Hair_C_02 {parent=dyn_Hair_C_01, loc=<3.2012 0 0>, rot=<0 -0.1957 0 -0.9807>}
 dyn_Hair_SideR_base {parent=head, loc=<8.267 5.7505 7.0468>, rot=<0.0898 0.1897 0.9749 -0.074>}
 dyn_Hair_SideR_01 {parent=dyn_Hair_SideR_base, loc=<0.3929 0 0>, rot=<0 0 0 1>}
 dyn_Hair_SideR_02 {parent=dyn_Hair_SideR_01, loc=<3.5144 0 0>, rot=<0 -0.0869 0.0286 -0.9958>}
 dyn_Hair_SideR_03 {parent=dyn_Hair_SideR_02, loc=<3.6014 0 0>, rot=<0 0 0 1>}
 dyn_Hair_SideBackR_base {parent=head, loc=<7.4546 3.8651 8.3083>, rot=<0.0328 -0.1467 0.9649 -0.2154>}
 dyn_Hair_SideBackR {parent=dyn_Hair_SideBackR_base, loc=<0.3929 0 0>, rot=<0 0 0 1>}
 dyn_Hair_SideC_base {parent=head, loc=<7.8588 8.6124 -5.3201>, rot=<0.0121 -0.9131 -0.4039 -0.0543>}
 dyn_Hair_SideC {parent=dyn_Hair_SideC_base, loc=<0.3929 0 0>, rot=<0 0 0 1>}
 clavicle_l {parent=spine_05, loc=<5.4831 -0.1053 -3.2946>, rot=<0.0465 -0.7289 -0.009 -0.6829>}
 upperarm_l {parent=clavicle_l, loc=<9.8658 -1.2136 -0.4105>, rot=<0.0164 -0.3245 -0.0269 -0.9454>}
 lowerarm_l {parent=upperarm_l, loc=<26.465 -0.8345 0.259>, rot=<0.0036 0.0161 -0.2298 -0.9731>}
 clavicle_r {parent=spine_05, loc=<5.4831 -0.1053 3.2946>, rot=<0.7289 0.0465 -0.6829 0.009>}
 upperarm_r {parent=clavicle_r, loc=<-9.8658 1.2136 0.4105>, rot=<0.0164 -0.3245 -0.0269 -0.9454>}
 lowerarm_r {parent=upperarm_r, loc=<-26.465 0.8345 -0.259>, rot=<0.0036 0.0161 -0.2298 -0.9731>}
 dyn_Hair_l_base {parent=spine_05, loc=<12.2976 1.918 -6.3566>, rot=<-0.1982 -0.6971 -0.6685 -0.1669>}
 dyn_Hair_l_01 {parent=dyn_Hair_l_base, loc=<0.0989 0 0.0145>, rot=<0 0.0726 0 -0.9974>}
 dyn_Hair_l_02 {parent=dyn_Hair_l_01, loc=<5.1884 0 0>, rot=<0 -0.062 0 -0.9981>}
 dyn_Hair_l_03 {parent=dyn_Hair_l_02, loc=<-4.3079 -5.8053 -0.6797>, rot=<0 0 0 1>}
 dyn_Hair_l_04 {parent=dyn_Hair_l_03, loc=<-1.1163 -5.792 0.0902>, rot=<0 0 0 1>}
 dyn_Hair_r_base {parent=spine_05, loc=<12.2978 1.918 6.3566>, rot=<0.3264 0.6058 0.6846 0.2404>}
 dyn_Hair_r_01 {parent=dyn_Hair_r_base, loc=<0.1004 0 -0.0004>, rot=<0 -0.0726 0 -0.9974>}
 dyn_Hair_r_02 {parent=dyn_Hair_r_01, loc=<5.1884 0 0>, rot=<0 -0.062 0 -0.9981>}
 dyn_Hair_r_03 {parent=dyn_Hair_r_02, loc=<-4.3079 -5.8053 0.8949>, rot=<0 0 0 1>}
 dyn_Hair_r_04 {parent=dyn_Hair_r_03, loc=<-1.1163 -5.792 0.0902>, rot=<0 0 0 1>}
 spine_04_ctrl {controlType=EULER_TRANSFORM, parent=spine_04, type=Control, loc=<0 0 0>, rot=<0 0 0 1>}
 spine_03_ctrl {controlType=EULER_TRANSFORM, parent=spine_03, type=Control, loc=<0 0 0>, rot=<0 0 0 1>}
 spine_02_ctrl {controlType=EULER_TRANSFORM, parent=spine_02, type=Control, loc=<0 0 0>, rot=<0 0 0 1>}
 dyn_Hair_BackL_02_Null {parent=pelvis, type=Null, loc=<59.0305 -7.7237 -5.5286>, rot=<0 0 -0.6508 0.7593>}
 dyn_Hair_BackL_02_ctrl {controlType=EULER_TRANSFORM, parent=dyn_Hair_BackL_02_Null, type=Control, loc=<0 0 0>, rot=<0 0 0 1>}
 dyn_Hair_BackL_03_Null {parent=pelvis, type=Null, loc=<52.158 -8.5088 -5.5286>, rot=<0 0 0.6478 -0.7618>}
 dyn_Hair_BackL_03_ctrl {controlType=EULER_TRANSFORM, parent=dyn_Hair_BackL_03_Null, type=Control, loc=<0 0 0>, rot=<0 0 0 1>}
 dyn_Hair_BackL_04_Null {parent=pelvis, type=Null, loc=<43.581 -11.6175 -7.8636>, rot=<0 0 -0.6478 0.7618>}
 dyn_Hair_BackL_04_ctrl {controlType=EULER_TRANSFORM, parent=dyn_Hair_BackL_04_Null, type=Control, loc=<0 0 0>, rot=<0 0 0 1>}
 dyn_Hair_BackL_05_Null {parent=pelvis, type=Null, loc=<43.581 -11.6175 -7.8636>, rot=<0 0 -0.6478 0.7618>}
 dyn_Hair_BackL_05_ctrl {controlType=EULER_TRANSFORM, parent=dyn_Hair_BackL_05_Null, type=Control, loc=<0 0 0>, rot=<0 0 0 1>}
 dyn_Hair_BackL_06_Null {parent=pelvis, type=Null, loc=<43.581 -11.6175 -7.8636>, rot=<0 0 -0.6478 0.7618>}
 dyn_Hair_BackL_06_ctrl {controlType=EULER_TRANSFORM, parent=dyn_Hair_BackL_06_Null, type=Control, loc=<0 0 0>, rot=<0 0 0 1>}
 dyn_Hair_BackL_01_Null {parent=pelvis, type=Null, loc=<61.8198 -4.4758 -5.5286>, rot=<0 0 0.1148 -0.9934>}
 dyn_Hair_BackL_01_ctrl {controlType=EULER_TRANSFORM, parent=dyn_Hair_BackL_01_Null, type=Control, loc=<0 0 0>, rot=<0 0 0 1>}
 dyn_Hair_Back_01_Null {parent=pelvis, type=Null, loc=<62.3385 -4.7288 -0.3258>, rot=<0 0 0.1148 -0.9934>}
 dyn_Hair_Back_01_ctrl {controlType=EULER_TRANSFORM, parent=dyn_Hair_Back_01_Null, type=Control, loc=<0 0 0>, rot=<0 0 0 1>}
 dyn_Hair_Back_02_Null {parent=pelvis, type=Null, loc=<59.5492 -7.9768 -0.3258>, rot=<0 0 -0.6508 0.7593>}
 dyn_Hair_Back_02_ctrl {controlType=EULER_TRANSFORM, parent=dyn_Hair_Back_02_Null, type=Control, loc=<0 0 0>, rot=<0 0 0 1>}
 dyn_Hair_Back_03_Null {parent=pelvis, type=Null, loc=<52.6768 -8.7618 -0.3258>, rot=<0 0 0.6478 -0.7618>}
 dyn_Hair_Back_03_ctrl {controlType=EULER_TRANSFORM, parent=dyn_Hair_Back_03_Null, type=Control, loc=<0 0 0>, rot=<0 0 0 1>}
 dyn_Hair_Back_04_Null {parent=pelvis, type=Null, loc=<44.2983 -12.3291 -0.6812>, rot=<0 0 -0.6478 0.7618>}
 dyn_Hair_Back_04_ctrl {controlType=EULER_TRANSFORM, parent=dyn_Hair_Back_04_Null, type=Control, loc=<0 0 0>, rot=<0 0 0 1>}
 dyn_Hair_Back_05_Null {parent=pelvis, type=Null, loc=<35.1845 -16.4827 -0.6812>, rot=<0 0 0.6478 -0.7618>}
 dyn_Hair_Back_05_ctrl {controlType=EULER_TRANSFORM, parent=dyn_Hair_Back_05_Null, type=Control, loc=<0 0 0>, rot=<0 0 0 1>}
 dyn_Hair_Back_06_Null {parent=pelvis, type=Null, loc=<27.9468 -16.7205 -0.6812>, rot=<0 0 -0.6935 0.7204>}
 dyn_Hair_Back_06_ctrl {controlType=EULER_TRANSFORM, parent=dyn_Hair_Back_06_Null, type=Control, loc=<0 0 0>, rot=<0 0 0 1>}
 dyn_Hair_BackR_01_Null {parent=pelvis, type=Null, loc=<62.0591 -3.8895 5.0959>, rot=<0 0 0.1148 -0.9934>}
 dyn_Hair_BackR_01_ctrl {controlType=EULER_TRANSFORM, parent=dyn_Hair_BackR_01_Null, type=Control, loc=<0 0 0>, rot=<0 0 0 1>}
 dyn_Hair_BackR_02_Null {parent=pelvis, type=Null, loc=<59.2698 -7.1374 5.0959>, rot=<0 0 -0.6508 0.7593>}
 dyn_Hair_BackR_02_ctrl {controlType=EULER_TRANSFORM, parent=dyn_Hair_BackR_02_Null, type=Control, loc=<0 0 0>, rot=<0 0 0 1>}
 dyn_Hair_BackR_03_Null {parent=pelvis, type=Null, loc=<52.3973 -7.9224 5.0959>, rot=<0 0 0.6478 -0.7618>}
 dyn_Hair_BackR_03_ctrl {controlType=EULER_TRANSFORM, parent=dyn_Hair_BackR_03_Null, type=Control, loc=<0 0 0>, rot=<0 0 0 1>}
 dyn_Hair_BackR_04_Null {parent=pelvis, type=Null, loc=<43.9889 -11.8327 5.7125>, rot=<0 0 -0.6478 0.7618>}
 dyn_Hair_BackR_04_ctrl {controlType=EULER_TRANSFORM, parent=dyn_Hair_BackR_04_Null, type=Control, loc=<0 0 0>, rot=<0 0 0 1>}
 dyn_Hair_BackR_05_Null {parent=pelvis, type=Null, loc=<37.3308 -14.8048 5.7125>, rot=<0 0 0.6478 -0.7618>}
 dyn_Hair_BackR_05_ctrl {controlType=EULER_TRANSFORM, parent=dyn_Hair_BackR_05_Null, type=Control, loc=<0 0 0>, rot=<0 0 0 1>}
 dyn_Hair_BackR_06_Null {parent=pelvis, type=Null, loc=<30.0547 -16.1126 5.7125>, rot=<0 0 -0.6935 0.7204>}
 dyn_Hair_BackR_06_ctrl {controlType=EULER_TRANSFORM, parent=dyn_Hair_BackR_06_Null, type=Control, loc=<0 0 0>, rot=<0 0 0 1>}
]
variables [
 Cape {type=bool, default=False}
 NewVar {type=bool, default=False}
 SPR_HeadUp {type=double, default=0}
 dyn_Hair_BackL_ctrls {
  type="TArray<FRigElementKey>"
  default=((Type=Control,Name="dyn_Hair_BackL_02"),(Type=Control,Name="dyn_Hair_BackL_03"),(Type=Control,Name="dyn_Hair_BackL_04"),(Type=Control,Name="dyn_Hair_BackL_05"),(Type=Control,Name="dyn_Hair_BackL_06"))
 }
 dyn_Hair_BackL_pose {type=FRigPose, default=(Elements=,HierarchyTopologyVersion=-1,PoseHash=-1)}
 dyn_Hair_BackR_ctrls {
  type="TArray<FRigElementKey>"
  default=((Type=Control,Name="dfrm_arm_spline_01_r_ctrl"),(Type=Control,Name="dfrm_arm_spline_03_r_ctrl"),(Type=Control,Name="dfrm_arm_spline_05_r_ctrl"),(Type=Control,Name="dfrm_arm_spline_07_r_ctrl"),(Type=Control,Name="dfrm_arm_spline_09_r_ctrl"))
 }
 dyn_Hair_BackR_pose {type=FRigPose, default=(Elements=,HierarchyTopologyVersion=-1,PoseHash=-1)}
 dyn_Hair_Back_ctrls {
  type="TArray<FRigElementKey>"
  default=((Type=Control,Name="dyn_Hair_BackL_02"),(Type=Control,Name="dyn_Hair_BackL_03"),(Type=Control,Name="dyn_Hair_BackL_04"),(Type=Control,Name="dyn_Hair_BackL_05"),(Type=Control,Name="dyn_Hair_BackL_06"))
 }
 dyn_Hair_Back_pose {type=FRigPose, default=(Elements=,HierarchyTopologyVersion=-1,PoseHash=-1)}
]
