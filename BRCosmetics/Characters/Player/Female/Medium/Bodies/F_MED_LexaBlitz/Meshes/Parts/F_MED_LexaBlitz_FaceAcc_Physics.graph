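# Physics asset for the LexaBlitz female character's FaceAcc mesh part, defining bodies and joints for dynamic hair.
# Short capsule chains for center, side and back-right hair hang from the head, while longer, thicker left and right chains anchor to spine_05.
# Hair joints use tight swing and twist limits, while head and spine joints rotate freely.
# Larger capsules cover head, upper spine and arms, with the arm bodies left unconstrained.
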PhysicsAsset /BRCosmetics/Characters/Player/Female/Medium/Bodies/F_MED_LexaBlitz/Meshes/Parts/F_MED_LexaBlitz_FaceAcc_Physics
bodies {
 dyn_Hair_C_01 [{shapeType=Capsule, length=1, radius=1.5, shapeName=dyn_Hair_C_01_capsule, center=<2 0 0>, rotation=<rot 90 0 0>}]
 dyn_Hair_C_02 [{shapeType=Capsule, length=1, radius=1.5, shapeName=dyn_Hair_C_02_capsule, center=<2 0 0>, rotation=<rot 90 0 0>}]
 dyn_Hair_C_base [{shapeType=Capsule, length=0.505, radius=0.505, shapeName=dyn_Hair_C_base_capsule, rotation=<rot 0 -90 23.3593>}]
 dyn_Hair_SideBackR [{shapeType=Capsule, length=1, radius=1, shapeName=dyn_Hair_SideBackR_capsule, center=<2 0 0>, rotation=<rot 90 0 0>}]
 dyn_Hair_SideBackR_base [{shapeType=Capsule, length=0.505, radius=0.505, shapeName=dyn_Hair_SideBackR_base_capsule, rotation=<rot 0 -90.0001 23.3595>}]
 dyn_Hair_SideC [{shapeType=Capsule, length=1, radius=1.5, shapeName=dyn_Hair_SideC_capsule, center=<2 0 0>, rotation=<rot 90 0 0>}]
 dyn_Hair_SideC_base [{shapeType=Capsule, length=0.505, radius=0.505, shapeName=dyn_Hair_SideC_base_capsule, rotation=<rot 0 -89.9999 23.3594>}]
 dyn_Hair_SideL_01 [{shapeType=Capsule, length=1, radius=1, shapeName=dyn_Hair_SideL_01_capsule, center=<2 0 0>, rotation=<rot 90 0 0>}]
 dyn_Hair_SideL_02 [{shapeType=Capsule, length=1, radius=1, shapeName=dyn_Hair_SideL_02_capsule, center=<2 0 0>, rotation=<rot 90 0 0>}]
 dyn_Hair_SideL_03 [{shapeType=Capsule, length=1, radius=1, shapeName=dyn_Hair_SideL_03_capsule, center=<2 0 0>, rotation=<rot 90 0 0>}]
 dyn_Hair_SideL_base [{shapeType=Capsule, length=0.505, radius=0.505, shapeName=dyn_Hair_SideL_base_capsule, rotation=<rot 0 -89.9999 23.3594>}]
 dyn_Hair_SideR_01 [{shapeType=Capsule, length=1, radius=1, shapeName=dyn_Hair_SideR_01_capsule, center=<2 0 0>, rotation=<rot 90 0 0>}]
 dyn_Hair_SideR_02 [{shapeType=Capsule, length=1, radius=1, shapeName=dyn_Hair_SideR_02_capsule, center=<2 0 0>, rotation=<rot 90 0 0>}]
 dyn_Hair_SideR_03 [{shapeType=Capsule, length=1, radius=1, shapeName=dyn_Hair_SideR_03_capsule, center=<2 0 0>, rotation=<rot 90 0 0>}]
 dyn_Hair_SideR_base [{shapeType=Capsule, length=0.505, radius=0.505, shapeName=dyn_Hair_SideR_base_capsule, rotation=<rot 0 -90.0001 23.3594>}]
 dyn_Hair_l_01 [{shapeType=Capsule, length=3, radius=2.5, shapeName=dyn_Hair_l_01_capsule, center=<4 0 0>, rotation=<rot 90 0 0>}]
 dyn_Hair_l_02 [{shapeType=Capsule, length=4, radius=2.5, shapeName=dyn_Hair_l_02_capsule, center=<4 0 0>, rotation=<rot 90 0 0>}]
 dyn_Hair_l_base [{shapeType=Capsule, length=1, radius=1, shapeName=dyn_Hair_l_base_capsule}]
 dyn_Hair_r_01 [{shapeType=Capsule, length=3, radius=2.5, shapeName=dyn_Hair_r_01_capsule, center=<4 0 0>, rotation=<rot 90 0 0>}]
 dyn_Hair_r_02 [{shapeType=Capsule, length=4, radius=2.5, shapeName=dyn_Hair_r_02_capsule, center=<4 0 0>, rotation=<rot 90 0 0>}]
 dyn_Hair_r_base [{shapeType=Capsule, length=1, radius=1, shapeName=dyn_Hair_r_base_capsule}]
 head [{shapeType=Capsule, length=7, radius=7, shapeName=head_capsule, center=<4.0618 2.4309 0>, rotation=<rot 90 0 0>}]
 lowerarm_l [{
   shapeType=Capsule
   length=21.5
   radius=8.5
   shapeName=lowerarm_l_capsule
   center=<15.4871 0.0001 -2.7308>
   rotation=<rot 80.0001 -0.0008 -0.0014>
  }]
 lowerarm_r [{shapeType=Capsule, length=21.5, radius=8.5, shapeName=lowerarm_r_capsule, center=<-14.2117 0 2.5059>, rotation=<rot 80.0001 -0.0008 -0.001>}]
 spine_04 [{shapeType=Capsule, length=19.5, radius=10.5, shapeName=spine_04_capsule, center=<-0.5867 0.561 0>}]
 spine_05 [{shapeType=Capsule, length=21, radius=9, shapeName=spine_05_capsule, center=<-0.3981 -2.0489 0>}]
 upperarm_l [{shapeType=Capsule, length=25.75, radius=4.755, shapeName=upperarm_l_capsule, center=<10.5007 0 0>, rotation=<rot -0.0001 -89.9999 90>}]
 upperarm_r [{shapeType=Capsule, length=27.25, radius=5.255, shapeName=upperarm_r_capsule, center=<-11.1137 0 0>, rotation=<rot -0.0002 -89.9998 90>}]
}
constraints [
 {
  bone1Name=spine_05
  bone2Name=spine_04
  swing1LimitDegrees=45
  swing1Motion=Free
  swing2LimitDegrees=45
  swing2Motion=Free
  twistLimitDegrees=45
  twistMotion=Free
 }
 {bone1Name=dyn_Hair_l_base, bone2Name=spine_05, swing1LimitDegrees=15, swing2LimitDegrees=15, twistLimitDegrees=15}
 {bone1Name=dyn_Hair_l_01, bone2Name=dyn_Hair_l_base, swing1LimitDegrees=25, swing2LimitDegrees=25, twistLimitDegrees=10}
 {bone1Name=dyn_Hair_l_02, bone2Name=dyn_Hair_l_01, swing1LimitDegrees=25, swing2LimitDegrees=25, twistLimitDegrees=10}
 {bone1Name=dyn_Hair_r_base, bone2Name=spine_05, swing1LimitDegrees=15, swing2LimitDegrees=15, twistLimitDegrees=15}
 {bone1Name=dyn_Hair_r_01, bone2Name=dyn_Hair_r_base, swing1LimitDegrees=25, swing2LimitDegrees=25, twistLimitDegrees=10}
 {bone1Name=dyn_Hair_r_02, bone2Name=dyn_Hair_r_01, swing1LimitDegrees=25, swing2LimitDegrees=25, twistLimitDegrees=10}
 {bone1Name=dyn_Hair_SideR_03, bone2Name=dyn_Hair_SideR_02, swing1LimitDegrees=15, swing2LimitDegrees=30, twistLimitDegrees=5}
 {bone1Name=dyn_Hair_C_base, bone2Name=head, swing1LimitDegrees=15, swing2LimitDegrees=30, twistLimitDegrees=5}
 {bone1Name=dyn_Hair_C_01, bone2Name=dyn_Hair_C_base, swing1LimitDegrees=15, swing2LimitDegrees=30, twistLimitDegrees=5}
 {bone1Name=dyn_Hair_C_02, bone2Name=dyn_Hair_C_01, swing1LimitDegrees=15, swing2LimitDegrees=30, twistLimitDegrees=5}
 {bone1Name=dyn_Hair_SideBackR_base, bone2Name=head, swing1LimitDegrees=15, swing2LimitDegrees=30, twistLimitDegrees=5}
 {bone1Name=dyn_Hair_SideBackR, bone2Name=dyn_Hair_SideBackR_base, swing1LimitDegrees=15, swing2LimitDegrees=30, twistLimitDegrees=5}
 {bone1Name=dyn_Hair_SideC_base, bone2Name=head, swing1LimitDegrees=15, swing2LimitDegrees=30, twistLimitDegrees=5}
 {bone1Name=dyn_Hair_SideC, bone2Name=dyn_Hair_SideC_base, swing1LimitDegrees=15, swing2LimitDegrees=30, twistLimitDegrees=5}
 {bone1Name=dyn_Hair_SideL_base, bone2Name=head, swing1LimitDegrees=10, swing2LimitDegrees=15, twistLimitDegrees=5}
 {bone1Name=dyn_Hair_SideL_01, bone2Name=dyn_Hair_SideL_base, swing1LimitDegrees=10, swing2LimitDegrees=15, twistLimitDegrees=5}
 {bone1Name=dyn_Hair_SideL_02, bone2Name=dyn_Hair_SideL_01, swing1LimitDegrees=10, swing2LimitDegrees=15, twistLimitDegrees=5}
 {bone1Name=dyn_Hair_SideL_03, bone2Name=dyn_Hair_SideL_02, swing1LimitDegrees=10, swing2LimitDegrees=15, twistLimitDegrees=5}
 {bone1Name=dyn_Hair_SideR_base, bone2Name=head, swing1LimitDegrees=15, swing2LimitDegrees=30, twistLimitDegrees=5}
 {bone1Name=dyn_Hair_SideR_01, bone2Name=dyn_Hair_SideR_base, swing1LimitDegrees=10, swing2LimitDegrees=15, twistLimitDegrees=5}
 {bone1Name=dyn_Hair_SideR_02, bone2Name=dyn_Hair_SideR_01, swing1LimitDegrees=15, swing2LimitDegrees=30, twistLimitDegrees=5}
 {
  bone1Name=head
  bone2Name=spine_05
  swing1LimitDegrees=45
  swing1Motion=Free
  swing2LimitDegrees=45
  swing2Motion=Free
  twistLimitDegrees=45
  twistMotion=Free
 }
]
