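# Post-process-style AnimBP for Lexa Blitz's face-accessory mesh, animating her hair over the copied body pose.
# A Control Rig pose feeds back-hair AnimDynamics, blended back toward the rig pose for backbling, DBNO and skydiving.
# Front-hair AnimDynamics follows, bypassed when dynamics are disabled, then RigidBody and bone tweaks for capes, skydiving and front end.
# The event graph detects backbling and computes state-based gravity and crystal wave vectors.
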AnimBlueprint /BRCosmetics/Characters/Player/Female/Medium/Bodies/F_MED_LexaBlitz/Meshes/Parts/F_MED_LexaBlitz_FaceAcc_AnimBP
target_skeleton=&/F_MED_LexaBlitz_FaceAcc_Skeleton
graphs [
 AnimGraph {nodes [
   n1 = Anim Dynamics - Bone: dyn_Hair_b01(Alpha=1, ExternalForce="0,0,0") {
    angularDampingOverride=0.99
    angularSpringConstant=75
    bAngularSpring=true
    bGravityOverrideInSimSpace=true
    bOverrideAngularDamping=true
    bUseGravityOverride=true
    simulationSpace=RootRelative
    alphaBoolBlend {blendOption=Linear}
    boundBone {boneName=dyn_Hair_b01}
    componentAppliedLinearAccClamp=<3500 3500 2500>
    componentLinearAccScale=<0.5 0.5 0.35>
    componentLinearVelScale=<0.55 0.55 0.4>
    gravityOverride {z=-980}
    physicsBodyDefinitions [{
      collisionType=CustomSphere
      sphereCollisionRadius=2
      boundBone {boneName=dyn_Hair_b01}
      boxExtents=<23 23 23>
      constraintSetup {angularLimitsMax=<7 2 5>, angularLimitsMin=<-7 -2 -15>, angularTarget=<0 0 0>}
      localJointOffset=<0 -5 0>
     }]
    relativeSpaceBone {boneName=root}
    sphericalLimits [
     {limitRadius=18, drivingBone {boneName=upperarm_r}, sphereLocalOffset=<-10 -12 0>}
     {limitRadius=0, drivingBone {boneName=None}, sphereLocalOffset=<0 0 0>}
    ]
   }
   Anim Dynamics - Bone: dyn_Hair_b02(ComponentPose=@n1, Alpha=1, ExternalForce="0,0,0") {
    angularDampingOverride=0.99
    angularSpringConstant=75
    bAngularSpring=true
    bGravityOverrideInSimSpace=true
    bOverrideAngularDamping=true
    bUseGravityOverride=true
    simulationSpace=RootRelative
    alphaBoolBlend {blendOption=Linear}
    boundBone {boneName=dyn_Hair_b02}
    componentAppliedLinearAccClamp=<3500 3500 2500>
    componentLinearAccScale=<0.5 0.5 0.35>
    componentLinearVelScale=<0.55 0.55 0.4>
    gravityOverride {z=-980}
    physicsBodyDefinitions [{
      collisionType=CustomSphere
      sphereCollisionRadius=2
      boundBone {boneName=dyn_Hair_b02}
      boxExtents=<23 23 23>
      constraintSetup {angularLimitsMax=<7 2 5>, angularLimitsMin=<-7 -2 -15>, angularTarget=<0 0 0>}
      localJointOffset=<0 -5 0>
     }]
    relativeSpaceBone {boneName=root}
    sphericalLimits [
     {limitRadius=18, drivingBone {boneName=upperarm_r}, sphereLocalOffset=<-10 -12 0>}
     {limitRadius=0, drivingBone {boneName=None}, sphereLocalOffset=<0 0 0>}
    ]
   }
   Copy Pose From Mesh() {bUseAttachedParent=true}
   n2 = Get isWearingPackBool() -> (isWearingPackBool:bool)
   n3 = Get bIsDBNO()
   n4 = Two Way Blend(
    A=Use cached pose 'HairDynPose'()
    B=Use cached pose 'HairCachePose'()
    Alpha=SelectFloat(A=1, B=0.2, bPickA=OR Boolean!(A=@n2, B=@n3, C:bool) -> (ReturnValue))
   ) {blendNode {alphaBoolBlend {blendOption=Linear}}}
   n5 = Use cached pose 'HairDynPose'()
   n6 = OR Boolean!(A=Get bIsSkydiveDiveMode(), B=Get bIsSkydiveFloating(), C=Get isWearingPackBool(), D:bool=Get bIsDBNO()) -> (ReturnValue)
   n7 = Anim Dynamics - Bone: dyn_Hair_a(
    ComponentPose=Local To Component(LocalPose=Blend Poses by bool(BlendPose_0=@n4, BlendPose_1=@n5, bActiveValue=@n6) {blendType=Linear})
    Alpha=1
    ExternalForce="0,0,0"
   ) {
    angularDampingOverride=0.99
    angularSpringConstant=75
    bAngularSpring=true
    bGravityOverrideInSimSpace=true
    bOverrideAngularDamping=true
    bUseGravityOverride=true
    simulationSpace=RootRelative
    alphaBoolBlend {blendOption=Linear}
    boundBone {boneName=dyn_Hair_a}
    componentAppliedLinearAccClamp=<3500 3500 2500>
    componentLinearAccScale=<0.5 0.5 0.35>
    componentLinearVelScale=<0.55 0.55 0.4>
    gravityOverride {z=-980}
    physicsBodyDefinitions [{
      collisionType=CustomSphere
      sphereCollisionRadius=2
      boundBone {boneName=dyn_Hair_a}
      boxExtents=<23 23 23>
      constraintSetup {angularLimitsMax=<1 2 10>, angularLimitsMin=<-4 -2 -25>, angularTarget=<0 0 0>}
      localJointOffset=<0 -5 0>
     }]
    relativeSpaceBone {boneName=root}
    sphericalLimits [
     {limitRadius=18, drivingBone {boneName=upperarm_r}, sphereLocalOffset=<-10 -12 0>}
     {limitRadius=0, drivingBone {boneName=None}, sphereLocalOffset=<0 0 0>}
    ]
   }
   n8 = Anim Dynamics - Bone: dyn_Hair_b01(ComponentPose=@n7, Alpha=1, ExternalForce="0,0,0") {
    angularDampingOverride=0.99
    angularSpringConstant=75
    bAngularSpring=true
    bGravityOverrideInSimSpace=true
    bOverrideAngularDamping=true
    bUseGravityOverride=true
    simulationSpace=RootRelative
    alphaBoolBlend {blendOption=Linear}
    boundBone {boneName=dyn_Hair_b01}
    componentAppliedLinearAccClamp=<3500 3500 2500>
    componentLinearAccScale=<0.5 0.5 0.35>
    componentLinearVelScale=<0.55 0.55 0.4>
    gravityOverride {z=-980}
    physicsBodyDefinitions [{
      collisionType=CustomSphere
      sphereCollisionRadius=2
      boundBone {boneName=dyn_Hair_b01}
      boxExtents=<23 23 23>
      constraintSetup {angularLimitsMax=<7 2 5>, angularLimitsMin=<-7 -2 -15>, angularTarget=<0 0 0>}
      localJointOffset=<0 -5 0>
     }]
    relativeSpaceBone {boneName=root}
    sphericalLimits [
     {limitRadius=5, drivingBone {boneName=head}, sphereLocalOffset=<0 0 0>}
     {limitRadius=0, drivingBone {boneName=None}, sphereLocalOffset=<0 0 0>}
    ]
   }
   n9 = Anim Dynamics - Bone: dyn_Hair_b02(ComponentPose=@n8, Alpha=1, ExternalForce="0,0,0") {
    angularDampingOverride=0.99
    angularSpringConstant=75
    bAngularSpring=true
    bGravityOverrideInSimSpace=true
    bOverrideAngularDamping=true
    bUseGravityOverride=true
    simulationSpace=RootRelative
    alphaBoolBlend {blendOption=Linear}
    boundBone {boneName=dyn_Hair_b02}
    componentAppliedLinearAccClamp=<3500 3500 2500>
    componentLinearAccScale=<0.5 0.5 0.35>
    componentLinearVelScale=<0.55 0.55 0.4>
    gravityOverride {z=-980}
    physicsBodyDefinitions [{
      collisionType=CustomSphere
      sphereCollisionRadius=2
      boundBone {boneName=dyn_Hair_b02}
      boxExtents=<23 23 23>
      constraintSetup {angularLimitsMax=<7 2 5>, angularLimitsMin=<-7 -2 -15>, angularTarget=<0 0 0>}
      localJointOffset=<0 -5 0>
     }]
    relativeSpaceBone {boneName=root}
    sphericalLimits [
     {limitRadius=18, drivingBone {boneName=upperarm_r}, sphereLocalOffset=<-10 -12 0>}
     {limitRadius=0, drivingBone {boneName=None}, sphereLocalOffset=<0 0 0>}
    ]
   }
   n10 = Anim Dynamics - Bone: dyn_Hair_c(ComponentPose=@n9, Alpha=1, ExternalForce="0,0,0") {
    angularDampingOverride=0.99
    angularSpringConstant=75
    bAngularSpring=true
    bGravityOverrideInSimSpace=true
    bOverrideAngularDamping=true
    bUseGravityOverride=true
    simulationSpace=RootRelative
    alphaBoolBlend {blendOption=Linear}
    boundBone {boneName=dyn_Hair_c}
    componentAppliedLinearAccClamp=<3500 3500 2500>
    componentLinearAccScale=<0.5 0.5 0.35>
    componentLinearVelScale=<0.55 0.55 0.4>
    gravityOverride {z=-980}
    physicsBodyDefinitions [{
      collisionType=CustomSphere
      sphereCollisionRadius=2
      boundBone {boneName=dyn_Hair_c}
      boxExtents=<23 23 23>
      constraintSetup {angularLimitsMax=<1 2 10>, angularLimitsMin=<-4 -2 -25>, angularTarget=<0 0 0>}
      localJointOffset=<0 -5 0>
     }]
    relativeSpaceBone {boneName=root}
    sphericalLimits [
     {limitRadius=18, drivingBone {boneName=upperarm_r}, sphereLocalOffset=<-10 -12 0>}
     {limitRadius=0, drivingBone {boneName=None}, sphereLocalOffset=<0 0 0>}
    ]
   }
   n11 = Anim Dynamics - Bone: dyn_Hair_d01(ComponentPose=@n10, Alpha=1, ExternalForce="0,0,0") {
    angularDampingOverride=0.99
    angularSpringConstant=75
    bAngularSpring=true
    bGravityOverrideInSimSpace=true
    bOverrideAngularDamping=true
    bUseGravityOverride=true
    simulationSpace=RootRelative
    alphaBoolBlend {blendOption=Linear}
    boundBone {boneName=dyn_Hair_d01}
    componentAppliedLinearAccClamp=<3500 3500 2500>
    componentLinearAccScale=<0.5 0.5 0.35>
    componentLinearVelScale=<0.55 0.55 0.4>
    gravityOverride {z=-980}
    physicsBodyDefinitions [{
      collisionType=CustomSphere
      sphereCollisionRadius=2
      boundBone {boneName=dyn_Hair_d01}
      boxExtents=<25 25 25>
      constraintSetup {twistAxis=AxisZ, angularLimitsMax=<1 0 5>, angularLimitsMin=<-7 0 -20>, angularTarget=<0 0 0>}
      localJointOffset=<0 -5 0>
     }]
    relativeSpaceBone {boneName=root}
    sphericalLimits [
     {limitRadius=18, drivingBone {boneName=upperarm_r}, sphereLocalOffset=<-10 -12 0>}
     {limitRadius=0, drivingBone {boneName=None}, sphereLocalOffset=<0 0 0>}
    ]
   }
   n12 = Anim Dynamics - Bone: dyn_Hair_d02(ComponentPose=@n11, Alpha=1, ExternalForce="0,0,0") {
    angularDampingOverride=0.99
    angularSpringConstant=75
    bAngularSpring=true
    bGravityOverrideInSimSpace=true
    bOverrideAngularDamping=true
    bUseGravityOverride=true
    simulationSpace=RootRelative
    alphaBoolBlend {blendOption=Linear}
    boundBone {boneName=dyn_Hair_d02}
    componentAppliedLinearAccClamp=<3500 3500 2500>
    componentLinearAccScale=<0.5 0.5 0.35>
    componentLinearVelScale=<0.55 0.55 0.4>
    gravityOverride {z=-980}
    physicsBodyDefinitions [{
      collisionType=CustomSphere
      sphereCollisionRadius=2
      boundBone {boneName=dyn_Hair_d02}
      boxExtents=<20 20 20>
      constraintSetup {twistAxis=AxisZ, angularLimitsMax=<1 0 5>, angularLimitsMin=<-7 0 -30>, angularTarget=<0 0 0>}
      localJointOffset=<0.1 -5 0.1>
     }]
    relativeSpaceBone {boneName=root}
    sphericalLimits [
     {limitRadius=18, drivingBone {boneName=upperarm_r}, sphereLocalOffset=<-10 -12 0>}
     {limitRadius=0, drivingBone {boneName=None}, sphereLocalOffset=<0 0 0>}
    ]
   }
   n13 = Anim Dynamics - Bone: dyn_Hair_d03(ComponentPose=@n12, Alpha=1, ExternalForce="0,0,0") {
    angularDampingOverride=0.99
    angularSpringConstant=75
    bAngularSpring=true
    bGravityOverrideInSimSpace=true
    bOverrideAngularDamping=true
    bUseGravityOverride=true
    simulationSpace=RootRelative
    alphaBoolBlend {blendOption=Linear}
    boundBone {boneName=dyn_Hair_d03}
    componentAppliedLinearAccClamp=<3500 3500 2500>
    componentLinearAccScale=<0.5 0.5 0.35>
    componentLinearVelScale=<0.55 0.55 0.4>
    gravityOverride {z=-980}
    physicsBodyDefinitions [{
      collisionType=CustomSphere
      sphereCollisionRadius=2
      boundBone {boneName=dyn_Hair_d03}
      boxExtents=<23 23 23>
      constraintSetup {twistAxis=AxisZ, angularLimitsMax=<1 0 5>, angularLimitsMin=<-7 0 -20>, angularTarget=<0 0 0>}
      localJointOffset=<0 -5 0>
     }]
    relativeSpaceBone {boneName=root}
    sphericalLimits [
     {limitRadius=18, drivingBone {boneName=upperarm_r}, sphereLocalOffset=<-10 -12 0>}
     {limitRadius=0, drivingBone {boneName=None}, sphereLocalOffset=<0 0 0>}
    ]
   }
   n14 = Anim Dynamics - Bone: dyn_Hair_d04(ComponentPose=@n13, Alpha=1, ExternalForce="0,0,0") {
    angularDampingOverride=0.99
    angularSpringConstant=75
    bAngularSpring=true
    bGravityOverrideInSimSpace=true
    bOverrideAngularDamping=true
    bUseGravityOverride=true
    simulationSpace=RootRelative
    alphaBoolBlend {blendOption=Linear}
    boundBone {boneName=dyn_Hair_d04}
    componentAppliedLinearAccClamp=<3500 3500 2500>
    componentLinearAccScale=<0.5 0.5 0.35>
    componentLinearVelScale=<0.55 0.55 0.4>
    gravityOverride {z=-980}
    physicsBodyDefinitions [{
      collisionType=CustomSphere
      sphereCollisionRadius=2
      boundBone {boneName=dyn_Hair_d04}
      boxExtents=<15 15 15>
      constraintSetup {twistAxis=AxisZ, angularLimitsMax=<1 0 5>, angularLimitsMin=<-7 0 -20>, angularTarget=<0 0 0>}
      localJointOffset=<0 -5 0>
     }]
    relativeSpaceBone {boneName=root}
    sphericalLimits [
     {limitRadius=18, drivingBone {boneName=upperarm_r}, sphereLocalOffset=<-10 -12 0>}
     {limitRadius=0, drivingBone {boneName=None}, sphereLocalOffset=<0 0 0>}
    ]
   }
   n15 = Component To Local(ComponentPose=@n14)
   n16 = Use cached pose 'HairCachePose'()
   n17 = Get bAnimDynamicsEnabled() -> (bAnimDynamicsEnabled:bool)
   n18 = RigidBody!(
    ComponentPose=Local To Component(LocalPose=Blend Poses by bool(BlendPose_0=@n15, BlendPose_1=@n16, bActiveValue=@n17) {blendType=Linear})
    bAlphaBoolEnabled:bool
    OverrideWorldGravity:Vector="0,0,0"
    ComponentLinearAccScale:Vector="0,0,0"
    ComponentLinearVelScale:Vector="0,0,0"
    bTransferBoneVelocities:bool=False
   ) -> (Pose) {
    alphaInputType=Bool
    bClampLinearTranslationLimitToRefPose=true
    bFreezeIncomingPoseOnStart=true
    bOverrideWorldGravity=true
    bTransferBoneVelocities=true
    overridePhysicsAsset=&/F_MED_LexaBlitz_FaceAcc_Physics
    simulationSpace=BaseBoneSpace
    alphaBoolBlend {blendOption=Linear}
    baseBoneRef {boneName=spine_05}
    componentAppliedLinearAccClamp=<100 100 100>
    componentLinearAccScale=<0.6 0.2 0.2>
    componentLinearVelScale=<0.6 0.2 0.2>
    simSpaceSettings {
     dampingAlpha=0
     maxAngularAcceleration=5000
     maxAngularVelocity=8
     maxLinearAcceleration=1000
     maxLinearVelocity=500
     worldAlpha=1
     externalLinearDragV=<1 1 1>
    }
   }
   n19 = Get is_WearingCape() -> (is_WearingCape:bool)
   n20 = "Transform (Modify) Bone - Bone: dyn_Hair_Back_04"!(ComponentPose=@n18, bAlphaBoolEnabled:bool=@n19, Translation="0,0,0", Rotation="0,0,0", Scale="0.7,1,1") -> (Pose) {
    alphaInputType=Bool
    rotationMode=BMM_Additive
    rotationSpace=BCS_BoneSpace
    scaleMode=BMM_Additive
    scaleSpace=BCS_BoneSpace
    translationMode=BMM_Additive
    translationSpace=BCS_BoneSpace
    alphaBoolBlend {blendOption=Linear}
    boneToModify {boneName=dyn_Hair_Back_04}
    scale {x=0.7}
   }
   n21 = "Transform (Modify) Bone - Bone: dyn_Hair_BackL_04"!(ComponentPose=@n20, bAlphaBoolEnabled:bool=@n19, Translation="0,0,0", Rotation="0,0,0", Scale="0.7,1,1") -> (Pose) {
    alphaInputType=Bool
    rotationMode=BMM_Additive
    rotationSpace=BCS_BoneSpace
    scaleMode=BMM_Additive
    scaleSpace=BCS_BoneSpace
    translationMode=BMM_Additive
    translationSpace=BCS_BoneSpace
    alphaBoolBlend {blendOption=Linear}
    boneToModify {boneName=dyn_Hair_BackL_04}
    scale {x=0.7}
   }
   n22 = "Transform (Modify) Bone - Bone: dyn_Hair_BackR_04"!(ComponentPose=@n21, bAlphaBoolEnabled:bool=@n19, Translation="0,0,0", Rotation="0,0,0", Scale="0.7,1,1") -> (Pose) {
    alphaInputType=Bool
    rotationMode=BMM_Additive
    rotationSpace=BCS_BoneSpace
    scaleMode=BMM_Additive
    scaleSpace=BCS_BoneSpace
    translationMode=BMM_Additive
    translationSpace=BCS_BoneSpace
    alphaBoolBlend {blendOption=Linear}
    boneToModify {boneName=dyn_Hair_BackR_04}
    scale {x=0.7}
   }
   n23 = "Transform (Modify) Bone - Bone: dyn_Hair_Back_04"(ComponentPose=@n22, bAlphaBoolEnabled=True, Translation="0,0,0", Rotation="0,-3,0", Scale="0.8,0.85,1") {
    alphaInputType=Bool
    rotationMode=BMM_Additive
    rotationSpace=BCS_BoneSpace
    scaleMode=BMM_Additive
    scaleSpace=BCS_BoneSpace
    translationMode=BMM_Additive
    translationSpace=BCS_BoneSpace
    alphaBoolBlend {blendOption=Linear}
    boneToModify {boneName=dyn_Hair_Back_04}
    rotation {yaw=-3}
    scale=<0.8 0.85>
   }
   n24 = OR Boolean!(A=Get bIsSkyDiving(), B=Get bIsInFrontEnd(), C) -> (ReturnValue)
   n25 = "Constraint - dyn_Hair_BackL_03a (PP)"!(ComponentPose=@n23, bAlphaBoolEnabled:bool=@n24, ConstraintWeights_0=0.2, ConstraintWeights_1:float=0.8) -> (Pose) {
    alphaInputType=Bool
    alphaBoolBlend {blendOption=Linear}
    boneToModify {boneName=dyn_Hair_BackL_03a}
    constraintSetup [{targetBone {boneName=dyn_Hair_BackL_03a}}, {targetBone {boneName=clavicle_l}}]
    constraintWeights [1,1]
   }
   n26 = "Transform (Modify) Bone - Bone: dyn_Hair_BackL_03a"!(ComponentPose=@n25, bAlphaBoolEnabled:bool=@n24, Translation="0,0,0", Rotation="0,0,-5", Scale="1,1,1") -> (Pose) {
    alphaInputType=Bool
    rotationMode=BMM_Additive
    rotationSpace=BCS_BoneSpace
    scaleMode=BMM_Additive
    scaleSpace=BCS_BoneSpace
    translationMode=BMM_Additive
    translationSpace=BCS_BoneSpace
    alphaBoolBlend {blendOption=Linear}
    boneToModify {boneName=dyn_Hair_BackL_03a}
    rotation {roll=-5}
   }
   n27 = "Transform (Modify) Bone - Bone: dyn_Hair_BackL_03"!(ComponentPose=@n26, bAlphaBoolEnabled:bool=True, Translation="0,0,0", Rotation="0,5,0", Scale="1,1,1") -> (Pose) {
    alphaInputType=Bool
    rotationMode=BMM_Additive
    rotationSpace=BCS_BoneSpace
    scaleMode=BMM_Additive
    scaleSpace=BCS_BoneSpace
    translationMode=BMM_Additive
    translationSpace=BCS_BoneSpace
    alphaBoolBlend {blendOption=Linear}
    boneToModify {boneName=dyn_Hair_BackL_03}
    rotation {yaw=5}
   }
   n28 = "Transform (Modify) Bone - Bone: dyn_Hair_BackL_05"!(ComponentPose=@n27, bAlphaBoolEnabled:bool=True, Translation="0,0,0", Rotation="0,-5,0", Scale="1,1,1") -> (Pose) {
    alphaInputType=Bool
    rotationMode=BMM_Additive
    rotationSpace=BCS_BoneSpace
    scaleMode=BMM_Additive
    scaleSpace=BCS_BoneSpace
    translationMode=BMM_Additive
    translationSpace=BCS_BoneSpace
    alphaBoolBlend {blendOption=Linear}
    boneToModify {boneName=dyn_Hair_BackL_05}
    rotation {yaw=-5}
   }
   n29 = "Constraint - dyn_Hair_BackR_03a (PP)"!(ComponentPose=@n28, bAlphaBoolEnabled:bool=@n24, ConstraintWeights_0=0.2, ConstraintWeights_1:float=0.8) -> (Pose) {
    alphaInputType=Bool
    alphaBoolBlend {blendOption=Linear}
    boneToModify {boneName=dyn_Hair_BackR_03a}
    constraintSetup [{targetBone {boneName=dyn_Hair_BackR_03a}}, {targetBone {boneName=clavicle_r}}]
    constraintWeights [1,1]
   }
   n30 = "Transform (Modify) Bone - Bone: dyn_Hair_BackR_03a"!(ComponentPose=@n29, bAlphaBoolEnabled:bool=True, Translation="0,0,0", Rotation="-10,0,5", Scale="1,1,1") -> (Pose) {
    alphaInputType=Bool
    rotationMode=BMM_Additive
    rotationSpace=BCS_BoneSpace
    scaleMode=BMM_Additive
    scaleSpace=BCS_BoneSpace
    translationMode=BMM_Additive
    translationSpace=BCS_BoneSpace
    alphaBoolBlend {blendOption=Linear}
    boneToModify {boneName=dyn_Hair_BackR_03a}
    rotation {pitch=-10, roll=5}
   }
   n31 = "Transform (Modify) Bone - Bone: dyn_Hair_BackR_03a"(ComponentPose=@n30, bAlphaBoolEnabled=@n24, Translation="0,0,0", Rotation="20,0,-5", Scale="1,1,1") {
    alphaInputType=Bool
    rotationMode=BMM_Additive
    rotationSpace=BCS_BoneSpace
    scaleMode=BMM_Additive
    scaleSpace=BCS_BoneSpace
    translationMode=BMM_Additive
    translationSpace=BCS_BoneSpace
    alphaBoolBlend {blendOption=Linear}
    boneToModify {boneName=dyn_Hair_BackR_03a}
    rotation {pitch=20, roll=-5}
   }
   n32 = "Transform (Modify) Bone - Bone: extra_crystal_head_l"(ComponentPose=@n31, Alpha=1, Translation="0,0,0", Rotation="0,0,0", Scale="1,1,1") {
    rotationSpace=BCS_BoneSpace
    scaleSpace=BCS_BoneSpace
    translationMode=BMM_Additive
    translationSpace=BCS_BoneSpace
    boneToModify {boneName=extra_crystal_head_l}
   }
   n33 = "Transform (Modify) Bone - Bone: extra_crystal_head_r"(ComponentPose=@n32, Alpha=1, Translation="0,0,0", Rotation="0,0,0", Scale="1,1,1") {
    rotationSpace=BCS_BoneSpace
    scaleSpace=BCS_BoneSpace
    translationMode=BMM_Additive
    translationSpace=BCS_BoneSpace
    boneToModify {boneName=extra_crystal_head_r}
   }
   Output Pose(Result=Component To Local(ComponentPose=@n33)) {layerGroup=None}
   n34 = Control Rig!(Source=Copy Pose From Mesh(SourceMeshComponent=Get Skel_Mesh()) {bUseAttachedParent=true}, Alpha, Cape:bool=false) -> (Pose) {
    alphaBoolBlend {blendOption=Linear}
    controlRigAssetReference {blueprintRigClass=&/F_MED_LexaBlitz_FaceAcc_CtrlRig.F_MED_LexaBlitz_FaceAcc_CtrlRig_C}
   }
   Save cached pose 'HairCachePose'(Pose=@n34) {cacheName=HairCachePose}
   n35 = Anim Dynamics - Bone: dyn_Hair_BackL_02(ComponentPose=Local To Component(LocalPose=Use cached pose 'HairCachePose'()), Alpha=0.2, ExternalForce="0,0,0") {
    angularDampingOverride=0.99
    angularSpringConstant=75
    bAngularSpring=true
    bGravityOverrideInSimSpace=true
    bOverrideAngularDamping=true
    bUseGravityOverride=true
    simulationSpace=RootRelative
    alphaBoolBlend {blendOption=Linear}
    boundBone {boneName=dyn_Hair_BackL_02}
    componentAppliedLinearAccClamp=<3500 3500 3000>
    componentLinearAccScale=<0.5 0.5 0.5>
    componentLinearVelScale=<0.65 0.65 0.65>
    gravityOverride {z=-980}
    physicsBodyDefinitions [{
      collisionType=CustomSphere
      sphereCollisionRadius=2
      boundBone {boneName=dyn_Hair_BackL_02}
      boxExtents=<40 40 40>
      constraintSetup {angularLimitsMax=<0 0 5>, angularLimitsMin=<0 0 0>, angularTarget=<0 0 0>}
      localJointOffset=<-2 -5 0>
     }]
    relativeSpaceBone {boneName=root}
   }
   n36 = Anim Dynamics - Bone: dyn_Hair_BackL_03(ComponentPose=@n35, Alpha=1, ExternalForce="0,0,0") {
    angularDampingOverride=0.99
    angularSpringConstant=75
    bAngularSpring=true
    bGravityOverrideInSimSpace=true
    bOverrideAngularDamping=true
    bUseGravityOverride=true
    simulationSpace=RootRelative
    alphaBoolBlend {blendOption=Linear}
    boundBone {boneName=dyn_Hair_BackL_03}
    componentAppliedLinearAccClamp=<3500 3500 3000>
    componentLinearAccScale=<0.5 0.5 0.5>
    componentLinearVelScale=<0.65 0.65 0.65>
    gravityOverride {z=-980}
    physicsBodyDefinitions [{
      collisionType=CustomSphere
      sphereCollisionRadius=2
      boundBone {boneName=dyn_Hair_BackL_03}
      boxExtents=<25 25 25>
      constraintSetup {angularLimitsMax=<0 0 20>, angularLimitsMin=<0 0 -2>, angularTarget=<0 0 0>}
      localJointOffset=<0 -5 0>
     }]
    relativeSpaceBone {boneName=root}
   }
   n37 = Anim Dynamics - Bone: dyn_Hair_BackL_03a(ComponentPose=@n36, Alpha=1, ExternalForce="0,0,0") {
    angularDampingOverride=0.99
    angularSpringConstant=75
    bAngularSpring=true
    bGravityOverrideInSimSpace=true
    bOverrideAngularDamping=true
    bUseGravityOverride=true
    bUseSphericalLimits=true
    simulationSpace=BoneRelative
    alphaBoolBlend {blendOption=Linear}
    boundBone {boneName=dyn_Hair_BackL_03a}
    componentAppliedLinearAccClamp=<3500 3500 3000>
    componentLinearAccScale=<0.5 0.5 0.5>
    componentLinearVelScale=<0.65 0.65 0.65>
    gravityOverride {z=-980}
    physicsBodyDefinitions [{
      collisionType=CustomSphere
      sphereCollisionRadius=2
      boundBone {boneName=dyn_Hair_BackL_03a}
      boxExtents=<25 25 25>
      constraintSetup {twistAxis=AxisY, angularLimitsMax=<15 10 7>, angularLimitsMin=<-15 -10 -2>, angularTarget=<0 0 0>}
      localJointOffset=<0.2 -5 0.2>
     }]
    relativeSpaceBone {boneName=root}
    sphericalLimits [
     {limitRadius=5, drivingBone {boneName=clavicle_l}, sphereLocalOffset=<5 3 0>}
     {limitRadius=0, drivingBone {boneName=None}, sphereLocalOffset=<0 0 0>}
    ]
   }
   n38 = Anim Dynamics - Bone: dyn_Hair_BackL_04(ComponentPose=@n37, Alpha=1, ExternalForce="0,0,0") {
    angularDampingOverride=0.99
    angularSpringConstant=75
    bAngularSpring=true
    bGravityOverrideInSimSpace=true
    bOverrideAngularDamping=true
    bUseGravityOverride=true
    bUseSphericalLimits=true
    simulationSpace=RootRelative
    alphaBoolBlend {blendOption=Linear}
    boundBone {boneName=dyn_Hair_BackL_04}
    componentAppliedLinearAccClamp=<3500 3500 3000>
    componentLinearAccScale=<0.5 0.5 0.5>
    componentLinearVelScale=<0.65 0.65 0.65>
    gravityOverride {z=-980}
    physicsBodyDefinitions [{
      collisionType=CustomSphere
      sphereCollisionRadius=2
      boundBone {boneName=dyn_Hair_BackL_04}
      boxExtents=<23 23 23>
      constraintSetup {angularLimitsMax=<15 2 30>, angularLimitsMin=<-5 -2 -2>, angularTarget=<0 0 0>}
      localJointOffset=<0 -5 0>
     }]
    relativeSpaceBone {boneName=root}
    sphericalLimits [
     {limitRadius=0, drivingBone {boneName=None}, sphereLocalOffset=<0 0 0>}
     {limitRadius=0, drivingBone {boneName=None}, sphereLocalOffset=<0 0 0>}
    ]
   }
   n39 = Anim Dynamics - Bone: dyn_Hair_BackL_05(ComponentPose=@n38, Alpha=1, ExternalForce="0,0,0") {
    angularDampingOverride=0.99
    angularSpringConstant=75
    bAngularSpring=true
    bGravityOverrideInSimSpace=true
    bOverrideAngularDamping=true
    bUseGravityOverride=true
    bUseSphericalLimits=true
    simulationSpace=RootRelative
    alphaBoolBlend {blendOption=Linear}
    boundBone {boneName=dyn_Hair_BackL_05}
    componentAppliedLinearAccClamp=<3500 3500 3000>
    componentLinearAccScale=<0.5 0.5 0.5>
    componentLinearVelScale=<0.65 0.65 0.65>
    gravityOverride {z=-980}
    physicsBodyDefinitions [{
      collisionType=CustomSphere
      sphereCollisionRadius=2
      boundBone {boneName=dyn_Hair_BackL_05}
      boxExtents=<23 23 23>
      constraintSetup {angularLimitsMax=<15 2 30>, angularLimitsMin=<-5 -2 -5>, angularTarget=<0 0 0>}
      localJointOffset=<0 -5 0>
     }]
    relativeSpaceBone {boneName=root}
    sphericalLimits [
     {limitRadius=18, drivingBone {boneName=upperarm_l}, sphereLocalOffset=<5 14 0>}
     {limitRadius=0, drivingBone {boneName=None}, sphereLocalOffset=<0 0 0>}
    ]
   }
   n40 = Anim Dynamics - Bone: dyn_Hair_BackL_06(ComponentPose=@n39, Alpha=1, ExternalForce="0,0,0") {
    angularDampingOverride=0.99
    angularSpringConstant=75
    bAngularSpring=true
    bGravityOverrideInSimSpace=true
    bOverrideAngularDamping=true
    bUseGravityOverride=true
    bUseSphericalLimits=true
    simulationSpace=RootRelative
    alphaBoolBlend {blendOption=Linear}
    boundBone {boneName=dyn_Hair_BackL_06}
    componentAppliedLinearAccClamp=<3500 3500 3000>
    componentLinearAccScale=<0.5 0.5 0.5>
    componentLinearVelScale=<0.65 0.65 0.65>
    gravityOverride {z=-980}
    physicsBodyDefinitions [{
      collisionType=CustomSphere
      sphereCollisionRadius=2
      boundBone {boneName=dyn_Hair_BackL_06}
      boxExtents=<23 23 23>
      constraintSetup {angularLimitsMax=<15 2 30>, angularLimitsMin=<-5 -2 -5>, angularTarget=<0 0 0>}
      localJointOffset=<0 -5 0>
     }]
    relativeSpaceBone {boneName=root}
    sphericalLimits [
     {limitRadius=18, drivingBone {boneName=upperarm_l}, sphereLocalOffset=<5 14 0>}
     {limitRadius=0, drivingBone {boneName=None}, sphereLocalOffset=<0 0 0>}
    ]
   }
   n41 = Anim Dynamics - Bone: dyn_Hair_Back_02(ComponentPose=@n40, Alpha=0.2, ExternalForce="0,0,0") {
    angularDampingOverride=0.99
    angularSpringConstant=75
    bAngularSpring=true
    bGravityOverrideInSimSpace=true
    bOverrideAngularDamping=true
    bUseGravityOverride=true
    simulationSpace=RootRelative
    alphaBoolBlend {blendOption=Linear}
    boundBone {boneName=dyn_Hair_Back_02}
    componentAppliedLinearAccClamp=<3500 3500 3000>
    componentLinearAccScale=<0.5 0.5 0.5>
    componentLinearVelScale=<0.65 0.65 0.65>
    gravityOverride {z=-980}
    physicsBodyDefinitions [{
      collisionType=CustomSphere
      sphereCollisionRadius=2
      boundBone {boneName=dyn_Hair_Back_02}
      boxExtents=<40 40 40>
      constraintSetup {angularLimitsMax=<0 0 5>, angularLimitsMin=<0 0 0>, angularTarget=<0 0 0>}
      localJointOffset=<-2 -5 0>
     }]
    relativeSpaceBone {boneName=root}
   }
   n42 = Anim Dynamics - Bone: dyn_Hair_Back_03(ComponentPose=@n41, Alpha=1, ExternalForce="0,0,0") {
    angularDampingOverride=0.99
    angularSpringConstant=75
    bAngularSpring=true
    bGravityOverrideInSimSpace=true
    bOverrideAngularDamping=true
    bUseGravityOverride=true
    simulationSpace=RootRelative
    alphaBoolBlend {blendOption=Linear}
    boundBone {boneName=dyn_Hair_Back_03}
    componentAppliedLinearAccClamp=<3500 3500 3000>
    componentLinearAccScale=<0.5 0.5 0.5>
    componentLinearVelScale=<0.65 0.65 0.65>
    gravityOverride {z=-980}
    physicsBodyDefinitions [{
      collisionType=CustomSphere
      sphereCollisionRadius=2
      boundBone {boneName=dyn_Hair_Back_03}
      boxExtents=<25 25 25>
      constraintSetup {angularLimitsMax=<0 0 20>, angularLimitsMin=<0 0 -2>, angularTarget=<0 0 0>}
      localJointOffset=<0 -5 0>
     }]
    relativeSpaceBone {boneName=root}
   }
   n43 = Anim Dynamics - Bone: dyn_Hair_Back_04(ComponentPose=@n42, Alpha=1, ExternalForce="0,0,0") {
    angularDampingOverride=0.99
    angularSpringConstant=75
    bAngularSpring=true
    bGravityOverrideInSimSpace=true
    bOverrideAngularDamping=true
    bUseGravityOverride=true
    simulationSpace=RootRelative
    alphaBoolBlend {blendOption=Linear}
    boundBone {boneName=dyn_Hair_Back_04}
    componentAppliedLinearAccClamp=<3500 3500 3000>
    componentLinearAccScale=<0.5 0.5 0.5>
    componentLinearVelScale=<0.65 0.65 0.65>
    gravityOverride {z=-980}
    physicsBodyDefinitions [{
      collisionType=CustomSphere
      sphereCollisionRadius=2
      boundBone {boneName=dyn_Hair_Back_04}
      boxExtents=<23 23 23>
      constraintSetup {angularLimitsMax=<2 2 30>, angularLimitsMin=<-2 -2 -2>, angularTarget=<0 0 0>}
      localJointOffset=<0 -5 0>
     }]
    relativeSpaceBone {boneName=root}
   }
   n44 = Anim Dynamics - Bone: dyn_Hair_Back_05(ComponentPose=@n43, Alpha=1, ExternalForce="0,0,0") {
    angularDampingOverride=0.99
    angularSpringConstant=75
    bAngularSpring=true
    bGravityOverrideInSimSpace=true
    bOverrideAngularDamping=true
    bUseGravityOverride=true
    bUseSphericalLimits=true
    simulationSpace=RootRelative
    alphaBoolBlend {blendOption=Linear}
    boundBone {boneName=dyn_Hair_Back_05}
    componentAppliedLinearAccClamp=<3500 3500 3000>
    componentLinearAccScale=<0.5 0.5 0.5>
    componentLinearVelScale=<0.65 0.65 0.65>
    gravityOverride {z=-980}
    physicsBodyDefinitions [{
      collisionType=CustomSphere
      sphereCollisionRadius=2
      boundBone {boneName=dyn_Hair_Back_05}
      boxExtents=<23 23 23>
      constraintSetup {angularLimitsMax=<3 2 30>, angularLimitsMin=<-3 -2 -5>, angularTarget=<0 0 0>}
      localJointOffset=<0 -5 0>
     }]
    relativeSpaceBone {boneName=root}
    sphericalLimits [
     {limitRadius=0, drivingBone {boneName=None}, sphereLocalOffset=<0 0 0>}
     {limitRadius=0, drivingBone {boneName=None}, sphereLocalOffset=<0 0 0>}
    ]
   }
   n45 = Anim Dynamics - Bone: dyn_Hair_Back_06(ComponentPose=@n44, Alpha=1, ExternalForce="0,0,0") {
    angularDampingOverride=0.99
    angularSpringConstant=75
    bAngularSpring=true
    bGravityOverrideInSimSpace=true
    bOverrideAngularDamping=true
    bUseGravityOverride=true
    simulationSpace=RootRelative
    alphaBoolBlend {blendOption=Linear}
    boundBone {boneName=dyn_Hair_Back_06}
    componentAppliedLinearAccClamp=<3500 3500 3000>
    componentLinearAccScale=<0.5 0.5 0.5>
    componentLinearVelScale=<0.65 0.65 0.65>
    gravityOverride {z=-980}
    physicsBodyDefinitions [{
      collisionType=CustomSphere
      sphereCollisionRadius=2
      boundBone {boneName=dyn_Hair_Back_06}
      boxExtents=<23 23 23>
      constraintSetup {angularLimitsMax=<4 2 30>, angularLimitsMin=<-4 -2 -5>, angularTarget=<0 0 0>}
      localJointOffset=<0 -5 0>
     }]
    relativeSpaceBone {boneName=root}
    sphericalLimits [
     {limitRadius=18, drivingBone {boneName=upperarm_r}, sphereLocalOffset=<5 -14 0>}
     {limitRadius=18, drivingBone {boneName=upperarm_l}, sphereLocalOffset=<5 14 0>}
    ]
   }
   n46 = Anim Dynamics - Bone: dyn_Hair_BackR_02(ComponentPose=@n45, Alpha=0.2, ExternalForce="0,0,0") {
    angularDampingOverride=0.99
    angularSpringConstant=75
    bAngularSpring=true
    bGravityOverrideInSimSpace=true
    bOverrideAngularDamping=true
    bUseGravityOverride=true
    simulationSpace=RootRelative
    alphaBoolBlend {blendOption=Linear}
    boundBone {boneName=dyn_Hair_BackR_02}
    componentAppliedLinearAccClamp=<3500 3500 3000>
    componentLinearAccScale=<0.5 0.5 0.5>
    componentLinearVelScale=<0.65 0.65 0.65>
    gravityOverride {z=-980}
    physicsBodyDefinitions [{
      collisionType=CustomSphere
      sphereCollisionRadius=2
      boundBone {boneName=dyn_Hair_BackR_02}
      boxExtents=<40 40 40>
      constraintSetup {angularLimitsMax=<0 0 5>, angularLimitsMin=<0 0 0>, angularTarget=<0 0 0>}
      localJointOffset=<-2 -5 0>
     }]
    relativeSpaceBone {boneName=root}
   }
   n47 = Anim Dynamics - Bone: dyn_Hair_BackR_03(ComponentPose=@n46, Alpha=1, ExternalForce="0,0,0") {
    angularDampingOverride=0.99
    angularSpringConstant=75
    bAngularSpring=true
    bGravityOverrideInSimSpace=true
    bOverrideAngularDamping=true
    bUseGravityOverride=true
    bUseSphericalLimits=true
    simulationSpace=RootRelative
    alphaBoolBlend {blendOption=Linear}
    boundBone {boneName=dyn_Hair_BackR_03}
    componentAppliedLinearAccClamp=<3500 3500 3000>
    componentLinearAccScale=<0.5 0.5 0.5>
    componentLinearVelScale=<0.65 0.65 0.65>
    gravityOverride {z=-980}
    physicsBodyDefinitions [{
      collisionType=CustomSphere
      sphereCollisionRadius=2
      boundBone {boneName=dyn_Hair_BackR_03}
      boxExtents=<25 25 25>
      constraintSetup {angularLimitsMax=<0 0 20>, angularLimitsMin=<0 0 -2>, angularTarget=<0 0 0>}
      localJointOffset=<0 -5 0>
     }]
    relativeSpaceBone {boneName=root}
    sphericalLimits [{limitRadius=18, drivingBone {boneName=None}, sphereLocalOffset=<-5 -13 0>}]
   }
   n48 = Anim Dynamics - Bone: dyn_Hair_BackR_03a(ComponentPose=@n47, Alpha=1, ExternalForce="0,0,0") {
    angularDampingOverride=0.99
    angularSpringConstant=75
    bAngularSpring=true
    bGravityOverrideInSimSpace=true
    bOverrideAngularDamping=true
    bUseGravityOverride=true
    bUseSphericalLimits=true
    simulationSpace=RootRelative
    alphaBoolBlend {blendOption=Linear}
    boundBone {boneName=dyn_Hair_BackR_03a}
    componentAppliedLinearAccClamp=<3500 3500 3000>
    componentLinearAccScale=<0.5 0.5 0.5>
    componentLinearVelScale=<0.65 0.65 0.65>
    gravityOverride {z=-980}
    physicsBodyDefinitions [{
      collisionType=CustomSphere
      sphereCollisionRadius=2
      boundBone {boneName=dyn_Hair_BackR_03a}
      boxExtents=<23 23 23>
      constraintSetup {angularLimitsMax=<1 2 20>, angularLimitsMin=<-2 -2 -2>, angularTarget=<0 0 0>}
      localJointOffset=<-2 -2 0>
     }]
    relativeSpaceBone {boneName=root}
    sphericalLimits [
     {limitRadius=5, drivingBone {boneName=clavicle_r}, sphereLocalOffset=<-5 -3 0>}
     {limitRadius=0, drivingBone {boneName=None}, sphereLocalOffset=<-2 0 -3>}
    ]
   }
   n49 = Anim Dynamics - Bone: dyn_Hair_BackR_04(ComponentPose=@n48, Alpha=1, ExternalForce="0,0,0") {
    angularDampingOverride=0.99
    angularSpringConstant=75
    bAngularSpring=true
    bGravityOverrideInSimSpace=true
    bOverrideAngularDamping=true
    bUseGravityOverride=true
    simulationSpace=RootRelative
    alphaBoolBlend {blendOption=Linear}
    boundBone {boneName=dyn_Hair_BackR_04}
    componentAppliedLinearAccClamp=<3500 3500 3000>
    componentLinearAccScale=<0.5 0.5 0.5>
    componentLinearVelScale=<0.65 0.65 0.65>
    gravityOverride {z=-980}
    physicsBodyDefinitions [{
      collisionType=CustomSphere
      sphereCollisionRadius=2
      boundBone {boneName=dyn_Hair_BackR_04}
      boxExtents=<23 23 23>
      constraintSetup {angularLimitsMax=<5 2 30>, angularLimitsMin=<-15 -2 -2>, angularTarget=<0 0 0>}
      localJointOffset=<0 -5 0>
     }]
    relativeSpaceBone {boneName=root}
   }
   n50 = Anim Dynamics - Bone: dyn_Hair_BackR_05(ComponentPose=@n49, Alpha=1, ExternalForce="0,0,0") {
    angularDampingOverride=0.99
    angularSpringConstant=75
    bAngularSpring=true
    bGravityOverrideInSimSpace=true
    bOverrideAngularDamping=true
    bUseGravityOverride=true
    bUseSphericalLimits=true
    simulationSpace=RootRelative
    alphaBoolBlend {blendOption=Linear}
    boundBone {boneName=dyn_Hair_BackR_05}
    componentAppliedLinearAccClamp=<3500 3500 3000>
    componentLinearAccScale=<0.5 0.5 0.5>
    componentLinearVelScale=<0.65 0.65 0.65>
    gravityOverride {z=-980}
    physicsBodyDefinitions [{
      collisionType=CustomSphere
      sphereCollisionRadius=2
      boundBone {boneName=dyn_Hair_BackR_05}
      boxExtents=<23 23 23>
      constraintSetup {angularLimitsMax=<5 0 30>, angularLimitsMin=<-15 0 -5>, angularTarget=<0 0 0>}
      localJointOffset=<0 -5 0>
     }]
    relativeSpaceBone {boneName=root}
    sphericalLimits [
     {limitRadius=18, drivingBone {boneName=upperarm_r}, sphereLocalOffset=<-10 -12 0>}
     {limitRadius=0, drivingBone {boneName=None}, sphereLocalOffset=<0 0 0>}
    ]
   }
   n51 = Anim Dynamics - Bone: dyn_Hair_BackR_06(ComponentPose=@n50, Alpha=1, ExternalForce="0,0,0") {
    angularDampingOverride=0.99
    angularSpringConstant=75
    bAngularSpring=true
    bGravityOverrideInSimSpace=true
    bOverrideAngularDamping=true
    bUseGravityOverride=true
    bUseSphericalLimits=true
    simulationSpace=RootRelative
    alphaBoolBlend {blendOption=Linear}
    boundBone {boneName=dyn_Hair_BackR_06}
    componentAppliedLinearAccClamp=<3500 3500 3000>
    componentLinearAccScale=<0.5 0.5 0.5>
    componentLinearVelScale=<0.65 0.65 0.65>
    gravityOverride {z=-980}
    physicsBodyDefinitions [{
      collisionType=CustomSphere
      sphereCollisionRadius=2
      boundBone {boneName=dyn_Hair_BackR_06}
      boxExtents=<23 23 23>
      constraintSetup {angularLimitsMax=<5 0 30>, angularLimitsMin=<-15 0 -5>, angularTarget=<0 0 0>}
      localJointOffset=<0 -5 0>
     }]
    relativeSpaceBone {boneName=root}
    sphericalLimits [
     {limitRadius=18, drivingBone {boneName=upperarm_r}, sphereLocalOffset=<-10 -12 0>}
     {limitRadius=0, drivingBone {boneName=None}, sphereLocalOffset=<0 0 0>}
    ]
   }
   Save cached pose 'HairDynPose'(Pose=Component To Local(ComponentPose=@n51)) {cacheName=HairDynPose}
  ]}
 EventGraph {nodes [
   Event BlueprintUpdateAnimation()
   > n1 = Sequence()
   > n2 = "Gravity Override State Function (Parameter Asset) (Nativized)"(
    CharacterPartAnimInstance=Self-Reference()
    GameTimeInSeconds=GetGameTimeInSeconds()
    DebugJointName=spine_05
    Parameters=&/F_MED_LexaBlitz_FaceAcc_Parameters
   )
   > Set Gravity Override(Gravity Override=@n2)
   Set CrystalAnim_A(
    execute=@n1.then_1
    CrystalAnim_A:Vector=MakeWaveVector!(Amplitudes:Rotator="1,0.5,1.75", Frequencies:Rotator="100,50,80", Offsets:Rotator="0,0,0", bUseNoise:bool=false, Min:Vector, Max:Vector) -> (ReturnValue:Vector)
   ) -> (Output_Get:Vector)
   > Set CrystalAnim_B(CrystalAnim_B:Vector=MakeWaveVector(Amplitudes="1,0.5,1.75", Frequencies="100,50,80", Offsets="-45,-45,90", bUseNoise=false)) -> (Output_Get:Vector)
   Event OnCharacterCustomizationLoaded()
   n3 = CheckForBackbling()
   > Set isWearingPackBool(isWearingPackBool:bool=OR Boolean(A=@n3, B=@n3.bIsBackblingACape, C=@n3.bIsBackblingLong)) -> (Output_Get:bool)
   > Set is_WearingCape(is_WearingCape:bool=@n3.bIsBackblingLong) -> (Output_Get:bool)
   Event BlueprintInitializeAnimation()
   > Set Skel_Mesh(Skel_Mesh=GetOwnerSkeletalMeshComponent())
   AND Boolean(A=@n3, B=NOT Boolean(A=@n3.bIsBackblingLong))
  ]}
]
variables [
 CrystalAnim_A {type=Vector}
 CrystalAnim_B {type=Vector}
 Gravity Override {type=Vector}
 Gravity Override Params {
  type=GravityOverrideParamsStruct
  default {
   onGround_Standing {windAmplitude=<-10 -150 0>, windFrequency=<-7 -1 -1>, windOffset=<-10 -100 0>}
   skydive_DivingDown {gravityOverride=<-500 0 0>}
   skydive_FloatingDown {gravityOverride=<0 -400 100>}
   skydive_GoingUp {gravityOverride=<0 -300 -200>}
   skydive_Parachuting {gravityOverride=<0 -600 -980>}
  }
 }
 is_WearingCape {type=boolean}
 isWearingPackBool {type=boolean}
 Skel_Mesh {type=SkeletalMeshComponent}
]
